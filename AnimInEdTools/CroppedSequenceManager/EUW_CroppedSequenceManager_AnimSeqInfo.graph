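# Editor utility widget in the Cropped Sequence Manager that shows a single property picker for an AnimSequence.
# On construct it binds the property view to itself and clears AnimSequence.
# Property edits run Python that passes the sequence and widget id to the CroppedSequenceManagerUI module's linked_anim_sequence_property_changed handler.
# PreConstruct applies the scale variable as render scale and derives compensating padding from it through clamped range mappings.
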EditorUtilityWidgetBlueprint /AnimInEdTools/CroppedSequenceManager/EUW_CroppedSequenceManager_AnimSeqInfo
graphs [EventGraph {nodes [
   "On Property Changed (SinglePropertyView_AnimSequence)"!() -> (OutputDelegate, then, PropertyName:Name=None) {is=title K2Node_ComponentBoundEvent}
   > Execute Python Script!(
    execute
    PythonScript="import sys\r\nCSM = sys.modules['CroppedSequenceManagerUI']\r\n\r\nCSM.cls.linked_anim_sequence_property_changed(anim_sequence, widget_id)\r\n"
    anim_sequence:Anim Sequence Object Reference=Get AnimSequence() -> (AnimSequence:Anim Sequence Object Reference)
    widget_id:int=Get widget id() -> (widget id:int=0)
   ) -> (then, ReturnValue) {inputs [anim_sequence, widget_id]}
   Event PreConstruct()
   n1 = Self-Reference()
   n2 = Get scale() -> (scale:double)
   > SetRenderScale!(execute, self:Widget Object Reference=@n1, Scale_X:double=@n2, Scale_Y:double=@n2) -> (then)
   n3 = MapRangeClamped(Value=@n2, InRangeA=0.3, InRangeB=1, OutRangeA=-200, OutRangeB=2)
   n4 = MapRangeClamped(Value=@n2, InRangeA=0.3, InRangeB=1, OutRangeA=-30, OutRangeB=2)
   > SetPadding!(
    execute
    self:User Widget Object Reference=@n1
    InPadding_Left:float=MapRangeClamped(Value=@n3, InRangeA=0.3, InRangeB=1, OutRangeA=-30, OutRangeB=0)
    InPadding_Top:float=@n4
    InPadding_Right:float=@n3
    InPadding_Bottom:float=@n4
   ) -> (then)
   Event Construct()
   > SetObject(self=Get SinglePropertyView_AnimSequence() -> (SinglePropertyView_AnimSequence:Single Property View Object Reference), NewObject=@n1)
   > Set AnimSequence(AnimSequence:Anim Sequence Object Reference) -> (Output_Get:Anim Sequence Object Reference)
   Event Tick()
  ]}]
variables [
 AnimSequence {type=AnimSequence, default=null}
 scale {type=number, default=1}
 widget id {type=integer}
]
widgets {
 info {rootWidgetClass=&HorizontalBox, widgetCount=2}
 widgets [
  HorizontalBox_0 {widgetClassPath=&HorizontalBox, widget=HorizontalBox_0}
  SinglePropertyView_AnimSequence {
   widgetClassPath=&SinglePropertyView
   parent=HorizontalBox_0
   widget=SinglePropertyView_AnimSequence
   properties {nameOverride=" ", propertyName=AnimSequence}
   slot {class=HorizontalBoxSlot, properties {horizontalAlignment=HAlign_Left, verticalAlignment=VAlign_Center, padding=<ltrb -16 2 0 2>}}
  }
 ]
}
