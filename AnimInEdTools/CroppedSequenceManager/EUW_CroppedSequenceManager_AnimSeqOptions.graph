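# Editor utility widget providing a collapsible options panel for a linked AnimSequence in the Cropped Sequence Manager.
# Looping, additive and root motion edits call Python animation_sequence_utils; FPS, interpolation, curve-skip and retarget-source changes go to the CroppedSequenceManagerUI module.
# InitWidgetValues reads state back from the assets, using supress_update to block feedback events.
# Skippable facial curves come from the project's per-actor-class list, matched to the track's actor class.
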EditorUtilityWidgetBlueprint /AnimInEdTools/CroppedSequenceManager/EUW_CroppedSequenceManager_AnimSeqOptions
graphs [
 CurvesAsStr {nodes [
   CurvesAsStr() {is=title K2Node_FunctionEntry}
   > n1 = For Each Loop(Array=Get Actor Facial Curves() -> (Actor Facial Curves:Array of Strings))
   Return Node!(execute=@n1.Completed, str:String=Get NewLocalVar_0() -> (NewLocalVar_0:String))
   > Set NewLocalVar_0(NewLocalVar_0:String=Append!(A=Get NewLocalVar_0(), B=@n1, C:String="\r\n") -> (ReturnValue)) -> (Output_Get:String)
  ]}
 EventGraph {nodes [
   n1 = "On Check State Changed (CheckBoxLooping)"!() -> (OutputDelegate, then, bIsChecked:bool) {is=title K2Node_ComponentBoundEvent}
   > n2 = Branch(Condition=Get supress_update() -> (supress_update:bool))
   Execute Python Script!(
    execute=@n2.else
    PythonScript="from utilities import animation_sequence_utils\r\n\r\nanimation_sequence_utils.set_looping(anim_sequence, enabled)\r\n"
    anim_sequence:Anim Sequence Object Reference=Get AnimSequence() -> (AnimSequence:Anim Sequence Object Reference)
    enabled:bool=@n1.bIsChecked
   ) -> (then, ReturnValue) {inputs [anim_sequence, enabled]}
   n3 = "On Selection Changed (ComboBox_IsAdditive)"!() -> (OutputDelegate, then, SelectedItem:String, SelectionType:ESelectInfo Enum) {is=title K2Node_ComponentBoundEvent}
   > n4 = Branch(Condition=Get supress_update())
   n5 = Execute Python Script!(
    execute=@n4.else
    PythonScript="from utilities import animation_sequence_utils\r\n\r\nanimation_sequence_utils.set_additive(anim_sequence, additive_type)"
    anim_sequence=Get AnimSequence()
    additive_type:String=@n3.SelectedItem
   ) -> (then, ReturnValue) {inputs [anim_sequence, additive_type]}
   InitWidgetValues(execute=[@n8.then, @n5.then, @n13.then, @n16.then])
   n6 = "On Check State Changed (CheckBoxSetRetargetSourceAsset)"!() -> (OutputDelegate, then, bIsChecked:bool) {is=title K2Node_ComponentBoundEvent}
   n7 = Branch(execute=[@n10.then, @n6.then, @n9.then, @n26.then, @n27.then], Condition=Get supress_update())
   n8 = Execute Python Script!(
    execute=@n7.else
    PythonScript="import sys\r\nCSM = sys.modules['CroppedSequenceManagerUI']\r\n\r\nCSM.cls.linked_anim_sequence_property_changed(widget_id = widget_id, new_anim_sequence = anim_sequence, custom_fps=fps, interpolation=interpolation, skip_zero_value_curves=skip_zero_value_curves, set_retarget_source_asset=set_retarget_source_asset,\r\nface_curves_list=face_curves_list)\r\n"
    widget_id:int=Get widget id() -> (widget id:int=0)
    anim_sequence=Get AnimSequence()
    fps:String=GetSelectedOption!(self:"Combo Box (String) Object Reference"=Get ComboBox_FPS() -> (ComboBox_FPS:Editor Utility Combo Box String Object Reference)) -> (ReturnValue:String)
    interpolation:String=GetSelectedOption(self=Get ComboBox_Interpolation() -> (ComboBox_Interpolation:Editor Utility Combo Box String Object Reference))
    set_retarget_source_asset:bool=IsChecked(self=Get CheckBoxSetRetargetSourceAsset() -> (CheckBoxSetRetargetSourceAsset:Editor Utility Check Box Object Reference))
    skip_zero_value_curves:bool=IsChecked(self=Get CheckBoxSkipZeroCurves() -> (CheckBoxSkipZeroCurves:Editor Utility Check Box Object Reference))
    face_curves_list:Array of Strings=Get Apply Facial Curves() -> (Apply Facial Curves:Array of Strings)
   ) -> (then, ReturnValue) {inputs [widget_id, anim_sequence, fps, interpolation, set_retarget_source_asset, skip_zero_value_curves, face_curves_list]}
   n9 = "On Check State Changed (CheckBoxSkipZeroCurves)"!() -> (OutputDelegate, then, bIsChecked:bool) {is=title K2Node_ComponentBoundEvent}
   "On Clicked (ButtonFaceCurves)"() {is=title K2Node_ComponentBoundEvent}
   > Show Message Dialog(
    Title=MakeLiteralText(Value="LOCTEXT(\"Cropped Sequence Manager\")")
    Message=Get Face Curves Msg() -> (Face Curves Msg:Text)
    MessageType=Ok
    DefaultValue=No
    MessageCategory=Info
   )
   "On Check State Changed (CheckBoxSkipFaceCurves)"!() -> (OutputDelegate, then, bIsChecked:bool) {is=title K2Node_ComponentBoundEvent}
   > n10 = Set Facial Curves() {is=title K2Node_CallFunction}
   n11 = "On Selection Changed (ComboBox_AdditiveType)"!() -> (OutputDelegate, then, SelectedItem:String, SelectionType:ESelectInfo Enum) {is=title K2Node_ComponentBoundEvent}
   > n12 = Branch(Condition=Get supress_update())
   n13 = Execute Python Script!(
    execute=@n12.else
    PythonScript="from utilities import animation_sequence_utils\r\n\r\nif additive_basepose_type == 'SkelRef Pose':\r\n   animation_sequence_utils.set_additive(anim_sequence, additive_type, additive_basepose_type='SkelRef')\r\nelse:\r\n   animation_sequence_utils.set_additive(anim_sequence, additive_type, additive_basepose_type=None )"
    anim_sequence=Get AnimSequence()
    additive_basepose_type:String=@n11.SelectedItem
    additive_type:String=GetSelectedOption(self=Get ComboBox_IsAdditive() -> (ComboBox_IsAdditive:Editor Utility Combo Box String Object Reference))
   ) -> (then, ReturnValue) {inputs [anim_sequence, additive_basepose_type, additive_type]}
   n14 = "On Check State Changed (CheckBoxRootMotionLock)"!() -> (OutputDelegate, then, bIsChecked:bool) {is=title K2Node_ComponentBoundEvent}
   n15 = Branch(execute=[@n14.then, @n19.then], Condition=Get supress_update())
   n16 = Execute Python Script!(
    execute=@n15.else
    PythonScript="from utilities import animation_sequence_utils\r\n\r\nanimation_sequence_utils.set_root_lock(anim_sequence, is_locked, option)\r\n"
    anim_sequence=Get AnimSequence()
    is_locked:bool=@n14.bIsChecked
    option:String=GetSelectedOption(self=Get ComboBox_RootMotionMode() -> (ComboBox_RootMotionMode:Editor Utility Combo Box String Object Reference))
   ) -> (then, ReturnValue) {inputs [anim_sequence, is_locked, option]}
   n17 = "On Check State Changed (CheckBoxRootMotionEnable)"!() -> (OutputDelegate, then, bIsChecked:bool) {is=title K2Node_ComponentBoundEvent}
   > n18 = Branch(Condition=Get supress_update())
   Execute Python Script!(
    execute=@n18.else
    PythonScript="from utilities import animation_sequence_utils\r\n\r\nanimation_sequence_utils.set_root_motion(anim_sequence, enabled)\r\n"
    anim_sequence=Get AnimSequence()
    enabled:bool=@n17.bIsChecked
   ) -> (then, ReturnValue) {inputs [anim_sequence, enabled]}
   n19 = "On Selection Changed (ComboBox_RootMotionMode)"!() -> (OutputDelegate, then, SelectedItem:String, SelectionType:ESelectInfo Enum) {is=title K2Node_ComponentBoundEvent}
   "On Clicked (VisibilityButton)"() {is=title K2Node_ComponentBoundEvent}
   n20 = "Equal (Enum)"(
    A:ESlateVisibility Enum=Get Visibility!(self:Widget Object Reference=Get VerticalBoxOPtions() -> (VerticalBoxOPtions:Vertical Box Object Reference)) -> (Visibility:ESlateVisibility Enum=Visible)
    B:ESlateVisibility Enum=Visible
   )
   > Set ShowHUD(ShowHUD:bool=@n20) -> (Output_Get:bool)
   > Execute Python Script!(
    execute
    PythonScript="import sys\r\nCSM = sys.modules['CroppedSequenceManagerUI']\r\n\r\nCSM.cls.save_sub_segment_options_hud_visibility(widget_id, is_visible)\r\n"
    widget_id:int=Get widget id() -> (widget id=0)
    is_visible:bool=NOT Boolean(A=@n20)
   ) -> (then, ReturnValue) {inputs [widget_id,is_visible]}
   > n21 = Branch(Condition=@n20)
   n22 = Get VerticalBoxOPtions()
   > SetVisibility(self=@n22, InVisibility=Collapsed)
   n23 = Get AssetImageVisibleicon() -> (AssetImageVisibleicon:Image Object Reference)
   > SetVisibility(self=@n23, InVisibility=Collapsed)
   n24 = Get AssetImageInivisibleIcon() -> (AssetImageInivisibleIcon:Image Object Reference)
   > SetVisibility(self=@n24, InVisibility=Visible)
   SetVisibility(execute=[@n28.then, @n21.else], self=@n22, InVisibility=Visible)
   > SetVisibility(self=@n23, InVisibility=Visible)
   > SetVisibility(self=@n24, InVisibility=Collapsed)
   n25 = "On Check State Changed (CheckBoxCustomFPS)"!() -> (OutputDelegate, then, bIsChecked:bool) {is=title K2Node_ComponentBoundEvent}
   > SetIsEnabled(self=Get ComboBox_FPS(), bInIsEnabled=@n25.bIsChecked)
   n26 = "On Selection Changed (ComboBox_FPS)"!() -> (OutputDelegate, then, SelectedItem:String, SelectionType:ESelectInfo Enum) {is=title K2Node_ComponentBoundEvent}
   n27 = "On Selection Changed (ComboBox_Interpolation)"!() -> (OutputDelegate, then, SelectedItem:String, SelectionType:ESelectInfo Enum) {is=title K2Node_ComponentBoundEvent}
   n28 = Set Options HUD Visible() {is=title K2Node_CustomEvent}
   Event Construct()
   > n29 = Sequence()
   > n30 = Execute Python Script!(
    execute
    PythonScript="import sys\r\nCSM = sys.modules['CroppedSequenceManagerUI']\r\n\r\nis_visible= CSM.cls.get_sub_segment_options_hud_visibility(widget_id)\r\n"
    widget_id=Get widget id() -> (widget id=0)
   ) -> (then, ReturnValue, is_visible:bool=~) {inputs [widget_id], outputs [is_visible]}
   > Branch(Condition=@n30.is_visible)
   > Set Options HUD Visible() {is=title K2Node_CallFunction}
   n31 = Execute Python Script!(
    execute=@n29.then_1
    PythonScript="import sys\r\nCSM = sys.modules['CroppedSequenceManagerUI']\r\n\r\ntrack_actor_class= CSM.cls.get_track_actor_class_by_id(widget_id)\r\n"
    widget_id=Get widget id() -> (widget id=0)
   ) -> (then, ReturnValue, track_actor_class:Actor Class Reference) {inputs [widget_id], outputs [track_actor_class]}
   > Set Track Actor Class(Track Actor Class:Actor Class Reference=@n31.track_actor_class) -> (Output_Get:Actor Class Reference)
   n32 = Get Current Project() -> (Current Project:BP Animation Browser Project Object Reference)
   > n33 = Branch(Condition=Is Not Empty!(TargetArray:Array of S AB Actor To Face Curves Structures=Get Face Curves by Class!(self:BP Animation Browser Project Object Reference=@n32) -> (Face Curves by Class:Array of S AB Actor To Face Curves Structures)) -> (ReturnValue:bool))
   n34 = Find Track Actor In Project Settings!() -> (Found:bool, Facial Curves:Array of Strings)
   > n35 = Branch(Condition=@n34)
   > Set Actor Facial Curves(Actor Facial Curves:Array of Strings=@n34.Facial Curves) -> (Output_Get:Array of Strings)
   > Set Facial Curves()
   > n36 = Branch(Condition=Is Empty!(TargetArray=Get Actor Facial Curves()) -> (ReturnValue:bool))
   n37 = Get CheckBoxSkipFaceCurves() -> (CheckBoxSkipFaceCurves:Editor Utility Check Box Object Reference)
   > SetVisibility(self=@n37, InVisibility=Collapsed)
   n38 = Get ButtonFaceCurvesText() -> (ButtonFaceCurvesText:Text Object Reference)
   > SetColorAndOpacity!(
    execute
    self=@n38
    InColorAndOpacity_SpecifiedColor:Linear Color Structure=(R=1,G=0,B=0,A=1)
    InColorAndOpacity_ColorUseRule:ESlateColorStylingMode Enum
   ) -> (then)
   n39 = Get Current Project()
   n40 = Get Track Actor Class() -> (Track Actor Class:Actor Class Reference)
   > Set Face Curves Msg(Face Curves Msg:Text=Format Text!(
     Format="LOCTEXT(\"No facial curves listed for actor class: \\\"{actor_class}\\\" in project settings:\\r\\n\\r\\nProject Name: {project_name}\\r\\nPath: {path}\")"
     actor_class:String=Get Class Display Name(Class=@n40)
     project_name:String=Get Name!(self:BP Animation Browser Project Object Reference=@n39) -> (Name:String)
     path:String=Get Object Path String(Object=@n39)
    ) -> (Result)) -> (Output_Get:Text)
   SetVisibility(execute=@n36.else, self=@n37, InVisibility=Visible)
   > SetColorAndOpacity(self=@n38, InColorAndOpacity_SpecifiedColor=(R=0,G=1,B=0,A=1))
   n41 = Get Current Project()
   n42 = Get Track Actor Class()
   > Set Face Curves Msg(Face Curves Msg=Format Text!(
     Format="LOCTEXT(\"Anim Curves listed for actor \\\"{actor_class}\\\" in \\\"\\\"Face Curves by Class\\\"\\\".\\r\\n\\r\\nIn project {project_name}\\r\\nPath: {path}\\r\\n\\r\\nCurves to be removed: \\r\\n{curves}\")"
     actor_class=Get Class Display Name(Class=@n42)
     project_name=Get Name(self=@n41)
     path=Get Object Path String(Object=@n41)
     curves:String=CurvesAsStr!() -> (str:String) {is=title K2Node_CallFunction}
    ) -> (Result))
   n43 = Get Current Project()
   n44 = Get Track Actor Class()
   n45 = Set Face Curves Msg(
    execute=@n35.else
    Face Curves Msg=Format Text!(
     Format="LOCTEXT(\"This track\\'s actor class \\r\\n\\\"{actor_class}\\\" \\r\\nis not listed in \\\"Face Curves by Class\\\" variable in\\r\\n\\r\\nProject Name: \\\"{project_name}\\\"\\r\\nAt path: {path}\")"
     actor_class=Get Class Display Name(Class=@n44)
     project_name=Get Name(self=@n43)
     path=Get Object Path String(Object=@n43)
    ) -> (Result)
   )
   SetVisibility(execute=[@n47.then, @n45.then], self=Get CheckBoxSkipFaceCurves(), InVisibility=Collapsed)
   n46 = Get Current Project()
   n47 = Set Face Curves Msg(
    execute=@n33.else
    Face Curves Msg=Format Text!(
     Format="LOCTEXT(\"Nothing specified in the \\\"Face Curves by Class\\\" variable \\r\\nin the project setting file: \\r\\n\\r\\nProject Name: \\\"{project_name}\\\"\\r\\n\\r\\nAt path: {path}\")"
     project_name=Get Name(self=@n46)
     path=Get Object Path String(Object=@n46)
    ) -> (Result)
   )
   Event PreConstruct()
   Event Tick()
  ]}
 Find Track Actor In Project Settings {nodes [
   Find Track Actor In Project Settings() {is=title K2Node_FunctionEntry}
   > n1 = For Each Loop!(Exec, Array:Array of S AB Actor To Face Curves Structures=Get Face Curves by Class(self=Get Current Project())) -> (
    LoopBody
    Array Element_TrackActorClass_3_#1:Actor Class Reference
    Array Element_FaceCurves_7_#2:Array of Strings
    Array Element_FaceCurvesDA_10_#3:PDA Animation Browser Character Curves List Object Reference
    Array Index
    Completed
   )
   Return Node!(execute=@n1.Completed, Found:bool=false, Facial Curves:Array of Strings)
   > Branch(Condition="Equal (Class)"(A=@n1, B=Get Track Actor Class()))
   > n2 = Is Valid(InputObject=@n1.Array Element_FaceCurvesDA_10_#3)
   Return Node(execute=@n2.Is Not Valid, Found=true, Facial Curves=@n1.Array Element_FaceCurves_7_#2)
   > Return Node(
    Found=true
    Facial Curves=Get Face Curves!(self:PDA Animation Browser Character Curves List Object Reference=@n1.Array Element_FaceCurvesDA_10_#3) -> (Face Curves:Array of Strings)
   )
  ]}
 Hide All Clicked {nodes [Hide All Clicked() {is=title K2Node_FunctionEntry}]}
 InitWidgetValues {nodes [
   InitWidgetValues() {is=title K2Node_FunctionEntry}
   > Set supress_update(supress_update:bool=true) -> (Output_Get:bool)
   > n1 = Sequence!(execute) -> (then_0, then_1, then_2:Exec, then_3:Exec, then_4:Exec, then_5:Exec, then_6:Exec, then_7:Exec, then_8:Exec, then_9:Exec, then_10:Exec)
   > n2 = Execute Python Script!(
    execute
    PythonScript="from utilities import animation_sequence_utils\r\n\r\ntype, pose_type = animation_sequence_utils.get_additive_options(anim_sequence)"
    anim_sequence:Anim Sequence Object Reference=Get AnimSequence()
   ) -> (then, ReturnValue, type:String, pose_type:String) {inputs [anim_sequence], outputs [type,pose_type]}
   > SetSelectedOption!(execute, self:"Combo Box (String) Object Reference"=Get ComboBox_IsAdditive(), Option:String=@n2.type) -> (then)
   > n3 = Branch(Condition="Equal Exactly (String)"(A=@n2.type, B=No Additive))
   n4 = Get ComboBox_AdditiveType() -> (ComboBox_AdditiveType:Editor Utility Combo Box String Object Reference)
   > SetVisibility(self=@n4, InVisibility=Hidden)
   SetVisibility(execute=@n3.else, self=@n4, InVisibility=Visible)
   > n5 = Branch(Condition="Equal Exactly (String)"(A=@n2.pose_type, B=Custom))
   n6 = Get ComboBox_AdditiveType()
   > n7 = AddOption!(execute, self:"Combo Box (String) Object Reference"=@n6, Option:String=@n2.pose_type) -> (then)
   SetSelectedOption(execute=[@n8.then, @n7.then], self=@n6, Option=@n2.pose_type)
   n8 = RemoveOption!(execute=@n5.else, self:"Combo Box (String) Object Reference"=@n6, Option:String=Custom) -> (then, ReturnValue:bool)
   n9 = Execute Python Script!(
    execute=@n1.then_1
    PythonScript="from utilities import animation_sequence_utils\r\n\r\nenabled = animation_sequence_utils.get_root_motion(anim_sequence)"
    anim_sequence=Get AnimSequence()
   ) -> (then, ReturnValue, enabled:bool=~) {inputs [anim_sequence], outputs [enabled]}
   > n10 = Branch(Condition=@n9.enabled)
   n11 = Get CheckBoxRootMotionEnable() -> (CheckBoxRootMotionEnable:Editor Utility Check Box Object Reference)
   > SetCheckedState!(execute, self:Check Box Object Reference=@n11, InCheckedState:ECheckBoxState Enum=Checked) -> (then)
   SetCheckedState(execute=@n10.else, self=@n11, InCheckedState=Unchecked)
   Set supress_update(execute=@n1.then_10, supress_update=false)
   n12 = Execute Python Script!(
    execute=@n1.then_2
    PythonScript="from utilities import animation_sequence_utils\r\n\r\nis_locked, mode = animation_sequence_utils.get_root_lock(anim_sequence)"
    anim_sequence=Get AnimSequence()
   ) -> (then, ReturnValue, is_locked:bool=~, mode:String) {inputs [anim_sequence], outputs [is_locked,mode]}
   > SetIsChecked!(
    execute
    self:Check Box Object Reference=Get CheckBoxRootMotionLock() -> (CheckBoxRootMotionLock:Editor Utility Check Box Object Reference)
    InIsChecked:bool=@n12.is_locked
   ) -> (then)
   > n13 = Branch(Condition=@n12.is_locked)
   > SetSelectedOption(self=Get ComboBox_RootMotionMode(), Option=@n12.mode)
   n14 = Get ComboBox_RootMotionMode()
   > SetVisibility(self=@n14, InVisibility=Visible)
   SetVisibility(execute=@n13.else, self=@n14, InVisibility=Hidden)
   n15 = Sequence!(execute=@n1.then_3) -> (then_0, then_1, then_2, then_3, then_4, then_5)
   > n16 = Execute Python Script!(
    execute
    PythonScript="from utilities import animation_sequence_utils\r\n\r\nenabled = animation_sequence_utils.get_looping(anim_sequence)"
    anim_sequence=Get AnimSequence()
   ) -> (then, ReturnValue, enabled:bool=~) {inputs [anim_sequence], outputs [enabled]}
   > SetIsChecked(self=Get CheckBoxLooping() -> (CheckBoxLooping:Editor Utility Check Box Object Reference), InIsChecked=@n16.enabled)
   n17 = Execute Python Script!(
    execute=@n15.then_1
    PythonScript="from utilities import animation_sequence_utils\r\n\r\nfps, is_custom_fps = animation_sequence_utils.get_fps(anim_sequence, True)"
    anim_sequence=Get AnimSequence()
   ) -> (then, ReturnValue, fps:int, is_custom_fps:bool=~) {inputs [anim_sequence], outputs [fps, is_custom_fps]}
   n18 = "To String (Integer)"(InInt=@n17.fps)
   n19 = Get ComboBox_FPS()
   > n20 = Branch(Condition="Equal (Integer)"(A=FindOptionIndex!(self:"Combo Box (String) Object Reference"=@n19, Option:String=@n18) -> (ReturnValue:int=0), B=-1))
   > n21 = AddOption(self=@n19, Option=@n18)
   SetSelectedOption(execute=[@n21.then, @n20.else], self=[@n19, @n19], Option=@n18)
   > SetIsChecked(self=Get CheckBoxCustomFPS() -> (CheckBoxCustomFPS:Editor Utility Check Box Object Reference), InIsChecked=@n17.is_custom_fps)
   > SetIsEnabled(self=@n19, bInIsEnabled=@n17.is_custom_fps)
   n22 = Execute Python Script!(
    execute=@n15.then_2
    PythonScript="from utilities import animation_sequence_utils\r\n\r\ninterpolation = animation_sequence_utils.get_interpolation(anim_sequence, True)"
    anim_sequence=Get AnimSequence()
   ) -> (then, ReturnValue, interpolation:String) {inputs [anim_sequence], outputs [interpolation]}
   > SetSelectedOption(self=Get ComboBox_Interpolation(), Option=@n22.interpolation)
   n23 = Execute Python Script!(
    execute=@n15.then_3
    PythonScript="from utilities import animation_sequence_utils\r\n\r\nskip_zero_curves = animation_sequence_utils.get_is_exported_with_skip_zero_curves(anim_sequence, level_sequence)"
    anim_sequence=Get AnimSequence()
    level_sequence:Level Sequence Object Reference=Get Level Seqeunce() -> (Level Seqeunce:Level Sequence Object Reference)
   ) -> (then, ReturnValue, skip_zero_curves:bool=~) {inputs [anim_sequence, level_sequence], outputs [skip_zero_curves]}
   > SetIsChecked(self=Get CheckBoxSkipZeroCurves(), InIsChecked=@n23.skip_zero_curves)
   n24 = Execute Python Script!(
    execute=@n15.then_4
    PythonScript="from utilities import animation_sequence_utils\r\n\r\nskip_face_curves = animation_sequence_utils.get_is_exported_with_skip_face_curves(anim_sequence, level_sequence, face_curves)"
    anim_sequence=Get AnimSequence()
    level_sequence=Get Level Seqeunce()
    face_curves:Array of Strings=Get Actor Facial Curves()
   ) -> (then, ReturnValue, skip_face_curves:bool=~) {inputs [anim_sequence, level_sequence, face_curves], outputs [skip_face_curves]}
   > SetIsChecked(self=Get CheckBoxSkipFaceCurves(), InIsChecked=@n24.skip_face_curves)
   > Set Facial Curves()
   n25 = Execute Python Script!(
    execute=@n15.then_5
    PythonScript="from utilities import animation_sequence_utils\r\n\r\nresult = animation_sequence_utils.get_is_retarget_source_valid(anim_sequence)"
    anim_sequence=Get AnimSequence()
   ) -> (then, ReturnValue, result:bool=~) {inputs [anim_sequence], outputs [result]}
   > SetIsChecked(self=Get CheckBoxSetRetargetSourceAsset(), InIsChecked=@n25.result)
  ]}
 OnKeyDown {nodes [
   n1 = OnKeyDown!() -> (then, MyGeometry:Geometry Structure, InKeyEvent:Key Event Structure) {is=title K2Node_FunctionEntry}
   n2 = GetKey!(Input:"Key Event Structure (by ref)"=@n1.InKeyEvent) -> (ReturnValue:Key Structure)
   > Set Is Ctrl Pressed(Is Ctrl Pressed:bool=OR Boolean(
     A="Equal (Key)"(A:Key Structure=@n2, B:Key Structure=LeftControl) -> (ReturnValue:bool=~) {is=title K2Node_PromotableOperator}
     B="Equal (Key)"(A=@n2, B=RightControl) -> (ReturnValue=~)
    )) -> (Output_Get:bool)
   > Return Node!(execute, ReturnValue:Event Reply Structure=Handled())
  ]}
 OnKeyUp {nodes [
   n1 = OnKeyUp!() -> (then, MyGeometry:Geometry Structure, InKeyEvent:Key Event Structure) {is=title K2Node_FunctionEntry}
   n2 = GetKey(Input=@n1.InKeyEvent)
   n3 = OR Boolean(A="Equal (Key)"(A=@n2, B=LeftControl) -> (ReturnValue=~), B="Equal (Key)"(A=@n2, B=RightControl) -> (ReturnValue=~))
   > n4 = Branch(Condition=@n3)
   > n5 = Set Is Ctrl Pressed(Is Ctrl Pressed=false)
   Return Node(execute=[@n4.else, @n5.then], ReturnValue=Handled())
   NOT Boolean(A=@n3)
  ]}
 Set Facial Curves {nodes [
   Set Facial Curves() {is=title K2Node_FunctionEntry}
   > n1 = Branch(Condition=IsChecked(self=Get CheckBoxSkipFaceCurves()))
   > n2 = Set Apply Facial Curves(Apply Facial Curves:Array of Strings=Get Actor Facial Curves()) -> (Output_Get:Array of Strings)
   Return Node!(execute=[@n3.then, @n2.then])
   n3 = Clear(execute=@n1.else, TargetArray=Get Apply Facial Curves())
  ]}
]
variables [
 Actor Facial Curves {type="array<string>", default []}
 Animatable Creature {type=AnimatableCreature_C, default=null}
 AnimSequence {type=AnimSequence, default=null}
 Apply Facial Curves {type="array<string>", default []}
 Current Project {type=BP_AnimationBrowser_Project_C, default=null}
 Face Curves Msg {type=string, default=""}
 Hide All Clicked {default=()}
 Is Ctrl Pressed {type=boolean}
 Level Seqeunce {type=LevelSequence, default=null}
 ShowHUD {type=boolean}
 supress_update {type=boolean}
 Track Actor Class {type=Class, default=null}
 widget id {type=integer}
]
widgets {
 info {rootWidgetClass=&HorizontalBox, widgetCount=56}
 widgets [
  HorizontalBox_0 {widgetClassPath=&HorizontalBox, widget=HorizontalBox_0, properties {bOverride_Cursor=true}}
  VerticalBox_2 {
   widgetClassPath=&VerticalBox
   parent=HorizontalBox_0
   widget=VerticalBox_2
   slot {class=HorizontalBoxSlot, properties {horizontalAlignment=HAlign_Left, padding=<ltrb 0 0 5 0>}}
  }
  VisibilityButton {
   widgetClassPath=&Button
   parent=VerticalBox_2
   widget=VisibilityButton
   properties {
    toolTipText="Show\\Hide Options Bar\r\n\r\n"
    widgetStyle {
     clickedSlateSound {resourceObject=null}
     disabled {
      bIsDynamicallyLoaded=false
      drawAs=RoundedBox
      imageType=NoImage
      mirroring=NoMirror
      resourceName=None
      resourceObject=null
      tiling=NoTile
      imageSize=<0 0>
      margin=<ltrb 0 0 0 0>
      outlineSettings {
       bUseBrushTransparency=false
       roundingType=FixedRadius
       width=1
       color {colorUseRule=UseColor_ColorTable, specifiedColor=<rgba 0 0 0 0>}
       cornerRadii=<4 4 4 4>
      }
      tintColor {colorUseRule=UseColor_Specified, specifiedColor=<rgba 0.0395 0.0395 0.0395 1>}
      uVRegion {bIsValid=false, max=<0 0>, min=<0 0>}
     }
     disabledForeground {colorUseRule=UseColor_ColorTable, specifiedColor=<rgba 0 0 0 0>}
     hovered {
      bIsDynamicallyLoaded=false
      drawAs=RoundedBox
      imageType=NoImage
      mirroring=NoMirror
      resourceName=None
      resourceObject=null
      tiling=NoTile
      imageSize=<32 32>
      margin=<ltrb 0 0 0 0>
      outlineSettings {
       bUseBrushTransparency=false
       roundingType=FixedRadius
       width=0
       color {colorUseRule=UseColor_ColorTable, specifiedColor=<rgba 0 0 0 0>}
       cornerRadii=<4 4 4 4>
      }
      tintColor {colorUseRule=UseColor_Specified, specifiedColor=<rgba 0.0953 0.0953 0.0953 1>}
      uVRegion {bIsValid=false, max=<0 0>, min=<0 0>}
     }
     hoveredForeground {colorUseRule=UseColor_ColorTable, specifiedColor=<rgba 0 0 0 0>}
     hoveredSlateSound {resourceObject=null}
     normal {
      bIsDynamicallyLoaded=false
      drawAs=RoundedBox
      imageType=FullColor
      mirroring=NoMirror
      resourceName=None
      resourceObject=null
      tiling=NoTile
      imageSize=<14 14>
      margin=<ltrb 0 0 0 0>
      outlineSettings {
       bUseBrushTransparency=false
       roundingType=FixedRadius
       width=1
       color {colorUseRule=UseColor_ColorTable, specifiedColor=<rgba 0 0 0 0>}
       cornerRadii=<4 4 4 4>
      }
      tintColor {colorUseRule=UseColor_Specified, specifiedColor=<rgba 0.0395 0.0395 0.0395 1>}
      uVRegion {bIsValid=false, max=<0 0>, min=<0 0>}
     }
     normalForeground {colorUseRule=UseColor_ColorTable, specifiedColor=<rgba 0 0 0 0>}
     normalPadding=<ltrb 12 1.5 12 1.5>
     pressed {
      bIsDynamicallyLoaded=false
      drawAs=RoundedBox
      imageType=NoImage
      mirroring=NoMirror
      resourceName=None
      resourceObject=null
      tiling=NoTile
      imageSize=<32 32>
      margin=<ltrb 0 0 0 0>
      outlineSettings {
       bUseBrushTransparency=false
       roundingType=FixedRadius
       width=0
       color {colorUseRule=UseColor_ColorTable, specifiedColor=<rgba 0 0 0 0>}
       cornerRadii=<4 4 4 4>
      }
      tintColor {colorUseRule=UseColor_Specified, specifiedColor=<rgba 0.0284 0.0284 0.0284 1>}
      uVRegion {bIsValid=false, max=<0 0>, min=<0 0>}
     }
     pressedForeground {colorUseRule=UseColor_ColorTable, specifiedColor=<rgba 0 0 0 0>}
     pressedPadding=<ltrb 12 2.5 12 0.5>
     pressedSlateSound {resourceObject=null}
    }
   }
   slot {class=VerticalBoxSlot, properties {horizontalAlignment=HAlign_Center, verticalAlignment=VAlign_Center, size {sizeRule=Fill, value=1}}}
  }
  HorizontalBox {
   widgetClassPath=&HorizontalBox
   parent=VisibilityButton
   widget=HorizontalBox
   properties {bIsEnabled=false}
   slot {class=ButtonSlot, properties {padding=<ltrb 0 0 0 0>}}
  }
  AssetImageInivisibleIcon {
   widgetClassPath=&Image
   bIsVariable=true
   parent=HorizontalBox
   widget=AssetImageInivisibleIcon
   properties {brush {
     bIsDynamicallyLoaded=false
     drawAs=Image
     imageType=FullColor
     mirroring=NoMirror
     resourceName=None
     resourceObject=&/icon_layer_not_visible
     tiling=NoTile
     imageSize=<16 16>
     margin=<ltrb 0 0 0 0>
     outlineSettings {
      bUseBrushTransparency=false
      roundingType=HalfHeightRadius
      width=0
      color {colorUseRule=UseColor_Specified, specifiedColor=<rgba 0 0 0 0>}
      cornerRadii=<0 0 0 0>
     }
     tintColor {colorUseRule=UseColor_Specified, specifiedColor=<rgba 0.2108 1 0 1>}
     uVRegion {bIsValid=false, max=<0 0>, min=<0 0>}
    }}
   slot {class=HorizontalBoxSlot, properties {horizontalAlignment=HAlign_Center, verticalAlignment=VAlign_Center}}
  }
  AssetImageVisibleicon {
   widgetClassPath=&Image
   bIsVariable=true
   parent=HorizontalBox
   widget=AssetImageVisibleicon
   properties {
    visibility=Collapsed
    brush {
     bIsDynamicallyLoaded=false
     drawAs=Image
     imageType=FullColor
     mirroring=NoMirror
     resourceName=None
     resourceObject=&/icon_layer_visible_16x
     tiling=NoTile
     imageSize=<16 16>
     margin=<ltrb 0 0 0 0>
     outlineSettings {
      bUseBrushTransparency=false
      roundingType=HalfHeightRadius
      width=0
      color {colorUseRule=UseColor_Specified, specifiedColor=<rgba 0 0 0 0>}
      cornerRadii=<0 0 0 0>
     }
     tintColor {colorUseRule=UseColor_Specified, specifiedColor=<rgba 0.2108 1 0 1>}
     uVRegion {bIsValid=false, max=<0 0>, min=<0 0>}
    }
   }
   slot {class=HorizontalBoxSlot, properties {horizontalAlignment=HAlign_Center, verticalAlignment=VAlign_Center}}
  }
  VerticalBoxOPtions {
   widgetClassPath=&VerticalBox
   bIsVariable=true
   parent=HorizontalBox_0
   widget=VerticalBoxOPtions
   properties {visibility=Collapsed}
   slot {class=HorizontalBoxSlot}
  }
  GridPanel_40 {
   widgetClassPath=&GridPanel
   parent=VerticalBoxOPtions
   widget=GridPanel_40
   slot {class=VerticalBoxSlot, properties {horizontalAlignment=HAlign_Left, verticalAlignment=VAlign_Center}}
  }
  TextAdditive {
   widgetClassPath=&TextBlock
   bIsVariable=true
   parent=GridPanel_40
   widget=TextAdditive
   properties {
    text=Additive:
    toolTipText="Quick access to set most common additive animation type. For more advanced options - open the AS(double click on thumbnail)"
    colorAndOpacity {colorUseRule=UseColor_Specified, specifiedColor=<rgba 0.1545 0.1545 0.1545 1>}
    font {
     bForceMonospaced=false
     bMaterialIsStencil=false
     fontMaterial=null
     fontObject=&Roboto
     letterSpacing=0
     monospacedWidth=1
     size=7.5
     skewAmount=0
     typefaceFontName=Bold
     outlineSettings {
      bApplyOutlineToDropShadows=false
      bMiteredCorners=false
      bSeparateFillAlpha=false
      outlineMaterial=null
      outlineSize=0
      outlineColor=<rgba 0 0 0 1>
     }
    }
   }
   slot {class=GridSlot, properties {horizontalAlignment=HAlign_Left, verticalAlignment=VAlign_Center}}
  }
  Border_0 {
   widgetClassPath=&Border
   parent=GridPanel_40
   widget=Border_0
   properties {
    verticalAlignment=VAlign_Center
    background {
     bIsDynamicallyLoaded=false
     drawAs=Border
     imageType=NoImage
     mirroring=NoMirror
     resourceName=None
     resourceObject=null
     tiling=NoTile
     imageSize=<32 32>
     margin=<ltrb 0 1 0 1>
     outlineSettings {
      bUseBrushTransparency=false
      roundingType=HalfHeightRadius
      width=0
      color {colorUseRule=UseColor_Specified, specifiedColor=<rgba 0 0 0 0>}
      cornerRadii=<0 0 0 0>
     }
     tintColor {colorUseRule=UseColor_Specified, specifiedColor=<rgba 1 1 1 1>}
     uVRegion {bIsValid=false, max=<0 0>, min=<0 0>}
    }
    brushColor=<rgba 0.0799 0.0799 0.0799 1>
    padding=<ltrb 0 0 0 0>
   }
   slot {class=GridSlot, properties {row=2}}
  }
  TextRootMotionLock {
   widgetClassPath=&TextBlock
   bIsVariable=true
   parent=Border_0
   widget=TextRootMotionLock
   properties {
    text=Force Root Lock:
    toolTipText=Locks Root Movement
    colorAndOpacity {colorUseRule=UseColor_Specified, specifiedColor=<rgba 0.1545 0.1545 0.1545 1>}
    font {
     bForceMonospaced=false
     bMaterialIsStencil=false
     fontMaterial=null
     fontObject=&Roboto
     letterSpacing=0
     monospacedWidth=1
     size=7.5
     skewAmount=0
     typefaceFontName=Bold
     outlineSettings {
      bApplyOutlineToDropShadows=false
      bMiteredCorners=false
      bSeparateFillAlpha=false
      outlineMaterial=null
      outlineSize=0
      outlineColor=<rgba 0 0 0 1>
     }
    }
   }
   slot {class=BorderSlot, properties {verticalAlignment=VAlign_Center, padding=<ltrb 0 0 0 0>}}
  }
  ComboBox_IsAdditive {
   widgetClassPath=/Script/Blutility.EditorUtilityComboBoxString
   bIsVariable=true
   parent=GridPanel_40
   widget=ComboBox_IsAdditive
   properties {
    selectedOption=No Additive
    contentPadding=<ltrb 0 0 0 0>
    defaultOptions [No Additive, Local Space, Mesh Space]
    font {
     bForceMonospaced=false
     bMaterialIsStencil=false
     fontMaterial=null
     fontObject=&Roboto
     letterSpacing=0
     monospacedWidth=1
     size=7.5
     skewAmount=0
     typefaceFontName=Bold
     outlineSettings {
      bApplyOutlineToDropShadows=false
      bMiteredCorners=false
      bSeparateFillAlpha=false
      outlineMaterial=null
      outlineSize=0
      outlineColor=<rgba 0 0 0 1>
     }
    }
    foregroundColor {colorUseRule=UseColor_Specified, specifiedColor=<rgba 0.5271 0.5271 0.5271 1>}
   }
   slot {class=GridSlot, properties {column=1, horizontalAlignment=HAlign_Center, verticalAlignment=VAlign_Center, padding=<ltrb 0 2 0 2>}}
  }
  ComboBox_AdditiveType {
   widgetClassPath=&EditorUtilityComboBoxString
   bIsVariable=true
   parent=GridPanel_40
   widget=ComboBox_AdditiveType
   properties {
    selectedOption=First Frame
    toolTipText="This provides shortcut to most common setups for additives: \r\n\r\n- relatively to first frame of this animation\r\n- relatively to skeletal basepose\r\n\r\nIf tool detects any other setup - it will set this field to \"Custom\"\r\n\r\nAny other setup can be made directly in AnimSeqeunce (double click on thumbnail to open Anim Sequence)\r\n\r\n"
    contentPadding=<ltrb 0 0 0 0>
    defaultOptions [First Frame, SkelRef Pose]
    font {
     bForceMonospaced=false
     bMaterialIsStencil=false
     fontMaterial=null
     fontObject=&Roboto
     letterSpacing=0
     monospacedWidth=1
     size=7.5
     skewAmount=0
     typefaceFontName=Bold
     outlineSettings {
      bApplyOutlineToDropShadows=false
      bMiteredCorners=false
      bSeparateFillAlpha=false
      outlineMaterial=null
      outlineSize=0
      outlineColor=<rgba 0 0 0 1>
     }
    }
    foregroundColor {colorUseRule=UseColor_Specified, specifiedColor=<rgba 0.5271 0.5271 0.5271 1>}
   }
   slot {class=GridSlot, properties {column=2, horizontalAlignment=HAlign_Center, verticalAlignment=VAlign_Center, padding=<ltrb 0 2 0 2>}}
  }
  Border {
   widgetClassPath=&Border
   parent=GridPanel_40
   widget=Border
   properties {
    horizontalAlignment=HAlign_Center
    verticalAlignment=VAlign_Center
    background {
     bIsDynamicallyLoaded=false
     drawAs=Border
     imageType=NoImage
     mirroring=NoMirror
     resourceName=None
     resourceObject=null
     tiling=NoTile
     imageSize=<32 32>
     margin=<ltrb 0 1 0 1>
     outlineSettings {
      bUseBrushTransparency=false
      roundingType=HalfHeightRadius
      width=0
      color {colorUseRule=UseColor_Specified, specifiedColor=<rgba 0 0 0 0>}
      cornerRadii=<0 0 0 0>
     }
     tintColor {colorUseRule=UseColor_Specified, specifiedColor=<rgba 1 1 1 1>}
     uVRegion {bIsValid=false, max=<0 0>, min=<0 0>}
    }
    brushColor=<rgba 0.0799 0.0799 0.0799 1>
    padding=<ltrb 0 2 0 2>
   }
   slot {class=GridSlot, properties {column=1, row=2}}
  }
  CheckBoxRootMotionLock {
   widgetClassPath=&EditorUtilityCheckBox
   bIsVariable=true
   parent=Border
   widget=CheckBoxRootMotionLock
   properties {horizontalAlignment=HAlign_Center}
   slot {class=BorderSlot, properties {horizontalAlignment=HAlign_Center, verticalAlignment=VAlign_Center, padding=<ltrb 0 2 0 2>}}
  }
  Border_4 {
   widgetClassPath=&Border
   parent=GridPanel_40
   widget=Border_4
   properties {
    background {
     bIsDynamicallyLoaded=false
     drawAs=Border
     imageType=NoImage
     mirroring=NoMirror
     resourceName=None
     resourceObject=null
     tiling=NoTile
     imageSize=<32 32>
     margin=<ltrb 0 1 0 1>
     outlineSettings {
      bUseBrushTransparency=false
      roundingType=HalfHeightRadius
      width=0
      color {colorUseRule=UseColor_Specified, specifiedColor=<rgba 0 0 0 0>}
      cornerRadii=<0 0 0 0>
     }
     tintColor {colorUseRule=UseColor_Specified, specifiedColor=<rgba 1 1 1 1>}
     uVRegion {bIsValid=false, max=<0 0>, min=<0 0>}
    }
    brushColor=<rgba 0.0799 0.0799 0.0799 1>
    padding=<ltrb 0 2 0 0>
   }
   slot {class=GridSlot, properties {column=2, horizontalAlignment=HAlign_Center, row=2, verticalAlignment=VAlign_Center}}
  }
  ComboBox_RootMotionMode {
   widgetClassPath=&EditorUtilityComboBoxString
   bIsVariable=true
   parent=Border_4
   widget=ComboBox_RootMotionMode
   properties {
    selectedOption=First Frame
    contentPadding=<ltrb 0 0 0 0>
    defaultOptions [Ref Pose, First Frame, Zero]
    font {
     bForceMonospaced=false
     bMaterialIsStencil=false
     fontMaterial=null
     fontObject=&Roboto
     letterSpacing=0
     monospacedWidth=1
     size=7.5
     skewAmount=0
     typefaceFontName=Bold
     outlineSettings {
      bApplyOutlineToDropShadows=false
      bMiteredCorners=false
      bSeparateFillAlpha=false
      outlineMaterial=null
      outlineSize=0
      outlineColor=<rgba 0 0 0 1>
     }
    }
    foregroundColor {colorUseRule=UseColor_Specified, specifiedColor=<rgba 0.5271 0.5271 0.5271 1>}
   }
   slot {class=BorderSlot, properties {padding=<ltrb 0 2 0 0>}}
  }
  Border_3 {
   widgetClassPath=&Border
   parent=GridPanel_40
   widget=Border_3
   properties {
    horizontalAlignment=HAlign_Center
    verticalAlignment=VAlign_Center
    background {
     bIsDynamicallyLoaded=false
     drawAs=Border
     imageType=NoImage
     mirroring=NoMirror
     resourceName=None
     resourceObject=null
     tiling=NoTile
     imageSize=<32 32>
     margin=<ltrb 0 0 0 1>
     outlineSettings {
      bUseBrushTransparency=false
      roundingType=HalfHeightRadius
      width=0
      color {colorUseRule=UseColor_Specified, specifiedColor=<rgba 0 0 0 0>}
      cornerRadii=<0 0 0 0>
     }
     tintColor {colorUseRule=UseColor_Specified, specifiedColor=<rgba 1 1 1 1>}
     uVRegion {bIsValid=false, max=<0 0>, min=<0 0>}
    }
    brushColor=<rgba 0.0799 0.0799 0.0799 1>
    padding=<ltrb 0 2 0 2>
   }
   slot {class=GridSlot, properties {column=1, row=3}}
  }
  CheckBoxRootMotionEnable {
   widgetClassPath=&EditorUtilityCheckBox
   bIsVariable=true
   parent=Border_3
   widget=CheckBoxRootMotionEnable
   properties {horizontalAlignment=HAlign_Center}
   slot {class=BorderSlot, properties {horizontalAlignment=HAlign_Center, verticalAlignment=VAlign_Center, padding=<ltrb 0 2 0 2>}}
  }
  Border_1 {
   widgetClassPath=&Border
   parent=GridPanel_40
   widget=Border_1
   properties {
    horizontalAlignment=HAlign_Center
    verticalAlignment=VAlign_Center
    background {
     bIsDynamicallyLoaded=false
     drawAs=Border
     imageType=NoImage
     mirroring=NoMirror
     resourceName=None
     resourceObject=null
     tiling=NoTile
     imageSize=<32 32>
     margin=<ltrb 0 0 0 1>
     outlineSettings {
      bUseBrushTransparency=false
      roundingType=HalfHeightRadius
      width=0
      color {colorUseRule=UseColor_Specified, specifiedColor=<rgba 0 0 0 0>}
      cornerRadii=<0 0 0 0>
     }
     tintColor {colorUseRule=UseColor_Specified, specifiedColor=<rgba 1 1 1 1>}
     uVRegion {bIsValid=false, max=<0 0>, min=<0 0>}
    }
    brushColor=<rgba 0.0799 0.0799 0.0799 1>
    padding=<ltrb 0 0 0 0>
   }
   slot {class=GridSlot, properties {column=2, row=3}}
  }
  SizeHolder {
   widgetClassPath=&EditorUtilityCheckBox
   parent=Border_1
   widget=SizeHolder
   properties {horizontalAlignment=HAlign_Center, visibility=Hidden}
   slot {class=BorderSlot, properties {horizontalAlignment=HAlign_Center, verticalAlignment=VAlign_Center, padding=<ltrb 0 0 0 0>}}
  }
  Border_2 {
   widgetClassPath=&Border
   parent=GridPanel_40
   widget=Border_2
   properties {
    horizontalAlignment=HAlign_Left
    verticalAlignment=VAlign_Center
    background {
     bIsDynamicallyLoaded=false
     drawAs=Border
     imageType=NoImage
     mirroring=NoMirror
     resourceName=None
     resourceObject=null
     tiling=NoTile
     imageSize=<32 32>
     margin=<ltrb 0 0 0 1>
     outlineSettings {
      bUseBrushTransparency=false
      roundingType=HalfHeightRadius
      width=0
      color {colorUseRule=UseColor_Specified, specifiedColor=<rgba 0 0 0 0>}
      cornerRadii=<0 0 0 0>
     }
     tintColor {colorUseRule=UseColor_Specified, specifiedColor=<rgba 1 1 1 1>}
     uVRegion {bIsValid=false, max=<0 0>, min=<0 0>}
    }
    brushColor=<rgba 0.0799 0.0799 0.0799 1>
    padding=<ltrb 0 2 0 0>
   }
   slot {class=GridSlot, properties {row=3}}
  }
  TextAdditive_1 {
   widgetClassPath=&TextBlock
   bIsVariable=true
   parent=Border_2
   widget=TextAdditive_1
   properties {
    text=Root Motion:
    toolTipText=Extracts root motion
    colorAndOpacity {colorUseRule=UseColor_Specified, specifiedColor=<rgba 0.1545 0.1545 0.1545 1>}
    font {
     bForceMonospaced=false
     bMaterialIsStencil=false
     fontMaterial=null
     fontObject=&Roboto
     letterSpacing=0
     monospacedWidth=1
     size=7.5
     skewAmount=0
     typefaceFontName=Bold
     outlineSettings {
      bApplyOutlineToDropShadows=false
      bMiteredCorners=false
      bSeparateFillAlpha=false
      outlineMaterial=null
      outlineSize=0
      outlineColor=<rgba 0 0 0 1>
     }
    }
   }
   slot {class=BorderSlot, properties {horizontalAlignment=HAlign_Left, verticalAlignment=VAlign_Center, padding=<ltrb 0 2 0 0>}}
  }
  Border_5 {
   widgetClassPath=&Border
   parent=GridPanel_40
   widget=Border_5
   properties {
    horizontalAlignment=HAlign_Left
    verticalAlignment=VAlign_Center
    background {
     bIsDynamicallyLoaded=false
     drawAs=Border
     imageType=NoImage
     mirroring=NoMirror
     resourceName=None
     resourceObject=null
     tiling=NoTile
     imageSize=<32 32>
     margin=<ltrb 0 0 0 1>
     outlineSettings {
      bUseBrushTransparency=false
      roundingType=HalfHeightRadius
      width=0
      color {colorUseRule=UseColor_Specified, specifiedColor=<rgba 0 0 0 0>}
      cornerRadii=<0 0 0 0>
     }
     tintColor {colorUseRule=UseColor_Specified, specifiedColor=<rgba 1 1 1 1>}
     uVRegion {bIsValid=false, max=<0 0>, min=<0 0>}
    }
    brushColor=<rgba 0.0799 0.0799 0.0799 1>
    padding=<ltrb 0 2 0 0>
   }
   slot {class=GridSlot, properties {row=4}}
  }
  TextAdditive_2 {
   widgetClassPath=&TextBlock
   bIsVariable=true
   parent=Border_5
   widget=TextAdditive_2
   properties {
    text=Looping:
    toolTipText="Sets looping. \r\nNote. It really depends how this animation will be implemnted, many systems don't use\\ignore this flag"
    colorAndOpacity {colorUseRule=UseColor_Specified, specifiedColor=<rgba 0.1545 0.1545 0.1545 1>}
    font {
     bForceMonospaced=false
     bMaterialIsStencil=false
     fontMaterial=null
     fontObject=&Roboto
     letterSpacing=0
     monospacedWidth=1
     size=7.5
     skewAmount=0
     typefaceFontName=Bold
     outlineSettings {
      bApplyOutlineToDropShadows=false
      bMiteredCorners=false
      bSeparateFillAlpha=false
      outlineMaterial=null
      outlineSize=0
      outlineColor=<rgba 0 0 0 1>
     }
    }
   }
   slot {class=BorderSlot, properties {horizontalAlignment=HAlign_Left, verticalAlignment=VAlign_Center, padding=<ltrb 0 2 0 0>}}
  }
  Border_6 {
   widgetClassPath=&Border
   parent=GridPanel_40
   widget=Border_6
   properties {
    horizontalAlignment=HAlign_Center
    verticalAlignment=VAlign_Center
    background {
     bIsDynamicallyLoaded=false
     drawAs=Border
     imageType=NoImage
     mirroring=NoMirror
     resourceName=None
     resourceObject=null
     tiling=NoTile
     imageSize=<32 32>
     margin=<ltrb 0 0 0 1>
     outlineSettings {
      bUseBrushTransparency=false
      roundingType=HalfHeightRadius
      width=0
      color {colorUseRule=UseColor_Specified, specifiedColor=<rgba 0 0 0 0>}
      cornerRadii=<0 0 0 0>
     }
     tintColor {colorUseRule=UseColor_Specified, specifiedColor=<rgba 1 1 1 1>}
     uVRegion {bIsValid=false, max=<0 0>, min=<0 0>}
    }
    brushColor=<rgba 0.0799 0.0799 0.0799 1>
    padding=<ltrb 0 2 0 2>
   }
   slot {class=GridSlot, properties {column=1, row=4}}
  }
  CheckBoxLooping {
   widgetClassPath=&EditorUtilityCheckBox
   bIsVariable=true
   parent=Border_6
   widget=CheckBoxLooping
   properties {horizontalAlignment=HAlign_Center}
   slot {class=BorderSlot, properties {horizontalAlignment=HAlign_Center, verticalAlignment=VAlign_Center, padding=<ltrb 0 2 0 2>}}
  }
  Border_7 {
   widgetClassPath=&Border
   parent=GridPanel_40
   widget=Border_7
   properties {
    horizontalAlignment=HAlign_Center
    verticalAlignment=VAlign_Center
    background {
     bIsDynamicallyLoaded=false
     drawAs=Border
     imageType=NoImage
     mirroring=NoMirror
     resourceName=None
     resourceObject=null
     tiling=NoTile
     imageSize=<32 32>
     margin=<ltrb 0 0 0 1>
     outlineSettings {
      bUseBrushTransparency=false
      roundingType=HalfHeightRadius
      width=0
      color {colorUseRule=UseColor_Specified, specifiedColor=<rgba 0 0 0 0>}
      cornerRadii=<0 0 0 0>
     }
     tintColor {colorUseRule=UseColor_Specified, specifiedColor=<rgba 1 1 1 1>}
     uVRegion {bIsValid=false, max=<0 0>, min=<0 0>}
    }
    brushColor=<rgba 0.0799 0.0799 0.0799 1>
    padding=<ltrb 0 0 0 0>
   }
   slot {class=GridSlot, properties {column=2, row=4}}
  }
  SizeHolder_1 {
   widgetClassPath=&EditorUtilityCheckBox
   parent=Border_7
   widget=SizeHolder_1
   properties {horizontalAlignment=HAlign_Center, visibility=Hidden}
   slot {class=BorderSlot, properties {horizontalAlignment=HAlign_Center, verticalAlignment=VAlign_Center, padding=<ltrb 0 0 0 0>}}
  }
  Border_8 {
   widgetClassPath=&Border
   parent=GridPanel_40
   widget=Border_8
   properties {
    horizontalAlignment=HAlign_Left
    verticalAlignment=VAlign_Center
    background {
     bIsDynamicallyLoaded=false
     drawAs=Border
     imageType=NoImage
     mirroring=NoMirror
     resourceName=None
     resourceObject=null
     tiling=NoTile
     imageSize=<32 32>
     margin=<ltrb 0 0 0 1>
     outlineSettings {
      bUseBrushTransparency=false
      roundingType=HalfHeightRadius
      width=0
      color {colorUseRule=UseColor_Specified, specifiedColor=<rgba 0 0 0 0>}
      cornerRadii=<0 0 0 0>
     }
     tintColor {colorUseRule=UseColor_Specified, specifiedColor=<rgba 1 1 1 1>}
     uVRegion {bIsValid=false, max=<0 0>, min=<0 0>}
    }
    brushColor=<rgba 0.0799 0.0799 0.0799 1>
    padding=<ltrb 0 2 0 0>
   }
   slot {class=GridSlot, properties {row=5}}
  }
  TextAdditive_3 {
   widgetClassPath=&TextBlock
   bIsVariable=true
   parent=Border_8
   widget=TextAdditive_3
   properties {
    text=Custom FPS:
    toolTipText="By default animations are exported at FPS at this LevelSequence DisplayRate\r\n\r\nYou can export at custom framerate.\r\n\r\nBe aware - if you export higher framerate that your LS - you will see subframes that might be broken in some ways."
    colorAndOpacity {colorUseRule=UseColor_Specified, specifiedColor=<rgba 0.1545 0.1545 0.1545 1>}
    font {
     bForceMonospaced=false
     bMaterialIsStencil=false
     fontMaterial=null
     fontObject=&Roboto
     letterSpacing=0
     monospacedWidth=1
     size=7.5
     skewAmount=0
     typefaceFontName=Bold
     outlineSettings {
      bApplyOutlineToDropShadows=false
      bMiteredCorners=false
      bSeparateFillAlpha=false
      outlineMaterial=null
      outlineSize=0
      outlineColor=<rgba 0 0 0 1>
     }
    }
   }
   slot {class=BorderSlot, properties {horizontalAlignment=HAlign_Left, verticalAlignment=VAlign_Center, padding=<ltrb 0 2 0 0>}}
  }
  Border_9 {
   widgetClassPath=&Border
   parent=GridPanel_40
   widget=Border_9
   properties {
    horizontalAlignment=HAlign_Center
    verticalAlignment=VAlign_Center
    background {
     bIsDynamicallyLoaded=false
     drawAs=Border
     imageType=NoImage
     mirroring=NoMirror
     resourceName=None
     resourceObject=null
     tiling=NoTile
     imageSize=<32 32>
     margin=<ltrb 0 0 0 1>
     outlineSettings {
      bUseBrushTransparency=false
      roundingType=HalfHeightRadius
      width=0
      color {colorUseRule=UseColor_Specified, specifiedColor=<rgba 0 0 0 0>}
      cornerRadii=<0 0 0 0>
     }
     tintColor {colorUseRule=UseColor_Specified, specifiedColor=<rgba 1 1 1 1>}
     uVRegion {bIsValid=false, max=<0 0>, min=<0 0>}
    }
    brushColor=<rgba 0.0799 0.0799 0.0799 1>
    padding=<ltrb 0 2 0 2>
   }
   slot {class=GridSlot, properties {column=1, row=5}}
  }
  CheckBoxCustomFPS {
   widgetClassPath=&EditorUtilityCheckBox
   bIsVariable=true
   parent=Border_9
   widget=CheckBoxCustomFPS
   properties {horizontalAlignment=HAlign_Center}
   slot {class=BorderSlot, properties {horizontalAlignment=HAlign_Center, verticalAlignment=VAlign_Center, padding=<ltrb 0 2 0 2>}}
  }
  Border_10 {
   widgetClassPath=&Border
   parent=GridPanel_40
   widget=Border_10
   properties {
    horizontalAlignment=HAlign_Center
    verticalAlignment=VAlign_Center
    background {
     bIsDynamicallyLoaded=false
     drawAs=Border
     imageType=NoImage
     mirroring=NoMirror
     resourceName=None
     resourceObject=null
     tiling=NoTile
     imageSize=<32 32>
     margin=<ltrb 0 0 0 1>
     outlineSettings {
      bUseBrushTransparency=false
      roundingType=HalfHeightRadius
      width=0
      color {colorUseRule=UseColor_Specified, specifiedColor=<rgba 0 0 0 0>}
      cornerRadii=<0 0 0 0>
     }
     tintColor {colorUseRule=UseColor_Specified, specifiedColor=<rgba 1 1 1 1>}
     uVRegion {bIsValid=false, max=<0 0>, min=<0 0>}
    }
    brushColor=<rgba 0.0799 0.0799 0.0799 1>
    padding=<ltrb 0 2 0 2>
   }
   slot {class=GridSlot, properties {column=2, row=5, verticalAlignment=VAlign_Center}}
  }
  ComboBox_FPS {
   widgetClassPath=&EditorUtilityComboBoxString
   bIsVariable=true
   parent=Border_10
   widget=ComboBox_FPS
   properties {
    selectedOption=30
    contentPadding=<ltrb 0 -2 0 -2>
    defaultOptions [12,15,24,25,30,48,50,60,100,120,240]
    font {
     bForceMonospaced=false
     bMaterialIsStencil=false
     fontMaterial=null
     fontObject=&Roboto
     letterSpacing=0
     monospacedWidth=1
     size=7.5
     skewAmount=0
     typefaceFontName=Bold
     outlineSettings {
      bApplyOutlineToDropShadows=false
      bMiteredCorners=false
      bSeparateFillAlpha=false
      outlineMaterial=null
      outlineSize=0
      outlineColor=<rgba 0 0 0 1>
     }
    }
    foregroundColor {colorUseRule=UseColor_Specified, specifiedColor=<rgba 0.5271 0.5271 0.5271 1>}
   }
   slot {class=BorderSlot, properties {horizontalAlignment=HAlign_Center, verticalAlignment=VAlign_Center, padding=<ltrb 0 2 0 2>}}
  }
  Border_12 {
   widgetClassPath=&Border
   parent=GridPanel_40
   widget=Border_12
   properties {
    horizontalAlignment=HAlign_Left
    verticalAlignment=VAlign_Center
    background {
     bIsDynamicallyLoaded=false
     drawAs=Border
     imageType=NoImage
     mirroring=NoMirror
     resourceName=None
     resourceObject=null
     tiling=NoTile
     imageSize=<32 32>
     margin=<ltrb 0 0 0 1>
     outlineSettings {
      bUseBrushTransparency=false
      roundingType=HalfHeightRadius
      width=0
      color {colorUseRule=UseColor_Specified, specifiedColor=<rgba 0 0 0 0>}
      cornerRadii=<0 0 0 0>
     }
     tintColor {colorUseRule=UseColor_Specified, specifiedColor=<rgba 1 1 1 1>}
     uVRegion {bIsValid=false, max=<0 0>, min=<0 0>}
    }
    brushColor=<rgba 0.0799 0.0799 0.0799 1>
    padding=<ltrb 0 2 4 2>
   }
   slot {class=GridSlot, properties {row=6, padding=<ltrb 0 2 0 0>}}
  }
  TextAdditive_4 {
   widgetClassPath=&TextBlock
   bIsVariable=true
   parent=Border_12
   widget=TextAdditive_4
   properties {
    text=Interpolation:
    colorAndOpacity {colorUseRule=UseColor_Specified, specifiedColor=<rgba 0.1545 0.1545 0.1545 1>}
    font {
     bForceMonospaced=false
     bMaterialIsStencil=false
     fontMaterial=null
     fontObject=&Roboto
     letterSpacing=0
     monospacedWidth=1
     size=7.5
     skewAmount=0
     typefaceFontName=Bold
     outlineSettings {
      bApplyOutlineToDropShadows=false
      bMiteredCorners=false
      bSeparateFillAlpha=false
      outlineMaterial=null
      outlineSize=0
      outlineColor=<rgba 0 0 0 1>
     }
    }
   }
   slot {class=BorderSlot, properties {horizontalAlignment=HAlign_Left, verticalAlignment=VAlign_Center, padding=<ltrb 0 2 4 2>}}
  }
  Border_13 {
   widgetClassPath=&Border
   parent=GridPanel_40
   widget=Border_13
   properties {
    background {
     bIsDynamicallyLoaded=false
     drawAs=Border
     imageType=NoImage
     mirroring=NoMirror
     resourceName=None
     resourceObject=null
     tiling=NoTile
     imageSize=<32 32>
     margin=<ltrb 0 0 0 1>
     outlineSettings {
      bUseBrushTransparency=false
      roundingType=HalfHeightRadius
      width=0
      color {colorUseRule=UseColor_Specified, specifiedColor=<rgba 0 0 0 0>}
      cornerRadii=<0 0 0 0>
     }
     tintColor {colorUseRule=UseColor_Specified, specifiedColor=<rgba 1 1 1 1>}
     uVRegion {bIsValid=false, max=<0 0>, min=<0 0>}
    }
    brushColor=<rgba 0.0799 0.0799 0.0799 1>
   }
   slot {class=GridSlot, properties {column=1, columnSpan=1, row=6, padding=<ltrb 0 2 0 0>}}
  }
  ComboBox_Interpolation {
   widgetClassPath=&EditorUtilityComboBoxString
   bIsVariable=true
   parent=Border_13
   widget=ComboBox_Interpolation
   properties {
    selectedOption=Linear
    contentPadding=<ltrb 0 -2 0 -2>
    defaultOptions [Linear,Step]
    font {
     bForceMonospaced=false
     bMaterialIsStencil=false
     fontMaterial=null
     fontObject=&Roboto
     letterSpacing=0
     monospacedWidth=1
     size=7.5
     skewAmount=0
     typefaceFontName=Bold
     outlineSettings {
      bApplyOutlineToDropShadows=false
      bMiteredCorners=false
      bSeparateFillAlpha=false
      outlineMaterial=null
      outlineSize=0
      outlineColor=<rgba 0 0 0 1>
     }
    }
    foregroundColor {colorUseRule=UseColor_Specified, specifiedColor=<rgba 0.5271 0.5271 0.5271 1>}
   }
   slot {class=BorderSlot}
  }
  Border_14 {
   widgetClassPath=&Border
   parent=GridPanel_40
   widget=Border_14
   properties {
    horizontalAlignment=HAlign_Center
    verticalAlignment=VAlign_Center
    background {
     bIsDynamicallyLoaded=false
     drawAs=Border
     imageType=NoImage
     mirroring=NoMirror
     resourceName=None
     resourceObject=null
     tiling=NoTile
     imageSize=<32 32>
     margin=<ltrb 0 0 0 1>
     outlineSettings {
      bUseBrushTransparency=false
      roundingType=HalfHeightRadius
      width=0
      color {colorUseRule=UseColor_Specified, specifiedColor=<rgba 0 0 0 0>}
      cornerRadii=<0 0 0 0>
     }
     tintColor {colorUseRule=UseColor_Specified, specifiedColor=<rgba 1 1 1 1>}
     uVRegion {bIsValid=false, max=<0 0>, min=<0 0>}
    }
    brushColor=<rgba 0.0799 0.0799 0.0799 1>
    padding=<ltrb 0 0 0 0>
   }
   slot {class=GridSlot, properties {column=2, row=6}}
  }
  SizeHolder_2 {
   widgetClassPath=&EditorUtilityCheckBox
   parent=Border_14
   widget=SizeHolder_2
   properties {horizontalAlignment=HAlign_Center, visibility=Hidden}
   slot {class=BorderSlot, properties {horizontalAlignment=HAlign_Center, verticalAlignment=VAlign_Center, padding=<ltrb 0 0 0 0>}}
  }
  Border_FaceCurves2 {
   widgetClassPath=&Border
   bIsVariable=true
   parent=GridPanel_40
   widget=Border_FaceCurves2
   properties {
    horizontalAlignment=HAlign_Center
    verticalAlignment=VAlign_Center
    background {
     bIsDynamicallyLoaded=false
     drawAs=Border
     imageType=NoImage
     mirroring=NoMirror
     resourceName=None
     resourceObject=null
     tiling=NoTile
     imageSize=<32 32>
     margin=<ltrb 0 0 0 1>
     outlineSettings {
      bUseBrushTransparency=false
      roundingType=HalfHeightRadius
      width=0
      color {colorUseRule=UseColor_Specified, specifiedColor=<rgba 0 0 0 0>}
      cornerRadii=<0 0 0 0>
     }
     tintColor {colorUseRule=UseColor_Specified, specifiedColor=<rgba 1 1 1 1>}
     uVRegion {bIsValid=false, max=<0 0>, min=<0 0>}
    }
    brushColor=<rgba 0.0799 0.0799 0.0799 1>
    padding=<ltrb 0 2 0 2>
   }
   slot {class=GridSlot, properties {column=2, row=7}}
  }
  HorizontalBox_1 {
   widgetClassPath=&HorizontalBox
   parent=Border_FaceCurves2
   widget=HorizontalBox_1
   slot {class=BorderSlot, properties {horizontalAlignment=HAlign_Center, verticalAlignment=VAlign_Center, padding=<ltrb 0 2 0 2>}}
  }
  CheckBoxSkipFaceCurves {
   widgetClassPath=&EditorUtilityCheckBox
   bIsVariable=true
   parent=HorizontalBox_1
   widget=CheckBoxSkipFaceCurves
   properties {horizontalAlignment=HAlign_Center}
   slot {class=HorizontalBoxSlot, properties {horizontalAlignment=HAlign_Center, verticalAlignment=VAlign_Center}}
  }
  ButtonFaceCurves {
   widgetClassPath=&EditorUtilityButton
   parent=HorizontalBox_1
   widget=ButtonFaceCurves
   properties {
    toolTipText=Facial Curves information
    widgetStyle {
     clickedSlateSound {resourceObject=null}
     disabled {
      bIsDynamicallyLoaded=false
      drawAs=RoundedBox
      imageType=NoImage
      mirroring=NoMirror
      resourceName=None
      resourceObject=null
      tiling=NoTile
      imageSize=<0 0>
      margin=<ltrb 0 0 0 0>
      outlineSettings {
       bUseBrushTransparency=false
       roundingType=FixedRadius
       width=1
       color {colorUseRule=UseColor_ColorTable, specifiedColor=<rgba 0 0 0 0>}
       cornerRadii=<4 4 4 4>
      }
      tintColor {colorUseRule=UseColor_Specified, specifiedColor=<rgba 0.0395 0.0395 0.0395 1>}
      uVRegion {bIsValid=false, max=<0 0>, min=<0 0>}
     }
     disabledForeground {colorUseRule=UseColor_ColorTable, specifiedColor=<rgba 0 0 0 0>}
     hovered {
      bIsDynamicallyLoaded=false
      drawAs=RoundedBox
      imageType=NoImage
      mirroring=NoMirror
      resourceName=None
      resourceObject=null
      tiling=NoTile
      imageSize=<32 32>
      margin=<ltrb 0 0 0 0>
      outlineSettings {
       bUseBrushTransparency=false
       roundingType=FixedRadius
       width=1
       color {colorUseRule=UseColor_ColorTable, specifiedColor=<rgba 0 0 0 0>}
       cornerRadii=<4 4 4 4>
      }
      tintColor {colorUseRule=UseColor_Specified, specifiedColor=<rgba 0.0953 0.0953 0.0953 1>}
      uVRegion {bIsValid=false, max=<0 0>, min=<0 0>}
     }
     hoveredForeground {colorUseRule=UseColor_ColorTable, specifiedColor=<rgba 0 0 0 0>}
     hoveredSlateSound {resourceObject=null}
     normal {
      bIsDynamicallyLoaded=false
      drawAs=RoundedBox
      imageType=NoImage
      mirroring=NoMirror
      resourceName=None
      resourceObject=null
      tiling=NoTile
      imageSize=<32 32>
      margin=<ltrb 0 0 0 0>
      outlineSettings {
       bUseBrushTransparency=false
       roundingType=FixedRadius
       width=1
       color {colorUseRule=UseColor_ColorTable, specifiedColor=<rgba 0 0 0 0>}
       cornerRadii=<4 4 4 4>
      }
      tintColor {colorUseRule=UseColor_Specified, specifiedColor=<rgba 0.0395 0.0395 0.0395 1>}
      uVRegion {bIsValid=false, max=<0 0>, min=<0 0>}
     }
     normalForeground {colorUseRule=UseColor_ColorTable, specifiedColor=<rgba 0 0 0 0>}
     normalPadding=<ltrb 8 2 6 2>
     pressed {
      bIsDynamicallyLoaded=false
      drawAs=RoundedBox
      imageType=NoImage
      mirroring=NoMirror
      resourceName=None
      resourceObject=null
      tiling=NoTile
      imageSize=<32 32>
      margin=<ltrb 0 0 0 0>
      outlineSettings {
       bUseBrushTransparency=false
       roundingType=FixedRadius
       width=1
       color {colorUseRule=UseColor_ColorTable, specifiedColor=<rgba 0 0 0 0>}
       cornerRadii=<4 4 4 4>
      }
      tintColor {colorUseRule=UseColor_Specified, specifiedColor=<rgba 0.0284 0.0284 0.0284 1>}
      uVRegion {bIsValid=false, max=<0 0>, min=<0 0>}
     }
     pressedForeground {colorUseRule=UseColor_ColorTable, specifiedColor=<rgba 0 0 0 0>}
     pressedPadding=<ltrb 8 2.5 6 2.5>
     pressedSlateSound {resourceObject=null}
    }
   }
   slot {class=HorizontalBoxSlot, properties {horizontalAlignment=HAlign_Center, verticalAlignment=VAlign_Center, padding=<ltrb 0 2 0 2>}}
  }
  ButtonFaceCurvesText {
   widgetClassPath=&TextBlock
   bIsVariable=true
   parent=ButtonFaceCurves
   widget=ButtonFaceCurvesText
   properties {
    text=?
    colorAndOpacity {colorUseRule=UseColor_Specified, specifiedColor=<rgba 1 0 0 1>}
    font {
     bForceMonospaced=false
     bMaterialIsStencil=false
     fontMaterial=null
     fontObject=&Roboto
     letterSpacing=0
     monospacedWidth=1
     size=9
     skewAmount=0
     typefaceFontName=Bold
     outlineSettings {
      bApplyOutlineToDropShadows=false
      bMiteredCorners=false
      bSeparateFillAlpha=false
      outlineMaterial=null
      outlineSize=0
      outlineColor=<rgba 0 0 0 1>
     }
    }
   }
   slot {class=ButtonSlot, properties {horizontalAlignment=HAlign_Left, padding=<ltrb 0 0 0 0>}}
  }
  Border_FaceCurves1 {
   widgetClassPath=&Border
   bIsVariable=true
   parent=GridPanel_40
   widget=Border_FaceCurves1
   properties {
    horizontalAlignment=HAlign_Left
    verticalAlignment=VAlign_Center
    background {
     bIsDynamicallyLoaded=false
     drawAs=Border
     imageType=NoImage
     mirroring=NoMirror
     resourceName=None
     resourceObject=null
     tiling=NoTile
     imageSize=<32 32>
     margin=<ltrb 0 0 0 1>
     outlineSettings {
      bUseBrushTransparency=false
      roundingType=HalfHeightRadius
      width=0
      color {colorUseRule=UseColor_Specified, specifiedColor=<rgba 0 0 0 0>}
      cornerRadii=<0 0 0 0>
     }
     tintColor {colorUseRule=UseColor_Specified, specifiedColor=<rgba 1 1 1 1>}
     uVRegion {bIsValid=false, max=<0 0>, min=<0 0>}
    }
    brushColor=<rgba 0.0799 0.0799 0.0799 1>
    padding=<ltrb 0 2 4 2>
   }
   slot {class=GridSlot, properties {columnSpan=2, row=7, padding=<ltrb 0 2 0 0>}}
  }
  TextAdditive_7 {
   widgetClassPath=&TextBlock
   bIsVariable=true
   parent=Border_FaceCurves1
   widget=TextAdditive_7
   properties {
    text=Skip Face Curves:
    colorAndOpacity {colorUseRule=UseColor_Specified, specifiedColor=<rgba 0.1545 0.1545 0.1545 1>}
    font {
     bForceMonospaced=false
     bMaterialIsStencil=false
     fontMaterial=null
     fontObject=&Roboto
     letterSpacing=0
     monospacedWidth=1
     size=7.5
     skewAmount=0
     typefaceFontName=Bold
     outlineSettings {
      bApplyOutlineToDropShadows=false
      bMiteredCorners=false
      bSeparateFillAlpha=false
      outlineMaterial=null
      outlineSize=0
      outlineColor=<rgba 0 0 0 1>
     }
    }
   }
   slot {class=BorderSlot, properties {horizontalAlignment=HAlign_Left, verticalAlignment=VAlign_Center, padding=<ltrb 0 2 4 2>}}
  }
  Border_16 {
   widgetClassPath=&Border
   parent=GridPanel_40
   widget=Border_16
   properties {
    background {
     bIsDynamicallyLoaded=false
     drawAs=Border
     imageType=NoImage
     mirroring=NoMirror
     resourceName=None
     resourceObject=null
     tiling=NoTile
     imageSize=<32 32>
     margin=<ltrb 0 0 0 1>
     outlineSettings {
      bUseBrushTransparency=false
      roundingType=HalfHeightRadius
      width=0
      color {colorUseRule=UseColor_Specified, specifiedColor=<rgba 0 0 0 0>}
      cornerRadii=<0 0 0 0>
     }
     tintColor {colorUseRule=UseColor_Specified, specifiedColor=<rgba 1 1 1 1>}
     uVRegion {bIsValid=false, max=<0 0>, min=<0 0>}
    }
    brushColor=<rgba 0.0799 0.0799 0.0799 1>
    padding=<ltrb 0 2 4 2>
   }
   slot {class=GridSlot, properties {columnSpan=2, row=8, padding=<ltrb 0 2 0 0>}}
  }
  TextAdditive_6 {
   widgetClassPath=&TextBlock
   bIsVariable=true
   parent=Border_16
   widget=TextAdditive_6
   properties {
    text=Set Retarget Source Asset
    colorAndOpacity {colorUseRule=UseColor_Specified, specifiedColor=<rgba 0.1545 0.1545 0.1545 1>}
    font {
     bForceMonospaced=false
     bMaterialIsStencil=false
     fontMaterial=null
     fontObject=&Roboto
     letterSpacing=0
     monospacedWidth=1
     size=7.5
     skewAmount=0
     typefaceFontName=Bold
     outlineSettings {
      bApplyOutlineToDropShadows=false
      bMiteredCorners=false
      bSeparateFillAlpha=false
      outlineMaterial=null
      outlineSize=0
      outlineColor=<rgba 0 0 0 1>
     }
    }
   }
   slot {class=BorderSlot, properties {padding=<ltrb 0 2 4 2>}}
  }
  Border_17 {
   widgetClassPath=&Border
   parent=GridPanel_40
   widget=Border_17
   properties {
    horizontalAlignment=HAlign_Center
    verticalAlignment=VAlign_Center
    background {
     bIsDynamicallyLoaded=false
     drawAs=Border
     imageType=NoImage
     mirroring=NoMirror
     resourceName=None
     resourceObject=null
     tiling=NoTile
     imageSize=<32 32>
     margin=<ltrb 0 0 0 1>
     outlineSettings {
      bUseBrushTransparency=false
      roundingType=HalfHeightRadius
      width=0
      color {colorUseRule=UseColor_Specified, specifiedColor=<rgba 0 0 0 0>}
      cornerRadii=<0 0 0 0>
     }
     tintColor {colorUseRule=UseColor_Specified, specifiedColor=<rgba 1 1 1 1>}
     uVRegion {bIsValid=false, max=<0 0>, min=<0 0>}
    }
    brushColor=<rgba 0.0799 0.0799 0.0799 1>
    padding=<ltrb 0 2 0 2>
   }
   slot {class=GridSlot, properties {column=2, row=8}}
  }
  CheckBoxSetRetargetSourceAsset {
   widgetClassPath=&EditorUtilityCheckBox
   bIsVariable=true
   parent=Border_17
   widget=CheckBoxSetRetargetSourceAsset
   properties {horizontalAlignment=HAlign_Center}
   slot {class=BorderSlot, properties {horizontalAlignment=HAlign_Center, verticalAlignment=VAlign_Center, padding=<ltrb 0 2 0 2>}}
  }
  Border_15 {
   widgetClassPath=&Border
   parent=GridPanel_40
   widget=Border_15
   properties {
    background {
     bIsDynamicallyLoaded=false
     drawAs=Border
     imageType=NoImage
     mirroring=NoMirror
     resourceName=None
     resourceObject=null
     tiling=NoTile
     imageSize=<32 32>
     margin=<ltrb 0 0 0 1>
     outlineSettings {
      bUseBrushTransparency=false
      roundingType=HalfHeightRadius
      width=0
      color {colorUseRule=UseColor_Specified, specifiedColor=<rgba 0 0 0 0>}
      cornerRadii=<0 0 0 0>
     }
     tintColor {colorUseRule=UseColor_Specified, specifiedColor=<rgba 1 1 1 1>}
     uVRegion {bIsValid=false, max=<0 0>, min=<0 0>}
    }
    brushColor=<rgba 0.0799 0.0799 0.0799 1>
    padding=<ltrb 0 2 4 2>
   }
   slot {class=GridSlot, properties {columnSpan=2, row=9, padding=<ltrb 0 2 0 0>}}
  }
  TextAdditive_5 {
   widgetClassPath=&TextBlock
   bIsVariable=true
   parent=Border_15
   widget=TextAdditive_5
   properties {
    text=Skip Curves w/ Zero Value:
    colorAndOpacity {colorUseRule=UseColor_Specified, specifiedColor=<rgba 0.1545 0.1545 0.1545 1>}
    font {
     bForceMonospaced=false
     bMaterialIsStencil=false
     fontMaterial=null
     fontObject=&Roboto
     letterSpacing=0
     monospacedWidth=1
     size=7.5
     skewAmount=0
     typefaceFontName=Bold
     outlineSettings {
      bApplyOutlineToDropShadows=false
      bMiteredCorners=false
      bSeparateFillAlpha=false
      outlineMaterial=null
      outlineSize=0
      outlineColor=<rgba 0 0 0 1>
     }
    }
   }
   slot {class=BorderSlot, properties {padding=<ltrb 0 2 4 2>}}
  }
  Border_11 {
   widgetClassPath=&Border
   parent=GridPanel_40
   widget=Border_11
   properties {
    horizontalAlignment=HAlign_Center
    verticalAlignment=VAlign_Center
    background {
     bIsDynamicallyLoaded=false
     drawAs=Border
     imageType=NoImage
     mirroring=NoMirror
     resourceName=None
     resourceObject=null
     tiling=NoTile
     imageSize=<32 32>
     margin=<ltrb 0 0 0 1>
     outlineSettings {
      bUseBrushTransparency=false
      roundingType=HalfHeightRadius
      width=0
      color {colorUseRule=UseColor_Specified, specifiedColor=<rgba 0 0 0 0>}
      cornerRadii=<0 0 0 0>
     }
     tintColor {colorUseRule=UseColor_Specified, specifiedColor=<rgba 1 1 1 1>}
     uVRegion {bIsValid=false, max=<0 0>, min=<0 0>}
    }
    brushColor=<rgba 0.0799 0.0799 0.0799 1>
    padding=<ltrb 0 2 0 2>
   }
   slot {class=GridSlot, properties {column=2, row=9}}
  }
  CheckBoxSkipZeroCurves {
   widgetClassPath=&EditorUtilityCheckBox
   bIsVariable=true
   parent=Border_11
   widget=CheckBoxSkipZeroCurves
   properties {horizontalAlignment=HAlign_Center}
   slot {class=BorderSlot, properties {horizontalAlignment=HAlign_Center, verticalAlignment=VAlign_Center, padding=<ltrb 0 2 0 2>}}
  }
 ]
}
