EditorUtilityWidgetBlueprint /AnimInEdTools/CroppedSequenceManager/EUW_CroppedSequenceManager
graphs [
 CreateButtonClicked {nodes [CreateButtonClicked() {is=title K2Node_FunctionEntry}]}
 Dispatch {nodes [Dispatch() {is=title K2Node_FunctionEntry}]}
 EventGraph {nodes [
   n1 = Get Current Animatable Creature() -> (Current Animatable Creature:Animatable Creature Object Reference)
   n2 = Create Widget!(
    execute:Exec=@n68.Is Valid
    WorldContextObject:Object Reference
    Class:User Widget Class Reference=/AnimInEdTools/AnimationBrowser/EUW_Animation_Dialog.EUW_Animation_Dialog_C
    OwningPlayer:Player Controller Object Reference
   ) -> (then:Exec, ReturnValue:EUW Animation Dialog Object Reference)
   Set Initial Settings!(
    execute
    self:EUW Animation Dialog Object Reference=[@n2, @n2]
    Project:BP Animation Browser Project Object Reference=Get Current Project() -> (Current Project:BP Animation Browser Project Object Reference)
    Selected Creature:Animatable Creature Object Reference=@n1
    Developer Mode:bool=false
    Set Plugin:bool=true
    Animatable Creatures:Array of Animatable Creature Object References=Get AnimatableCreatures() -> (AnimatableCreatures:Array of Animatable Creature Object References)
    Show Project Dialog:bool=false
    Override Anim Path Validation:bool=false
   ) -> (then) {is=title K2Node_CallFunction}
   > Update UI from AnimSequence!(
    execute
    self:EUW Animation Dialog Object Reference=@n2
    Anim Sequence Path:String=Get Ref Anim Sequence Full Path() -> (Ref Anim Sequence Full Path:String)
    Animatable Creature:Animatable Creature Object Reference=@n1
   ) -> (then)
   n3 = "On Clicked (Button_Scan)"() {is=title K2Node_ComponentBoundEvent}
   n4 = Execute Python Script!(
    execute=[@n42.then, @n3.then]
    PythonScript="import sys\r\nCSM = sys.modules['CroppedSequenceManagerUI']\r\n\r\nresult = CSM.cls.check_if_same_level()"
   ) -> (then, ReturnValue, result:bool=~) {outputs [result]}
   > n5 = Branch(Condition=@n4.result)
   n6 = Execute Python Script!(
    execute=[@n34.then, @n43.then, @n44.then, @n36.then, @n48.then, @n40.then, @n46.then, @n5.then, @n41.then, @n38.then]
    PythonScript="import sys\r\nCSM = sys.modules['CroppedSequenceManagerUI']\r\n\r\nwidgets, sequences, create_buttons, fix_buttons, unlink_buttons, options_widgets = CSM.cls.init_ui()"
   ) -> (
    then
    ReturnValue
    widgets:Array of Border Object References
    sequences:Array of Anim Sequence Object References
    create_buttons:Array of Button Object References
    fix_buttons:Array of Button Object References
    unlink_buttons:Array of Button Object References
    options_widgets:Array of Horizontal Box Object References
   ) {outputs [widgets, sequences, create_buttons, fix_buttons, unlink_buttons, options_widgets]}
   > Set AnimSeq Border Widgets(AnimSeq Border Widgets:Array of Border Object References=@n6.widgets) -> (Output_Get:Array of Border Object References)
   > Set AnimSeq Sequences(AnimSeq Sequences:Array of Anim Sequence Object References=@n6.sequences) -> (Output_Get:Array of Anim Sequence Object References)
   > Set AnimSeq Create Buttons(AnimSeq Create Buttons:Array of Button Object References=@n6.create_buttons) -> (Output_Get:Array of Button Object References)
   > "Set AnimSeq Fix  Buttons"("AnimSeq Fix  Buttons":Array of Button Object References=@n6.fix_buttons) -> (Output_Get:Array of Button Object References)
   > Set Track Unlink Buttons(Track Unlink Buttons:Array of Button Object References=@n6.unlink_buttons) -> (Output_Get:Array of Button Object References)
   > Set AnimSeq Options Widgets(AnimSeq Options Widgets:Array of Horizontal Box Object References=@n6.options_widgets) -> (Output_Get:Array of Horizontal Box Object References)
   > n7 = Sequence!(execute) -> (then_0, then_1, then_2:Exec, then_3:Exec, then_4:Exec, then_5:Exec)
   > n8 = Sequence!(execute) -> (then_0, then_1, then_2)
   > n9 = LoadGameFromSlot!(execute, SlotName:String=AnimationBrowser, UserIndex:int=0) -> (then, ReturnValue:Save Game Object Reference)
   > n10 = Cast To SG_AnimationBrowser!(execute, Object=@n9) -> (then, CastFailed, AsSG Animation Browser:SG Animation Browser Object Reference) {is=title K2Node_DynamicCast}
   > n11 = Set Save Game(Save Game:SG Animation Browser Object Reference=@n10) -> (Output_Get:SG Animation Browser Object Reference)
   > n12 = LoadAsset_Blocking(Asset=Get Project!(self:SG Animation Browser Object Reference=@n11) -> (Project:BP Animation Browser Project Soft Object Reference))
   > n13 = Cast To BP_AnimationBrowser_Project!(execute, Object=@n12) -> (then, CastFailed, AsBP Animation Browser Project:BP Animation Browser Project Object Reference) {is=title K2Node_DynamicCast}
   n14 = Set Current Project(execute=[@n13.then, @n52.then], Current Project:BP Animation Browser Project Object Reference=@n13) -> (Output_Get:BP Animation Browser Project Object Reference)
   n15 = Get Compression Settings!(self:BP Animation Browser Project Object Reference=@n14) -> (Compression Settings:Anim Sequence Soft Object Reference)
   > n16 = Branch(Condition=IsValidSoftObjectReference(SoftObjectReference=@n15))
   > n17 = LoadAsset_Blocking(Asset=@n15)
   > n18 = Cast To AnimSequence!(execute, Object=@n17) -> (then, CastFailed, AsAnim Sequence:Anim Sequence Object Reference)
   > Execute Python Script!(
    execute
    PythonScript="import sys\r\nCSM = sys.modules['CroppedSequenceManagerUI']\r\n\r\nCSM.cls.set_project_settings(bone_compression, curve_compression, animation_interpolation_type)"
    bone_compression:Anim Bone Compression Settings Object Reference=GetBoneCompressionSettings!(AnimationSequence:Anim Sequence Object Reference=@n18) -> (CompressionSettings:Anim Bone Compression Settings Object Reference)
    curve_compression:Anim Curve Compression Settings Object Reference=GetCurveCompressionSettings!(AnimationSequence:Anim Sequence Object Reference=@n18) -> (CompressionSettings:Anim Curve Compression Settings Object Reference)
    animation_interpolation_type:EAnimInterpolationType Enum=GetAnimationInterpolationType!(AnimationSequence:Anim Sequence Object Reference=@n18) -> (InterpolationType:EAnimInterpolationType Enum=Linear)
   ) -> (then, ReturnValue) {inputs [bone_compression, curve_compression, animation_interpolation_type]}
   Execute Python Script!(
    execute=@n16.else
    PythonScript="import sys\r\nCSM = sys.modules['CroppedSequenceManagerUI']\r\n\r\nCSM.cls.set_project_settings(None,None, None)"
   ) -> (then, ReturnValue)
   SetAnimatableCreature(execute=@n8.then_2)
   > n19 = GetCurrentLevelSequence!(execute) -> (then, ReturnValue:Level Sequence Object Reference)
   > n20 = Branch(Condition="Not Equal (Object)"(A:Object Reference=@n19, B:Object Reference=Get Current Level Sequence() -> (Current Level Sequence:Level Sequence Object Reference)) -> (ReturnValue:bool=~) {is=title K2Node_PromotableOperator})
   > n21 = Scan for Animatable Creatures()
   Set Current Level Sequence(execute=[@n21.then, @n20.else], Current Level Sequence:Level Sequence Object Reference=@n19) -> (Output_Get:Level Sequence Object Reference)
   n22 = Get Current Project()
   Set Text(
    execute=@n8.then_1
    Text="To Text (String)"(InString=Get Name!(self:BP Animation Browser Project Object Reference=@n22) -> (Name:String))
    self=Get TextProject() -> (TextProject:Text Object Reference)
   )
   n23 = For Each Loop(
    Exec=@n7.then_1
    Array:Array of Border Object References=Get AnimSeq Border Widgets() -> (AnimSeq Border Widgets:Array of Border Object References)
   ) -> (Array Element:Border Object Reference)
   > n24 = Sequence!(execute) -> (then_0, then_1)
   > n25 = Create Widget(Class=&/EUW_CroppedSequenceManager_AnimSeqInfo.EUW_CroppedSequenceManager_AnimSeqInfo_C) -> (ReturnValue:EUW Cropped Sequence Manager Anim Seq Info Object Reference)
   > Set scale!(
    execute
    scale:double=GetValue!(self:Slider Object Reference=Get EditorUtilitySliderUIScale() -> (EditorUtilitySliderUIScale:Editor Utility Slider Object Reference)) -> (ReturnValue:float)
    self:EUW Cropped Sequence Manager Anim Seq Info Object Reference=@n25
   ) -> (then, Output_Get:double)
   > AddChild(self=@n23, Content=@n25)
   > Set widget id!(execute, widget id:int=@n23.Array Index, self:EUW Cropped Sequence Manager Anim Seq Info Object Reference=@n25) -> (then, Output_Get:int=0)
   n26 = "Get (a copy)"(
    Array:Array of Anim Sequence Object References=Get AnimSeq Sequences() -> (AnimSeq Sequences:Array of Anim Sequence Object References)
    Dimension 1=@n23.Array Index
   ) -> (Output:Anim Sequence Object Reference)
   > Set AnimSequence!(execute, AnimSequence:Anim Sequence Object Reference=@n26, self:EUW Cropped Sequence Manager Anim Seq Info Object Reference=@n25) -> (then, Output_Get:Anim Sequence Object Reference)
   > Bind Event to On Property Changed(
    self:Property View Base Object Reference=Get SinglePropertyView_AnimSequence!(self:EUW Cropped Sequence Manager Anim Seq Info Object Reference=@n25) -> (SinglePropertyView_AnimSequence:Single Property View Object Reference)
    Delegate=Create Event()
   ) {is=title K2Node_AddDelegate}
   Is Valid(exec=@n24.then_1, InputObject=@n26)
   > n27 = Create Widget(Class=&/EUW_CroppedSequenceManager_AnimSeqOptions.EUW_CroppedSequenceManager_AnimSeqOptions_C) -> (ReturnValue:EUW Cropped Sequence Manager Anim Seq Options Object Reference)
   > Set widget id(widget id=@n23.Array Index, self:EUW Cropped Sequence Manager Anim Seq Options Object Reference=@n27) -> (Output_Get=0)
   > Set AnimSequence(AnimSequence=@n26, self:EUW Cropped Sequence Manager Anim Seq Options Object Reference=@n27)
   > Set Level Seqeunce!(
    execute
    Level Seqeunce:Level Sequence Object Reference=Get Current Level Sequence()
    self:EUW Cropped Sequence Manager Anim Seq Options Object Reference=@n27
   ) -> (then, Output_Get:Level Sequence Object Reference)
   > Set Current Project!(execute, Current Project=Get Current Project(), self:EUW Cropped Sequence Manager Anim Seq Options Object Reference=@n27) -> (then, Output_Get)
   > AddChild(
    self="Get (a copy)"(
     Array:Array of Horizontal Box Object References=Get AnimSeq Options Widgets() -> (AnimSeq Options Widgets:Array of Horizontal Box Object References)
     Dimension 1=@n23.Array Index
    ) -> (Output:Horizontal Box Object Reference)
    Content=@n27
   )
   > InitWidgetValues(self:EUW Cropped Sequence Manager Anim Seq Options Object Reference=@n27)
   n28 = Sequence(execute=@n7.then_4)
   > n29 = For Each Loop(Array:Array of Button Object References="Get AnimSeq Fix  Buttons"() -> ("AnimSeq Fix  Buttons":Array of Button Object References)) -> (Array Element:Button Object Reference)
   n30 = Create Event()
   > Bind Event to On Clicked(self:Button Object Reference=@n29, Delegate=@n30) {is=title K2Node_AddDelegate}
   n31 = For Each Loop(Exec=@n28.then_1, Array=Get Track Unlink Buttons() -> (Track Unlink Buttons:Array of Button Object References))
   > Bind Event to On Clicked(self=@n31, Delegate=@n30)
   SetVisibility(execute=@n7.then_5, self=Get Border_Loading() -> (Border_Loading:Border Object Reference), InVisibility=Collapsed)
   n32 = For Each Loop(Exec=@n7.then_3, Array=Get AnimSeq Create Buttons() -> (AnimSeq Create Buttons:Array of Button Object References))
   n33 = CreateButtonClicked_Event() {is=title K2Node_CustomEvent}
   Bind Event to On Clicked(execute=@n32.LoopBody, self=@n32, Delegate=@n33)
   n34 = Execute Python Script(execute=[@n50.then, @n5.else], PythonScript="import sys\r\nCSM = sys.modules['CroppedSequenceManagerUI']\r\n\r\nCSM.cls.init()")
   n35 = "On Selection Changed (ComboBox_ToolMode)"!() -> (OutputDelegate, then, SelectedItem:String, SelectionType:ESelectInfo Enum) {is=title K2Node_ComponentBoundEvent}
   > n36 = Execute Python Script!(
    execute
    PythonScript="import sys\r\nCSM = sys.modules['CroppedSequenceManagerUI']\r\n\r\nCSM.cls.mode = mode"
    mode:String=@n35.SelectedItem
   ) -> (then, ReturnValue) {inputs [mode]}
   n37 = "On Controller Capture End (EditorUtilitySliderUIScale)"() {is=title K2Node_ComponentBoundEvent}
   n38 = Execute Python Script!(
    execute=[@n37.then, @n39.then]
    PythonScript="import sys\r\nCSM = sys.modules['CroppedSequenceManagerUI']\r\n\r\nCSM.cls.set_ui_scale(scale)"
    scale:float=GetValue(self=Get EditorUtilitySliderUIScale())
   ) -> (then, ReturnValue) {inputs [scale]}
   n39 = "On Mouse Capture End (EditorUtilitySliderUIScale)"() {is=title K2Node_ComponentBoundEvent}
   "On Clicked (Button_AddTrack)"() {is=title K2Node_ComponentBoundEvent}
   > n40 = Execute Python Script!(execute, PythonScript="import sys\r\nCSM = sys.modules['CroppedSequenceManagerUI']\r\n\r\nCSM.cls.add_track()") -> (then, ReturnValue)
   "On Clicked (Button_AddBookmark)"() {is=title K2Node_ComponentBoundEvent}
   > n41 = Execute Python Script!(
    execute
    PythonScript="import sys\r\nCSM = sys.modules['CroppedSequenceManagerUI']\r\n\r\nCSM.cls.bookmark_color = bookmark_color\r\nCSM.cls.button_make_bookmarks_clicked()"
    bookmark_color:Linear Color Structure=Get BookmarkColor() -> (BookmarkColor:Linear Color Structure)
   ) -> (then, ReturnValue) {inputs [bookmark_color]}
   "On Property Changed (SinglePropertyView_LevelSequence)"!() -> (OutputDelegate, then, PropertyName:Name=None) {is=title K2Node_ComponentBoundEvent}
   > n42 = OpenLevelSequence!(execute, LevelSequence:Level Sequence Object Reference=Get Current Level Sequence()) -> (then, ReturnValue:bool)
   "On Clicked (Button_?)"() {is=title K2Node_ComponentBoundEvent}
   > LaunchURL!(execute, URL:String=https://confluence.it.epicgames.com/display/PP/Cropped+Sequence+Manager) -> (then)
   n43 = SequenceInfoWidgetChanged!() -> (OutputDelegate, then, PropertyName:Name) {is=title K2Node_CustomEvent}
   n44 = RefreshTool() {is=title K2Node_CustomEvent}
   n45 = Create Custom Anim Sequence!() -> (OutputDelegate, then, AnimSequencePath:String) {is=title K2Node_CustomEvent}
   > n46 = Execute Python Script!(
    execute
    PythonScript="import sys\r\nCSM = sys.modules['CroppedSequenceManagerUI']\r\n\r\nCSM.cls.create_new_animation_sequence(anim_seq_path)"
    anim_seq_path:String=@n45.AnimSequencePath
   ) -> (then, ReturnValue) {inputs [anim_seq_path]}
   n47 = Add New Animation!() -> (OutputDelegate, then, AnimSequence Path:String, LevelSequence Path:String, Animatable Creature:Animatable Creature Object Reference) {is=title K2Node_CustomEvent}
   > n48 = Execute Python Script(
    PythonScript="import sys\r\nCSM = sys.modules['CroppedSequenceManagerUI']\r\n\r\nCSM.cls.create_new_animation_sequence(anim_seq_path)"
    anim_seq_path=@n47.AnimSequence Path
   ) {inputs [anim_seq_path]}
   n49 = Construct Step 1() {is=title K2Node_CustomEvent}
   > Execute Python Script!(
    execute
    PythonScript="import CroppedSequenceManagerUI\r\nimport importlib\r\n\r\nif CroppedSequenceManagerUI.DEBUG:\r\n    importlib.reload(CroppedSequenceManagerUI)"
   ) -> (then, ReturnValue)
   > Execute Python Script!(
    execute
    PythonScript="import sys\r\nCSM = sys.modules['CroppedSequenceManagerUI']\r\n\r\nCSM.cls.init_widget_layouts(layout_tracks, layout_sub_sequences)\r\nCSM.cls.init_fonts(text_regular, text_bold_big, text_add_track, text_header)\r\nCSM.cls.init_checkbox_style(checkbox_style)\r\nCSM.cls.init_add_button_style(add_button)\r\nCSM.cls.init_add_button_image(add_button_image)\r\nCSM.cls.init_zoom_button_style(zoom_button)\r\n\r\nCSM.cls.init_widget_combo_box_tracks(combo_box_add_track)\r\nCSM.cls.mode = mode\r\n"
    combo_box_add_track:Editor Utility Combo Box String Object Reference=Get ComboBox_AddTrack() -> (ComboBox_AddTrack:Editor Utility Combo Box String Object Reference)
    layout_tracks:Vertical Box Object Reference=Get Layout_Tracks() -> (Layout_Tracks:Vertical Box Object Reference)
    layout_sub_sequences:Vertical Box Object Reference=Get Layout_Cropped_Sequences() -> (Layout_Cropped_Sequences:Vertical Box Object Reference)
    text_regular:Text Object Reference=Get TextStyleSammple_Regular() -> (TextStyleSammple_Regular:Text Object Reference)
    text_bold_big:Text Object Reference=Get TextStyleSammple_Bold_Big() -> (TextStyleSammple_Bold_Big:Text Object Reference)
    text_add_track:Text Object Reference=Get Add_Track_Text() -> (Add_Track_Text:Text Object Reference)
    checkbox_style:Editor Utility Check Box Object Reference=Get Checkbox_Style() -> (Checkbox_Style:Editor Utility Check Box Object Reference)
    add_button:Button Object Reference=Get AddAnimationButton() -> (AddAnimationButton:Button Object Reference)
    add_button_image:Image Object Reference=Get AssetImagePreviewAddButton() -> (AssetImagePreviewAddButton:Image Object Reference)
    zoom_button:Button Object Reference=Get ZoomToFrames() -> (ZoomToFrames:Button Object Reference)
    mode:String=GetSelectedOption!(self:"Combo Box (String) Object Reference"=Get ComboBox_ToolMode() -> (ComboBox_ToolMode:Editor Utility Combo Box String Object Reference)) -> (ReturnValue:String)
    text_header:Text Object Reference=Get TextSampleHeader() -> (TextSampleHeader:Text Object Reference)
   ) -> (then, ReturnValue) {inputs [
     combo_box_add_track
     layout_tracks
     layout_sub_sequences
     text_regular
     text_bold_big
     text_add_track
     checkbox_style
     add_button
     add_button_image
     zoom_button
     mode
     text_header
    ]}
   > n50 = Execute Python Script!(
    execute
    PythonScript="import sys\r\nCSM = sys.modules['CroppedSequenceManagerUI']\r\n\r\nCSM.cls.init_widget_add_bookmarks(add_bookmark_button, etext_start_bm_label, etext_start_bm_frame, etext_end_bm_frame, bm_color)\r\n"
    add_bookmark_button:Editor Utility Button Object Reference=Get Button_AddBookmark() -> (Button_AddBookmark:Editor Utility Button Object Reference)
    etext_start_bm_label:Editor Utility Editable Text Box Object Reference=Get EditorUtilityEditableTextBoxStartBookmark() -> (EditorUtilityEditableTextBoxStartBookmark:Editor Utility Editable Text Box Object Reference)
    etext_start_bm_frame:Editor Utility Editable Text Box Object Reference=Get EditorUtilityEditableTextBoxStartFrame() -> (EditorUtilityEditableTextBoxStartFrame:Editor Utility Editable Text Box Object Reference)
    etext_end_bm_frame:Editor Utility Editable Text Box Object Reference=Get EditorUtilityEditableTextBoxEndFrame() -> (EditorUtilityEditableTextBoxEndFrame:Editor Utility Editable Text Box Object Reference)
    bm_color:Linear Color Structure=Get BookmarkColor()
   ) -> (then, ReturnValue) {inputs [add_bookmark_button, etext_start_bm_label, etext_start_bm_frame, etext_end_bm_frame, bm_color]}
   Event PreConstruct()
   n51 = Self-Reference()
   > SetObject(self=Get SinglePropertyView_LevelSequence() -> (SinglePropertyView_LevelSequence:Single Property View Object Reference), NewObject=@n51)
   > SetObject(self=Get SinglePropertyView_BookMarkColor() -> (SinglePropertyView_BookMarkColor:Single Property View Object Reference), NewObject=@n51)
   > SetCurrentUser()
   Event Construct()
   > Set Timer by Event(Delegate=@n49, Time=0.01, bLooping=false)
   n52 = Branch(Condition="Not Equal (Object)"(A=@n13, B=Get Current Project()) -> (ReturnValue=~))
   n53 = Create Widget(
    execute=[@n71.then, @n69.then]
    Class=&/EUW_CroppedSequenceManager_CreateCustomAnimSequence.EUW_CroppedSequenceManager_CreateCustomAnimSequence_C
   ) -> (ReturnValue:EUW Cropped Sequence Manager Create Custom Anim Sequence Object Reference)
   Bind Event to Create New Animation(execute=@n55.then, self:EUW Cropped Sequence Manager Create Custom Anim Sequence Object Reference=@n53, Delegate=Create Event()) {is=title K2Node_AddDelegate}
   Validate UI Input!(execute=@n54.then, self:EUW Animation Dialog Object Reference=@n2) -> (then, Is Valid:bool)
   Event Tick()
   Event Destruct()
   Execute Python Script!(
    execute
    PythonScript="import sys\r\nCSM = sys.modules['CroppedSequenceManager']\r\n\r\nCSM.cls.CSRenamer.revert_name(False)\r\nCSM.cls.CSRenamer.clear_temp_sub_folders()"
   ) -> (then, ReturnValue)
   Get EditorUtilityEditableTextBoxEndBookmark() -> (EditorUtilityEditableTextBoxEndBookmark:Editor Utility Editable Text Box Object Reference)
   Set Ref Anim Sequence Name(Ref Anim Sequence Name:String) -> (Output_Get:String)
   Bind Event to Generate Animation Paths(execute=@n56.then, self:EUW Animation Dialog Object Reference=@n2, Delegate=Create Event()) {is=title K2Node_AddDelegate}
   > n54 = Set Options from Animatable Creature!(
    execute
    self:EUW Animation Dialog Object Reference=@n2
    Selected Creature:Animatable Creature Object Reference=Get Current Animatable Creature()
    All Creatures:Array of Animatable Creature Object References=Get AnimatableCreatures()
    Update Assets:bool=true
    Update Plugins:bool=true
    Set Plugin:bool=true
   ) -> (then) {is=title K2Node_CallFunction}
   n55 = AddChild(execute=@n64.then, self=Get ModalLayer() -> (ModalLayer:Named Slot Object Reference), Content=@n53)
   AddChild(execute=@n2.then, self=Get ModalLayer(), Content=@n2)
   > n56 = Set CurrentUser!(execute, CurrentUser:String=Get Current User() -> (Current User:String), self:EUW Animation Dialog Object Reference=@n2) -> (then, Output_Get:String)
   n57 = Execute Python Script!(
    execute=@n74.then
    PythonScript="import sys\r\nCSM = sys.modules['CroppedSequenceManagerUI']\r\n\r\nac_path = CSM.cls.current_ac_file\r\nref_as_path = CSM.cls.ref_anim_sequence_full_path"
   ) -> (then, ReturnValue, ac_path:"Soft Object Path Structure (by ref)", ref_as_path:String) {outputs [ac_path,ref_as_path]}
   n58 = GetObjectFromSoftPath!(Path:"Soft Object Path Structure (by ref)"=@n57.ac_path) -> (ReturnValue:Object Reference)
   n59 = SetCreateWidgetWarning!(execute=@n72.then, self, CreateWidget:EUW Cropped Sequence Manager Create Custom Anim Sequence Object Reference=@n53) -> (then)
   n60 = Get Ref Anim Sequence Full Path()
   n61 = FindSubstring(SearchIn=@n60, Substring=/, bSearchFromEnd=true)
   Set Text(
    execute=[@n59.then, @n72.else]
    Text="To Text (String)"(InString=Left!(SourceString:String=@n60, Count:int=@n61) -> (ReturnValue:String))
    self:Text Box Object Reference=Get AnimPathTextBox!(self:EUW Cropped Sequence Manager Create Custom Anim Sequence Object Reference=@n53) -> (AnimPathTextBox:Text Box Object Reference)
   )
   n62 = Get Ref Anim Sequence Full Path()
   n63 = RightChop!(SourceString:String=@n62, Count:int=int + int(A=FindSubstring(SearchIn=@n62, Substring=., StartPosition=0))) -> (ReturnValue:String)
   > n64 = Set Text(
    Text="To Text (String)"(InString=Append(A=@n63, B=_WEP))
    self=Get AnimNameTextBox!(self:EUW Cropped Sequence Manager Create Custom Anim Sequence Object Reference=@n53) -> (AnimNameTextBox:Text Box Object Reference)
   )
   n65 = NOT Boolean(A=Get Is Ctrl Pressed() -> (Is Ctrl Pressed:bool))
   n66 = Cast To AnimatableCreature!(execute=@n70.then, Object=@n58) -> (then, CastFailed, AsAnimatable Creature:Animatable Creature Object Reference) {is=title K2Node_DynamicCast}
   > n67 = Set Current Animatable Creature(Current Animatable Creature:Animatable Creature Object Reference=@n66) -> (Output_Get:Animatable Creature Object Reference)
   n68 = Is Valid(exec=[@n67.then, @n66.CastFailed], InputObject=Get Current Animatable Creature())
   n69 = Set Show Warning On Create Widget(execute=@n68.Is Not Valid, Show Warning On Create Widget:bool=true) -> (Output_Get:bool)
   n70 = Branch(execute=@n73.then, Condition=@n65)
   n71 = Set Show Warning On Create Widget(execute=@n70.else, Show Warning On Create Widget=false)
   n72 = Branch(execute=@n53.then, Condition=Get Show Warning On Create Widget() -> (Show Warning On Create Widget:bool))
   n73 = Set Ref Anim Sequence Full Path(execute=@n57.then, Ref Anim Sequence Full Path:String=@n57.ref_as_path) -> (Output_Get:String)
   n74 = Set Current Animatable Creature(execute=@n33.then)
  ]}
 Get Animatable Creature From ID {nodes [
   n1 = Get Animatable Creature From ID!() -> (then, widget_id:int) {is=title K2Node_FunctionEntry}
   > n2 = Execute Python Script!(
    execute
    PythonScript="import sys\r\nCSM = sys.modules['CroppedSequenceManagerUI']\r\n\r\npath = CSM.cls.get_ac_file_by_id(widget_id)\r\n"
    widget_id:int=@n1
   ) -> (then, ReturnValue, path:Soft Object Path Structure) {inputs [widget_id], outputs [path]}
   > n3 = Set Path(Path:Soft Object Path Structure=@n2.path) -> (Output_Get:Soft Object Path Structure)
   > n4 = Cast To AnimatableCreature(Object=GetObjectFromSoftPath(Path=@n3))
   > Return Node!(execute, Animatable Creature:Animatable Creature Object Reference=@n4)
   Return Node(execute=@n4.CastFailed)
  ]}
 LevelSaved {nodes [LevelSaved() {is=title K2Node_FunctionEntry}]}
 OnKeyDown {nodes [
   n1 = OnKeyDown!() -> (then, MyGeometry:Geometry Structure, InKeyEvent:Key Event Structure) {is=title K2Node_FunctionEntry}
   n2 = GetKey!(Input:"Key Event Structure (by ref)"=@n1.InKeyEvent) -> (ReturnValue:Key Structure)
   > Set Is Ctrl Pressed(Is Ctrl Pressed:bool=OR Boolean(
     A="Equal (Key)"(A:Key Structure=@n2, B:Key Structure=LeftControl) -> (ReturnValue:bool=~) {is=title K2Node_PromotableOperator}
     B="Equal (Key)"(A=@n2, B=RightControl) -> (ReturnValue=~)
    )) -> (Output_Get:bool)
   > Return Node!(execute, ReturnValue:Event Reply Structure=Handled())
  ]}
 OnKeyUp {nodes [
   n1 = OnKeyUp!() -> (then, MyGeometry:Geometry Structure, InKeyEvent:Key Event Structure) {is=title K2Node_FunctionEntry}
   > Set Is Ctrl Pressed(Is Ctrl Pressed=false)
   > Return Node(ReturnValue=Handled())
   n2 = GetKey(Input=@n1.InKeyEvent)
   NOT Boolean(A=OR Boolean(A="Equal (Key)"(A=@n2, B=LeftControl) -> (ReturnValue=~), B="Equal (Key)"(A=@n2, B=RightControl) -> (ReturnValue=~)))
  ]}
 Scan for Animatable Creatures {nodes [
   Scan for Animatable Creatures() {is=title K2Node_FunctionEntry}
   > Clear(TargetArray=Get Assets() -> (Assets:Array of Strings))
   > Clear(TargetArray=Get AnimPaths() -> (AnimPaths:Array of Strings))
   > Clear(TargetArray:Array of Object Soft Class References=Get Actors() -> (Actors:Array of Object Soft Class References))
   > Clear(TargetArray:Array of Strings=Get AC_Paths() -> (AC_Paths:Array of Strings))
   > n1 = For Each Loop(Array:Array of Animatable Creature Object References=Get AnimatableCreatures()) -> (Array Element:Animatable Creature Object Reference)
   Execute Python Script!(
    execute=@n1.Completed
    PythonScript="import sys\r\nCSM = sys.modules['CroppedSequenceManagerUI']\r\n\r\nCSM.cls.CSM.intake_animatable_creatures(ac_files, actors, assets, anim_paths)\r\n"
    ac_files:Array of Strings=Get AC_Paths()
    actors:Array of Object Soft Class References=Get Actors()
    assets:Array of Strings=Get Assets()
    anim_paths:Array of Strings=Get AnimPaths()
   ) -> (then, ReturnValue) {inputs [ac_files,actors,assets,anim_paths]}
   > Return Node!(execute)
   Add(execute=@n1.LoopBody, TargetArray=Get Assets(), NewItem=Get Asset!(self:Animatable Creature Object Reference=@n1) -> (Asset:String)) {is=title K2Node_CallArrayFunction}
   > Add(TargetArray=Get AnimPaths(), NewItem=Get Animation Path!(self:Animatable Creature Object Reference=@n1) -> (Animation Path:String))
   > Add(
    TargetArray:Array of Object Soft Class References=Get Actors()
    NewItem:"Object Soft Class Reference (by ref)"=Get Sequencer Actor!(self:Animatable Creature Object Reference=@n1) -> (Sequencer Actor:Object Soft Class Reference)
   )
   > Add(TargetArray:Array of Strings=Get AC_Paths(), NewItem:"String (by ref)"=GetSoftObjectPath!(Object=@n1) -> (ReturnValue_PathString:String))
  ]}
 SetAnimatableCreature {nodes [
   SetAnimatableCreature() {is=title K2Node_FunctionEntry}
   > n1 = Branch(Condition=IsLoadingAssets!(self:Asset Registry Interface=GetAssetRegistry()) -> (ReturnValue:bool))
   n2 = Get Current Project()
   n3 = Set AnimatableCreatures(
    execute=@n1.else
    AnimatableCreatures:Array of Animatable Creature Object References=Get Animatable Creatures!(Project:BP Animation Browser Project Object Reference=@n2, bSearchSubClasses:bool=true, __WorldContext:Object Reference) -> (Animatable Creatures:Array of Animatable Creature Object References) {is=title K2Node_CallFunction}
   ) -> (Output_Get:Array of Animatable Creature Object References)
   > Append Array!(
    execute
    TargetArray:Array of Animatable Creature Object References=@n3
    SourceArray:Array of Animatable Creature Object References=Get Animatable Props!(Project:BP Animation Browser Project Object Reference=@n2, __WorldContext:Object Reference) -> (Animatable Creatures:Array of Animatable Creature Object References) {is=title K2Node_CallFunction}
   ) -> (then)
  ]}
 SetCreateWidgetWarning {nodes [
   n1 = SetCreateWidgetWarning!() -> (then, CreateWidget:EUW Cropped Sequence Manager Create Custom Anim Sequence Object Reference) {is=title K2Node_FunctionEntry}
   > n2 = Execute Python Script!(
    execute
    PythonScript="import sys\r\nCSM = sys.modules['CroppedSequenceManagerUI']\r\n\r\nresult, tooltip_text = CSM.cls.get_create_animation_dialogue_tooltip(project_name)"
    project_name:String=Get Name(self=Get Current Project())
   ) -> (then, ReturnValue, tooltip_text:Text) {inputs [project_name], outputs [tooltip_text]}
   > SetVisibility(
    self=[
     @n1
     Get Button_?!(self:EUW Cropped Sequence Manager Create Custom Anim Sequence Object Reference=@n1) -> (Button_?:Editor Utility Button Object Reference)
    ]
    InVisibility=Visible
   )
   > Set Warning Button Text!(execute, Warning Button Text:Text=@n2.tooltip_text, self:EUW Cropped Sequence Manager Create Custom Anim Sequence Object Reference=@n1) -> (then, Output_Get:Text)
  ]}
 SetCurrentUser {nodes [
   SetCurrentUser() {is=title K2Node_FunctionEntry}
   > Branch(Condition=IsEmpty(InString=Get Current User()))
   > n1 = Execute Python Script!(execute, PythonScript="import os\r\nuser = os.getlogin()") -> (then, ReturnValue, user:String) {outputs [user]}
   > Set Current User(Current User:String=@n1.user) -> (Output_Get:String)
  ]}
]
variables [
 AnimatableCreatures {type="array<AnimatableCreature_C>", default []}
 AnimationBrowser {type=Animation_Browser_C, default=null}
 AnimSeq Border Widgets {type="array<Border>", default []}
 AnimSeq Create Buttons {type="array<Button>", default []}
 "AnimSeq Fix  Buttons" {type="array<Button>", default []}
 AnimSeq Options Widgets {type="array<HorizontalBox>", default []}
 AnimSeq Sequences {type="array<AnimSequence>", default [null]}
 BlendSpace {type=BlendSpace, default=null}
 BoneCompressionSettings {type=Object, default=null}
 BookmarkColor {type=LinearColor, default=<rgba 0 1 1 0.4>}
 CreateButtonClicked {default=()}
 Current Animatable Creature {type=AnimatableCreature_C, default=null}
 Current Level Sequence {type=LevelSequence, default=null}
 Current Project {type=BP_AnimationBrowser_Project_C, default=null}
 Current User {type=string, default=""}
 Data Asset {type=DA_AnimationBrowser_C, default=/AnimInEdTools/AnimationBrowser/DA_Play_AnimationBrowser}
 Dispatch {default=()}
 Is Ctrl Pressed {type=boolean}
 LevelSaved {default=()}
 NewVar {type=string, default=""}
 Ref Anim Asset {type=string, default=""}
 Ref Anim Sequence Full Path {type=string, default=""}
 Ref Anim Sequence Name {type=string, default=""}
 Save Game {type=SG_AnimationBrowser_C, default=null}
 Sequencer Actor {type=Class, default=null}
 Show Warning On Create Widget {type=boolean}
 Track Unlink Buttons {type="array<Button>", default []}
]
widgets {
 info {rootWidgetClass=&Overlay, widgetCount=71}
 widgets [
  Overlay_0 {widgetClassPath=&Overlay, widget=Overlay_0}
  EditorUtilityScrollBox_0 {
   widgetClassPath=&EditorUtilityScrollBox
   parent=Overlay_0
   widget=EditorUtilityScrollBox_0
   properties {orientation=Orient_Horizontal}
   slot {class=OverlaySlot, properties {horizontalAlignment=HAlign_Fill}}
  }
  EditorUtilityScrollBox_1 {
   widgetClassPath=&EditorUtilityScrollBox
   parent=EditorUtilityScrollBox_0
   widget=EditorUtilityScrollBox_1
   slot {class=ScrollBoxSlot, properties {verticalAlignment=VAlign_Top, size {sizeRule=Fill, value=1}}}
  }
  CoreLayout {
   widgetClassPath=&VerticalBox
   bIsVariable=true
   parent=EditorUtilityScrollBox_1
   widget=CoreLayout
   properties {renderTransformPivot=<0 0>}
   slot {class=ScrollBoxSlot, properties {verticalAlignment=VAlign_Top, padding=<ltrb 4 0 4 4>, size {sizeRule=Fill, value=1}}}
  }
  Border {
   widgetClassPath=&Border
   parent=CoreLayout
   widget=Border
   properties {
    verticalAlignment=VAlign_Center
    background {
     bIsDynamicallyLoaded=false
     drawAs=Border
     imageType=NoImage
     mirroring=NoMirror
     resourceName=None
     resourceObject=null
     tiling=NoTile
     imageSize=<32 32>
     margin=<ltrb 0 0 0 1>
     outlineSettings {
      bUseBrushTransparency=false
      roundingType=HalfHeightRadius
      width=0
      color {colorUseRule=UseColor_Specified, specifiedColor=<rgba 0 0 0 0>}
      cornerRadii=<0 0 0 0>
     }
     tintColor {colorUseRule=UseColor_Specified, specifiedColor=<rgba 0.1302 0.1302 0.1302 1>}
     uVRegion {bIsValid=false, max=<0 0>, min=<0 0>}
    }
    padding=<ltrb 0 0 0 4>
   }
   slot {class=VerticalBoxSlot, properties {verticalAlignment=VAlign_Top}}
  }
  HorizontalBoxLevelSequencebar {
   widgetClassPath=&HorizontalBox
   parent=Border
   widget=HorizontalBoxLevelSequencebar
   slot {class=BorderSlot, properties {verticalAlignment=VAlign_Center, padding=<ltrb 0 0 0 4>}}
  }
  TextProjectLabel {
   widgetClassPath=&TextBlock
   parent=HorizontalBoxLevelSequencebar
   widget=TextProjectLabel
   properties {
    text=Project:
    toolTipText="Current project. \r\nIs selected in AnimBrowser tool. "
    colorAndOpacity {colorUseRule=UseColor_Specified, specifiedColor=<rgba 0.1545 0.1545 0.1545 1>}
    font {
     bForceMonospaced=false
     bMaterialIsStencil=false
     fontMaterial=null
     fontObject=&Roboto
     letterSpacing=0
     monospacedWidth=1
     size=9
     skewAmount=0
     typefaceFontName=Bold
     outlineSettings {
      bApplyOutlineToDropShadows=false
      bMiteredCorners=false
      bSeparateFillAlpha=false
      outlineMaterial=null
      outlineSize=0
      outlineColor=<rgba 0 0 0 1>
     }
    }
   }
   slot {class=HorizontalBoxSlot, properties {horizontalAlignment=HAlign_Left, verticalAlignment=VAlign_Center}}
  }
  TextProject {
   widgetClassPath=&TextBlock
   bIsVariable=true
   parent=HorizontalBoxLevelSequencebar
   widget=TextProject
   properties {
    text=None
    colorAndOpacity {colorUseRule=UseColor_Specified, specifiedColor=<rgba 0.0194 0.4878 0.0456 1>}
    font {
     bForceMonospaced=false
     bMaterialIsStencil=false
     fontMaterial=null
     fontObject=&Roboto
     letterSpacing=0
     monospacedWidth=1
     size=9
     skewAmount=0
     typefaceFontName=Bold
     outlineSettings {
      bApplyOutlineToDropShadows=false
      bMiteredCorners=false
      bSeparateFillAlpha=false
      outlineMaterial=null
      outlineSize=0
      outlineColor=<rgba 0 0 0 1>
     }
    }
   }
   slot {class=HorizontalBoxSlot, properties {horizontalAlignment=HAlign_Left, verticalAlignment=VAlign_Center, padding=<ltrb 4 0 4 0>}}
  }
  Spacer_6 {
   widgetClassPath=&Spacer
   parent=HorizontalBoxLevelSequencebar
   widget=Spacer_6
   properties {size=<0 0>}
   slot {class=HorizontalBoxSlot, properties {verticalAlignment=VAlign_Top, size {sizeRule=Fill, value=1}}}
  }
  SinglePropertyView_LevelSequence {
   widgetClassPath=&SinglePropertyView
   parent=HorizontalBoxLevelSequencebar
   widget=SinglePropertyView_LevelSequence
   properties {propertyName=Current Level Sequence}
   slot {class=HorizontalBoxSlot, properties {horizontalAlignment=HAlign_Center, verticalAlignment=VAlign_Center}}
  }
  Spacer_7 {
   widgetClassPath=&Spacer
   parent=HorizontalBoxLevelSequencebar
   widget=Spacer_7
   properties {size=<1 0>}
   slot {class=HorizontalBoxSlot, properties {verticalAlignment=VAlign_Top, size {sizeRule=Fill, value=1}}}
  }
  Button_? {
   widgetClassPath=&EditorUtilityButton
   parent=HorizontalBoxLevelSequencebar
   widget=Button_?
   properties {
    toolTipText=Open Confuelnce Page
    widgetStyle {
     clickedSlateSound {resourceObject=null}
     disabled {
      bIsDynamicallyLoaded=false
      drawAs=RoundedBox
      imageType=NoImage
      mirroring=NoMirror
      resourceName=None
      resourceObject=null
      tiling=NoTile
      imageSize=<0 0>
      margin=<ltrb 0 0 0 0>
      outlineSettings {
       bUseBrushTransparency=false
       roundingType=FixedRadius
       width=1
       color {colorUseRule=UseColor_ColorTable, specifiedColor=<rgba 0 0 0 0>}
       cornerRadii=<4 4 4 4>
      }
      tintColor {colorUseRule=UseColor_Specified, specifiedColor=<rgba 0.0395 0.0395 0.0395 1>}
      uVRegion {bIsValid=false, max=<0 0>, min=<0 0>}
     }
     disabledForeground {colorUseRule=UseColor_ColorTable, specifiedColor=<rgba 0 0 0 0>}
     hovered {
      bIsDynamicallyLoaded=false
      drawAs=RoundedBox
      imageType=NoImage
      mirroring=NoMirror
      resourceName=None
      resourceObject=null
      tiling=NoTile
      imageSize=<32 32>
      margin=<ltrb 0 0 0 0>
      outlineSettings {
       bUseBrushTransparency=false
       roundingType=FixedRadius
       width=1
       color {colorUseRule=UseColor_ColorTable, specifiedColor=<rgba 0 0 0 0>}
       cornerRadii=<4 4 4 4>
      }
      tintColor {colorUseRule=UseColor_Specified, specifiedColor=<rgba 0.0953 0.0953 0.0953 1>}
      uVRegion {bIsValid=false, max=<0 0>, min=<0 0>}
     }
     hoveredForeground {colorUseRule=UseColor_ColorTable, specifiedColor=<rgba 0 0 0 0>}
     hoveredSlateSound {resourceObject=null}
     normal {
      bIsDynamicallyLoaded=false
      drawAs=RoundedBox
      imageType=NoImage
      mirroring=NoMirror
      resourceName=None
      resourceObject=null
      tiling=NoTile
      imageSize=<32 32>
      margin=<ltrb 0 0 0 0>
      outlineSettings {
       bUseBrushTransparency=false
       roundingType=FixedRadius
       width=1
       color {colorUseRule=UseColor_ColorTable, specifiedColor=<rgba 0 0 0 0>}
       cornerRadii=<4 4 4 4>
      }
      tintColor {colorUseRule=UseColor_Specified, specifiedColor=<rgba 0.0395 0.0395 0.0395 1>}
      uVRegion {bIsValid=false, max=<0 0>, min=<0 0>}
     }
     normalForeground {colorUseRule=UseColor_ColorTable, specifiedColor=<rgba 0 0 0 0>}
     normalPadding=<ltrb 8 2 6 2>
     pressed {
      bIsDynamicallyLoaded=false
      drawAs=RoundedBox
      imageType=NoImage
      mirroring=NoMirror
      resourceName=None
      resourceObject=null
      tiling=NoTile
      imageSize=<32 32>
      margin=<ltrb 0 0 0 0>
      outlineSettings {
       bUseBrushTransparency=false
       roundingType=FixedRadius
       width=1
       color {colorUseRule=UseColor_ColorTable, specifiedColor=<rgba 0 0 0 0>}
       cornerRadii=<4 4 4 4>
      }
      tintColor {colorUseRule=UseColor_Specified, specifiedColor=<rgba 0.0284 0.0284 0.0284 1>}
      uVRegion {bIsValid=false, max=<0 0>, min=<0 0>}
     }
     pressedForeground {colorUseRule=UseColor_ColorTable, specifiedColor=<rgba 0 0 0 0>}
     pressedPadding=<ltrb 8 2.5 6 2.5>
     pressedSlateSound {resourceObject=null}
    }
   }
   slot {class=HorizontalBoxSlot, properties {horizontalAlignment=HAlign_Right, verticalAlignment=VAlign_Center, padding=<ltrb 4 0 0 0>}}
  }
  TextBlock_14 {
   widgetClassPath=&TextBlock
   parent=Button_?
   widget=TextBlock_14
   properties {
    text=?
    colorAndOpacity {colorUseRule=UseColor_Specified, specifiedColor=<rgba 0 1 0 1>}
    font {
     bForceMonospaced=false
     bMaterialIsStencil=false
     fontMaterial=null
     fontObject=&Roboto
     letterSpacing=0
     monospacedWidth=1
     size=12
     skewAmount=0
     typefaceFontName=Bold
     outlineSettings {
      bApplyOutlineToDropShadows=false
      bMiteredCorners=false
      bSeparateFillAlpha=false
      outlineMaterial=null
      outlineSize=0
      outlineColor=<rgba 0 0 0 1>
     }
    }
   }
   slot {class=ButtonSlot, properties {horizontalAlignment=HAlign_Left, padding=<ltrb 0 0 0 0>}}
  }
  VerticalBox_0 {
   widgetClassPath=&VerticalBox
   parent=CoreLayout
   widget=VerticalBox_0
   slot {class=VerticalBoxSlot, properties {horizontalAlignment=HAlign_Right, verticalAlignment=VAlign_Top, size {sizeRule=Fill, value=1}}}
  }
  HorizontalBox_Mode {
   widgetClassPath=&HorizontalBox
   parent=VerticalBox_0
   widget=HorizontalBox_Mode
   slot {class=VerticalBoxSlot, properties {horizontalAlignment=HAlign_Right}}
  }
  TextBlock_3 {
   widgetClassPath=&TextBlock
   parent=HorizontalBox_Mode
   widget=TextBlock_3
   properties {
    text=Mode:
    toolTipText="Modes Selection: \r\n\r\nStadart : lists all linked anim sequences\r\nAim Offset: mode designed to make BlendSpace asset made of single frame poses. "
    font {
     bForceMonospaced=false
     bMaterialIsStencil=false
     fontMaterial=null
     fontObject=&Roboto
     letterSpacing=0
     monospacedWidth=1
     size=9
     skewAmount=0
     typefaceFontName=Bold
     outlineSettings {
      bApplyOutlineToDropShadows=false
      bMiteredCorners=false
      bSeparateFillAlpha=false
      outlineMaterial=null
      outlineSize=0
      outlineColor=<rgba 0 0 0 1>
     }
    }
   }
   slot {class=HorizontalBoxSlot, properties {verticalAlignment=VAlign_Center}}
  }
  ComboBox_ToolMode {
   widgetClassPath=/Script/Blutility.EditorUtilityComboBoxString
   bIsVariable=true
   parent=HorizontalBox_Mode
   widget=ComboBox_ToolMode
   properties {
    selectedOption=Standard
    toolTipText="Modes Selection: \r\n\r\nStadart : lists all linked anim sequences\r\n"
    contentPadding=<ltrb 4 2 10 2>
    defaultOptions [Standard]
    font {
     bForceMonospaced=false
     bMaterialIsStencil=false
     fontMaterial=null
     fontObject=&Roboto
     letterSpacing=0
     monospacedWidth=1
     size=9
     skewAmount=0
     typefaceFontName=Bold
     outlineSettings {
      bApplyOutlineToDropShadows=false
      bMiteredCorners=false
      bSeparateFillAlpha=false
      outlineMaterial=null
      outlineSize=0
      outlineColor=<rgba 0 0 0 1>
     }
    }
    foregroundColor {colorUseRule=UseColor_Specified, specifiedColor=<rgba 0.5271 0.5271 0.5271 1>}
   }
   slot {class=HorizontalBoxSlot, properties {horizontalAlignment=HAlign_Right, verticalAlignment=VAlign_Center}}
  }
  HorizontalBox_Mode_1 {
   widgetClassPath=&HorizontalBox
   parent=VerticalBox_0
   widget=HorizontalBox_Mode_1
   properties {toolTipText="Will update when you realese it, not on drag."}
   slot {class=VerticalBoxSlot}
  }
  TextBlock_6 {
   widgetClassPath=&TextBlock
   parent=HorizontalBox_Mode_1
   widget=TextBlock_6
   properties {
    text=UI:
    font {
     bForceMonospaced=false
     bMaterialIsStencil=false
     fontMaterial=null
     fontObject=&Roboto
     letterSpacing=0
     monospacedWidth=1
     size=9
     skewAmount=0
     typefaceFontName=Bold
     outlineSettings {
      bApplyOutlineToDropShadows=false
      bMiteredCorners=false
      bSeparateFillAlpha=false
      outlineMaterial=null
      outlineSize=0
      outlineColor=<rgba 0 0 0 1>
     }
    }
   }
   slot {class=HorizontalBoxSlot, properties {horizontalAlignment=HAlign_Left, verticalAlignment=VAlign_Center}}
  }
  EditorUtilitySliderUIScale {
   widgetClassPath=/Script/Blutility.EditorUtilitySlider
   bIsVariable=true
   parent=HorizontalBox_Mode_1
   widget=EditorUtilitySliderUIScale
   properties {minValue=0.6, value=1}
   slot {class=HorizontalBoxSlot, properties {padding=<ltrb 10 0 0 0>, size {sizeRule=Fill, value=1}}}
  }
  VerticalBoxSequencerTracks {
   widgetClassPath=&VerticalBox
   parent=CoreLayout
   widget=VerticalBoxSequencerTracks
   slot {class=VerticalBoxSlot, properties {horizontalAlignment=HAlign_Center, verticalAlignment=VAlign_Top, padding=<ltrb 0 4 0 10>}}
  }
  Border_1 {
   widgetClassPath=&Border
   parent=VerticalBoxSequencerTracks
   widget=Border_1
   properties {
    horizontalAlignment=HAlign_Center
    verticalAlignment=VAlign_Center
    background {
     bIsDynamicallyLoaded=false
     drawAs=Border
     imageType=NoImage
     mirroring=NoMirror
     resourceName=None
     resourceObject=null
     tiling=NoTile
     imageSize=<32 32>
     margin=<ltrb 0 0 0 1>
     outlineSettings {
      bUseBrushTransparency=false
      roundingType=HalfHeightRadius
      width=0
      color {colorUseRule=UseColor_Specified, specifiedColor=<rgba 0 0 0 0>}
      cornerRadii=<0 0 0 0>
     }
     tintColor {colorUseRule=UseColor_Specified, specifiedColor=<rgba 0.1302 0.1302 0.1302 1>}
     uVRegion {bIsValid=false, max=<0 0>, min=<0 0>}
    }
    padding=<ltrb 0 0 0 0>
   }
   slot {class=VerticalBoxSlot, properties {padding=<ltrb 0 0 0 2>}}
  }
  TracksText {
   widgetClassPath=&TextBlock
   bIsVariable=true
   parent=Border_1
   widget=TracksText
   properties {
    text=Linked Sequencer Tracks:
    toolTipText=List of Sequencer Tracks exposed for linking Animation Sequences
    colorAndOpacity {colorUseRule=UseColor_Specified, specifiedColor=<rgba 0.1545 0.1545 0.1545 1>}
    font {
     bForceMonospaced=false
     bMaterialIsStencil=false
     fontMaterial=null
     fontObject=&Roboto
     letterSpacing=0
     monospacedWidth=1
     size=9
     skewAmount=0
     typefaceFontName=Bold
     outlineSettings {
      bApplyOutlineToDropShadows=false
      bMiteredCorners=false
      bSeparateFillAlpha=false
      outlineMaterial=null
      outlineSize=0
      outlineColor=<rgba 0 0 0 1>
     }
    }
   }
   slot {class=BorderSlot, properties {horizontalAlignment=HAlign_Center, verticalAlignment=VAlign_Center, padding=<ltrb 0 0 0 0>}}
  }
  Layout_Tracks {
   widgetClassPath=&VerticalBox
   bIsVariable=true
   parent=VerticalBoxSequencerTracks
   widget=Layout_Tracks
   slot {class=VerticalBoxSlot}
  }
  HorizontalBoxAddTrack {
   widgetClassPath=&HorizontalBox
   bIsVariable=true
   parent=VerticalBoxSequencerTracks
   widget=HorizontalBoxAddTrack
   properties {toolTipText="Select Track in the list and press + to add it to the tool. \r\n\r\nTool is looking for bindings that have tracks of either type:\r\n\r\n[MovieSceneSkeletalAnimationTrack]\r\n[MovieSceneControlRigParameterTrack]\r\n"}
   slot {class=VerticalBoxSlot, properties {horizontalAlignment=HAlign_Center, verticalAlignment=VAlign_Center}}
  }
  Add_Track_Text {
   widgetClassPath=&TextBlock
   bIsVariable=true
   parent=HorizontalBoxAddTrack
   widget=Add_Track_Text
   properties {
    text=Add Track:
    colorAndOpacity {colorUseRule=UseColor_Specified, specifiedColor=<rgba 0.1545 0.1545 0.1545 1>}
    font {
     bForceMonospaced=false
     bMaterialIsStencil=false
     fontMaterial=null
     fontObject=&Roboto
     letterSpacing=0
     monospacedWidth=1
     size=9
     skewAmount=0
     typefaceFontName=Bold
     outlineSettings {
      bApplyOutlineToDropShadows=false
      bMiteredCorners=false
      bSeparateFillAlpha=false
      outlineMaterial=null
      outlineSize=0
      outlineColor=<rgba 0 0 0 1>
     }
    }
   }
   slot {class=HorizontalBoxSlot, properties {horizontalAlignment=HAlign_Left, verticalAlignment=VAlign_Center, padding=<ltrb 0 0 12 0>}}
  }
  ComboBox_AddTrack {
   widgetClassPath=&EditorUtilityComboBoxString
   bIsVariable=true
   parent=HorizontalBoxAddTrack
   widget=ComboBox_AddTrack
   properties {
    contentPadding=<ltrb 0 0 0 0>
    font {
     bForceMonospaced=false
     bMaterialIsStencil=false
     fontMaterial=null
     fontObject=&Roboto
     letterSpacing=0
     monospacedWidth=1
     size=9
     skewAmount=0
     typefaceFontName=Bold
     outlineSettings {
      bApplyOutlineToDropShadows=false
      bMiteredCorners=false
      bSeparateFillAlpha=false
      outlineMaterial=null
      outlineSize=0
      outlineColor=<rgba 0 0 0 1>
     }
    }
    foregroundColor {colorUseRule=UseColor_Specified, specifiedColor=<rgba 0.5271 0.5271 0.5271 1>}
   }
   slot {class=HorizontalBoxSlot}
  }
  Button_AddTrack {
   widgetClassPath=&EditorUtilityButton
   parent=HorizontalBoxAddTrack
   widget=Button_AddTrack
   slot {class=HorizontalBoxSlot, properties {horizontalAlignment=HAlign_Center, padding=<ltrb 6 0 0 0>}}
  }
  AssetImagePlusButton {
   widgetClassPath=&Image
   parent=Button_AddTrack
   widget=AssetImagePlusButton
   properties {brush {
     bIsDynamicallyLoaded=false
     drawAs=Image
     imageType=FullColor
     mirroring=NoMirror
     resourceName=None
     resourceObject=/AnimInEdTools/AnimationBrowser/Icons/Icon_LevelSequence_add
     tiling=NoTile
     imageSize=<14 14>
     margin=<ltrb 0 0 0 0>
     outlineSettings {
      bUseBrushTransparency=false
      roundingType=HalfHeightRadius
      width=0
      color {colorUseRule=UseColor_Specified, specifiedColor=<rgba 0 0 0 0>}
      cornerRadii=<0 0 0 0>
     }
     tintColor {colorUseRule=UseColor_Specified, specifiedColor=<rgba 0.2108 1 0 1>}
     uVRegion {bIsValid=false, max=<0 0>, min=<0 0>}
    }}
   slot {class=ButtonSlot, properties {padding=<ltrb 0 0 0 0>}}
  }
  VerticalBoxCreateBookmarks {
   widgetClassPath=&VerticalBox
   parent=CoreLayout
   widget=VerticalBoxCreateBookmarks
   properties {toolTipText="Exportable segments can be created by manually placing bookmarks named according to convention below. \r\n\r\nOr \r\n\r\nYou can use this widget to quickly create correct bookmarks for your desired export range.\r\n\r\nInput name of your desired bookmark in the left field and the frme ## in the right field.  And press + button. \r\n\r\nCSM Bookmark logic works as follows:\r\n\r\nSegments:\r\n\r\n- BMs that start with ##and have text will be considered as segement start.\r\n\r\n- [1] BMs that have only ## and no text will be considered as end of segment. \r\n\r\n- [2] BMs that have same text and start with > will be considered as explicit end (will work if there are other BMs inbetween). \r\n\r\n- if neither [1] or [2] are not found the next BM with ## will be considered as segment end\r\n\r\n- if no BMs with ## found - end of timepline will be considered as end\r\n\r\nSingle Frames: \r\n\r\n- BMs that start with @ and have text will be considered as ONE frame long export segment"}
   slot {class=VerticalBoxSlot, properties {horizontalAlignment=HAlign_Center, verticalAlignment=VAlign_Top}}
  }
  Border_2 {
   widgetClassPath=&Border
   parent=VerticalBoxCreateBookmarks
   widget=Border_2
   properties {
    horizontalAlignment=HAlign_Center
    verticalAlignment=VAlign_Center
    background {
     bIsDynamicallyLoaded=false
     drawAs=Border
     imageType=NoImage
     mirroring=NoMirror
     resourceName=None
     resourceObject=null
     tiling=NoTile
     imageSize=<32 32>
     margin=<ltrb 0 0 0 1>
     outlineSettings {
      bUseBrushTransparency=false
      roundingType=HalfHeightRadius
      width=0
      color {colorUseRule=UseColor_Specified, specifiedColor=<rgba 0 0 0 0>}
      cornerRadii=<0 0 0 0>
     }
     tintColor {colorUseRule=UseColor_Specified, specifiedColor=<rgba 0.1302 0.1302 0.1302 1>}
     uVRegion {bIsValid=false, max=<0 0>, min=<0 0>}
    }
    padding=<ltrb 0 0 12 0>
   }
   slot {class=VerticalBoxSlot, properties {padding=<ltrb 0 0 0 2>}}
  }
  TracksText_1 {
   widgetClassPath=&TextBlock
   bIsVariable=true
   parent=Border_2
   widget=TracksText_1
   properties {
    text="Bookmark(s) Creator:"
    colorAndOpacity {colorUseRule=UseColor_Specified, specifiedColor=<rgba 0.1545 0.1545 0.1545 1>}
    font {
     bForceMonospaced=false
     bMaterialIsStencil=false
     fontMaterial=null
     fontObject=&Roboto
     letterSpacing=0
     monospacedWidth=1
     size=9
     skewAmount=0
     typefaceFontName=Bold
     outlineSettings {
      bApplyOutlineToDropShadows=false
      bMiteredCorners=false
      bSeparateFillAlpha=false
      outlineMaterial=null
      outlineSize=0
      outlineColor=<rgba 0 0 0 1>
     }
    }
   }
   slot {class=BorderSlot, properties {horizontalAlignment=HAlign_Center, verticalAlignment=VAlign_Center, padding=<ltrb 0 0 12 0>}}
  }
  HorizontalBoxAddBookmark {
   widgetClassPath=&HorizontalBox
   bIsVariable=true
   parent=VerticalBoxCreateBookmarks
   widget=HorizontalBoxAddBookmark
   slot {class=VerticalBoxSlot}
  }
  GridPanel_195 {widgetClassPath=&GridPanel, parent=HorizontalBoxAddBookmark, widget=GridPanel_195, slot {class=HorizontalBoxSlot}}
  Add_Track_Text_1 {
   widgetClassPath=&TextBlock
   bIsVariable=true
   parent=GridPanel_195
   widget=Add_Track_Text_1
   properties {
    text=Start Bookmark:
    toolTipText=Opening bookmark of the Linked segment.
    colorAndOpacity {colorUseRule=UseColor_Specified, specifiedColor=<rgba 0.1545 0.1545 0.1545 1>}
    font {
     bForceMonospaced=false
     bMaterialIsStencil=false
     fontMaterial=null
     fontObject=&Roboto
     letterSpacing=0
     monospacedWidth=1
     size=9
     skewAmount=0
     typefaceFontName=Bold
     outlineSettings {
      bApplyOutlineToDropShadows=false
      bMiteredCorners=false
      bSeparateFillAlpha=false
      outlineMaterial=null
      outlineSize=0
      outlineColor=<rgba 0 0 0 1>
     }
    }
   }
   slot {class=GridSlot, properties {horizontalAlignment=HAlign_Left, verticalAlignment=VAlign_Center, padding=<ltrb 0 0 12 0>}}
  }
  EditorUtilityEditableTextBoxStartBookmark {
   widgetClassPath=&EditorUtilityEditableTextBox
   bIsVariable=true
   parent=GridPanel_195
   widget=EditorUtilityEditableTextBoxStartBookmark
   properties {
    hintText=Segment Name
    toolTipText="Name of the first bookmark (that marks the begining of the export segment)"
    widgetStyle {
     backgroundColor {colorUseRule=UseColor_Specified, specifiedColor=<rgba 1 1 1 1>}
     backgroundImageFocused {
      bIsDynamicallyLoaded=false
      drawAs=RoundedBox
      imageType=NoImage
      mirroring=NoMirror
      resourceName=None
      resourceObject=null
      tiling=NoTile
      imageSize=<0 0>
      margin=<ltrb 0 0 0 0>
      outlineSettings {
       bUseBrushTransparency=false
       roundingType=FixedRadius
       width=1
       color {colorUseRule=UseColor_ColorTable, specifiedColor=<rgba 0 0 0 0>}
       cornerRadii=<4 4 4 4>
      }
      tintColor {colorUseRule=UseColor_Specified, specifiedColor=<rgba 0.0048 0.0048 0.0048 1>}
      uVRegion {bIsValid=false, max=<0 0>, min=<0 0>}
     }
     backgroundImageHovered {
      bIsDynamicallyLoaded=false
      drawAs=RoundedBox
      imageType=NoImage
      mirroring=NoMirror
      resourceName=None
      resourceObject=null
      tiling=NoTile
      imageSize=<0 0>
      margin=<ltrb 0 0 0 0>
      outlineSettings {
       bUseBrushTransparency=false
       roundingType=FixedRadius
       width=1
       color {colorUseRule=UseColor_ColorTable, specifiedColor=<rgba 0 0 0 0>}
       cornerRadii=<4 4 4 4>
      }
      tintColor {colorUseRule=UseColor_Specified, specifiedColor=<rgba 0.0048 0.0048 0.0048 1>}
      uVRegion {bIsValid=false, max=<0 0>, min=<0 0>}
     }
     backgroundImageNormal {
      bIsDynamicallyLoaded=false
      drawAs=RoundedBox
      imageType=NoImage
      mirroring=NoMirror
      resourceName=None
      resourceObject=null
      tiling=NoTile
      imageSize=<0 0>
      margin=<ltrb 0 0 0 0>
      outlineSettings {
       bUseBrushTransparency=false
       roundingType=FixedRadius
       width=1
       color {colorUseRule=UseColor_ColorTable, specifiedColor=<rgba 0 0 0 0>}
       cornerRadii=<4 4 4 4>
      }
      tintColor {colorUseRule=UseColor_Specified, specifiedColor=<rgba 0.0048 0.0048 0.0048 1>}
      uVRegion {bIsValid=false, max=<0 0>, min=<0 0>}
     }
     backgroundImageReadOnly {
      bIsDynamicallyLoaded=false
      drawAs=RoundedBox
      imageType=NoImage
      mirroring=NoMirror
      resourceName=None
      resourceObject=null
      tiling=NoTile
      imageSize=<0 0>
      margin=<ltrb 0 0 0 0>
      outlineSettings {
       bUseBrushTransparency=false
       roundingType=FixedRadius
       width=1
       color {colorUseRule=UseColor_ColorTable, specifiedColor=<rgba 0 0 0 0>}
       cornerRadii=<4 4 4 4>
      }
      tintColor {colorUseRule=UseColor_Specified, specifiedColor=<rgba 0.0284 0.0284 0.0284 1>}
      uVRegion {bIsValid=false, max=<0 0>, min=<0 0>}
     }
     focusedForegroundColor {colorUseRule=UseColor_Specified, specifiedColor=<rgba 1 1 1 1>}
     foregroundColor {colorUseRule=UseColor_Specified, specifiedColor=<rgba 0.5271 0.5271 0.5271 1>}
     hScrollBarPadding=<ltrb 0 0 0 0>
     padding=<ltrb 8 2 2 2>
     readOnlyForegroundColor {colorUseRule=UseColor_Foreground, specifiedColor=<rgba 1 0 1 1>}
     scrollBarStyle {
      thickness=8
      draggedThumbImage {
       bIsDynamicallyLoaded=false
       drawAs=RoundedBox
       imageType=NoImage
       mirroring=NoMirror
       resourceName=None
       resourceObject=null
       tiling=NoTile
       imageSize=<0 0>
       margin=<ltrb 0 0 0 0>
       outlineSettings {
        bUseBrushTransparency=false
        roundingType=FixedRadius
        width=0
        color {colorUseRule=UseColor_Specified, specifiedColor=<rgba 0 0 0 0>}
        cornerRadii=<4 4 4 4>
       }
       tintColor {colorUseRule=UseColor_Specified, specifiedColor=<rgba 0.2159 0.2159 0.2159 1>}
       uVRegion {bIsValid=false, max=<0 0>, min=<0 0>}
      }
      horizontalBackgroundImage {
       bIsDynamicallyLoaded=false
       drawAs=NoDrawType
       imageType=NoImage
       mirroring=NoMirror
       resourceName=None
       resourceObject=null
       tiling=NoTile
       imageSize=<0 0>
       margin=<ltrb 0 0 0 0>
       outlineSettings {
        bUseBrushTransparency=false
        roundingType=HalfHeightRadius
        width=0
        color {colorUseRule=UseColor_Specified, specifiedColor=<rgba 0 0 0 0>}
        cornerRadii=<0 0 0 0>
       }
       tintColor {colorUseRule=UseColor_Specified, specifiedColor=<rgba 1 1 1 1>}
       uVRegion {bIsValid=false, max=<0 0>, min=<0 0>}
      }
      horizontalBottomSlotImage {
       bIsDynamicallyLoaded=false
       drawAs=NoDrawType
       imageType=NoImage
       mirroring=NoMirror
       resourceName=None
       resourceObject=null
       tiling=NoTile
       imageSize=<0 0>
       margin=<ltrb 0 0 0 0>
       outlineSettings {
        bUseBrushTransparency=false
        roundingType=HalfHeightRadius
        width=0
        color {colorUseRule=UseColor_Specified, specifiedColor=<rgba 0 0 0 0>}
        cornerRadii=<0 0 0 0>
       }
       tintColor {colorUseRule=UseColor_Specified, specifiedColor=<rgba 1 1 1 1>}
       uVRegion {bIsValid=false, max=<0 0>, min=<0 0>}
      }
      horizontalTopSlotImage {
       bIsDynamicallyLoaded=false
       drawAs=NoDrawType
       imageType=NoImage
       mirroring=NoMirror
       resourceName=None
       resourceObject=null
       tiling=NoTile
       imageSize=<0 0>
       margin=<ltrb 0 0 0 0>
       outlineSettings {
        bUseBrushTransparency=false
        roundingType=HalfHeightRadius
        width=0
        color {colorUseRule=UseColor_Specified, specifiedColor=<rgba 0 0 0 0>}
        cornerRadii=<0 0 0 0>
       }
       tintColor {colorUseRule=UseColor_Specified, specifiedColor=<rgba 1 1 1 1>}
       uVRegion {bIsValid=false, max=<0 0>, min=<0 0>}
      }
      hoveredThumbImage {
       bIsDynamicallyLoaded=false
       drawAs=RoundedBox
       imageType=NoImage
       mirroring=NoMirror
       resourceName=None
       resourceObject=null
       tiling=NoTile
       imageSize=<0 0>
       margin=<ltrb 0 0 0 0>
       outlineSettings {
        bUseBrushTransparency=false
        roundingType=FixedRadius
        width=0
        color {colorUseRule=UseColor_Specified, specifiedColor=<rgba 0 0 0 0>}
        cornerRadii=<4 4 4 4>
       }
       tintColor {colorUseRule=UseColor_Specified, specifiedColor=<rgba 0.2159 0.2159 0.2159 1>}
       uVRegion {bIsValid=false, max=<0 0>, min=<0 0>}
      }
      normalThumbImage {
       bIsDynamicallyLoaded=false
       drawAs=RoundedBox
       imageType=NoImage
       mirroring=NoMirror
       resourceName=None
       resourceObject=null
       tiling=NoTile
       imageSize=<0 0>
       margin=<ltrb 0 0 0 0>
       outlineSettings {
        bUseBrushTransparency=false
        roundingType=FixedRadius
        width=0
        color {colorUseRule=UseColor_Specified, specifiedColor=<rgba 0 0 0 0>}
        cornerRadii=<4 4 4 4>
       }
       tintColor {colorUseRule=UseColor_Specified, specifiedColor=<rgba 0.0953 0.0953 0.0953 1>}
       uVRegion {bIsValid=false, max=<0 0>, min=<0 0>}
      }
      verticalBackgroundImage {
       bIsDynamicallyLoaded=false
       drawAs=NoDrawType
       imageType=NoImage
       mirroring=NoMirror
       resourceName=None
       resourceObject=null
       tiling=NoTile
       imageSize=<0 0>
       margin=<ltrb 0 0 0 0>
       outlineSettings {
        bUseBrushTransparency=false
        roundingType=HalfHeightRadius
        width=0
        color {colorUseRule=UseColor_Specified, specifiedColor=<rgba 0 0 0 0>}
        cornerRadii=<0 0 0 0>
       }
       tintColor {colorUseRule=UseColor_Specified, specifiedColor=<rgba 1 1 1 1>}
       uVRegion {bIsValid=false, max=<0 0>, min=<0 0>}
      }
      verticalBottomSlotImage {
       bIsDynamicallyLoaded=false
       drawAs=NoDrawType
       imageType=NoImage
       mirroring=NoMirror
       resourceName=None
       resourceObject=null
       tiling=NoTile
       imageSize=<0 0>
       margin=<ltrb 0 0 0 0>
       outlineSettings {
        bUseBrushTransparency=false
        roundingType=HalfHeightRadius
        width=0
        color {colorUseRule=UseColor_Specified, specifiedColor=<rgba 0 0 0 0>}
        cornerRadii=<0 0 0 0>
       }
       tintColor {colorUseRule=UseColor_Specified, specifiedColor=<rgba 1 1 1 1>}
       uVRegion {bIsValid=false, max=<0 0>, min=<0 0>}
      }
      verticalTopSlotImage {
       bIsDynamicallyLoaded=false
       drawAs=NoDrawType
       imageType=NoImage
       mirroring=NoMirror
       resourceName=None
       resourceObject=null
       tiling=NoTile
       imageSize=<0 0>
       margin=<ltrb 0 0 0 0>
       outlineSettings {
        bUseBrushTransparency=false
        roundingType=HalfHeightRadius
        width=0
        color {colorUseRule=UseColor_Specified, specifiedColor=<rgba 0 0 0 0>}
        cornerRadii=<0 0 0 0>
       }
       tintColor {colorUseRule=UseColor_Specified, specifiedColor=<rgba 1 1 1 1>}
       uVRegion {bIsValid=false, max=<0 0>, min=<0 0>}
      }
     }
     textStyle {
      overflowPolicy=Clip
      transformPolicy=None
      colorAndOpacity {colorUseRule=UseColor_Foreground, specifiedColor=<rgba 1 0 1 1>}
      font {
       bForceMonospaced=false
       bMaterialIsStencil=false
       fontMaterial=null
       fontObject=null
       letterSpacing=0
       monospacedWidth=1
       size=10
       skewAmount=0
       typefaceFontName=Regular
       outlineSettings {
        bApplyOutlineToDropShadows=false
        bMiteredCorners=false
        bSeparateFillAlpha=false
        outlineMaterial=null
        outlineSize=0
        outlineColor=<rgba 0 0 0 1>
       }
      }
      highlightColor {colorUseRule=UseColor_ColorTable, specifiedColor=<rgba 0 0 0 0>}
      highlightShape {
       bIsDynamicallyLoaded=false
       drawAs=Image
       imageType=NoImage
       mirroring=NoMirror
       resourceName=None
       resourceObject=null
       tiling=NoTile
       imageSize=<0 0>
       margin=<ltrb 0 0 0 0>
       outlineSettings {
        bUseBrushTransparency=false
        roundingType=HalfHeightRadius
        width=0
        color {colorUseRule=UseColor_Specified, specifiedColor=<rgba 0 0 0 0>}
        cornerRadii=<0 0 0 0>
       }
       tintColor {colorUseRule=UseColor_ColorTable, specifiedColor=<rgba 0 0 0 0>}
       uVRegion {bIsValid=false, max=<0 0>, min=<0 0>}
      }
      selectedBackgroundColor {colorUseRule=UseColor_ColorTable, specifiedColor=<rgba 0 0 0 0>}
      shadowColorAndOpacity=<rgba 0 0 0 1>
      shadowOffset=<0 0>
      strikeBrush {
       bIsDynamicallyLoaded=false
       drawAs=Image
       imageType=NoImage
       mirroring=NoMirror
       resourceName=None
       resourceObject=null
       tiling=NoTile
       imageSize=<32 32>
       margin=<ltrb 0 0 0 0>
       outlineSettings {
        bUseBrushTransparency=false
        roundingType=HalfHeightRadius
        width=0
        color {colorUseRule=UseColor_Specified, specifiedColor=<rgba 0 0 0 0>}
        cornerRadii=<0 0 0 0>
       }
       tintColor {colorUseRule=UseColor_Specified, specifiedColor=<rgba 1 1 1 1>}
       uVRegion {bIsValid=false, max=<0 0>, min=<0 0>}
      }
      underlineBrush {
       bIsDynamicallyLoaded=false
       drawAs=Image
       imageType=NoImage
       mirroring=NoMirror
       resourceName=None
       resourceObject=null
       tiling=NoTile
       imageSize=<32 32>
       margin=<ltrb 0 0 0 0>
       outlineSettings {
        bUseBrushTransparency=false
        roundingType=HalfHeightRadius
        width=0
        color {colorUseRule=UseColor_Specified, specifiedColor=<rgba 0 0 0 0>}
        cornerRadii=<0 0 0 0>
       }
       tintColor {colorUseRule=UseColor_Specified, specifiedColor=<rgba 1 1 1 1>}
       uVRegion {bIsValid=false, max=<0 0>, min=<0 0>}
      }
     }
     vScrollBarPadding=<ltrb 0 0 0 0>
    }
   }
   slot {class=GridSlot, properties {column=2, verticalAlignment=VAlign_Center}}
  }
  EditorUtilityEditableTextBoxStartFrame {
   widgetClassPath=&EditorUtilityEditableTextBox
   bIsVariable=true
   parent=GridPanel_195
   widget=EditorUtilityEditableTextBoxStartFrame
   properties {
    hintText="  ##  "
    toolTipText=Frame number to place this bookmark
    widgetStyle {
     backgroundColor {colorUseRule=UseColor_Specified, specifiedColor=<rgba 1 1 1 1>}
     backgroundImageFocused {
      bIsDynamicallyLoaded=false
      drawAs=RoundedBox
      imageType=NoImage
      mirroring=NoMirror
      resourceName=None
      resourceObject=null
      tiling=NoTile
      imageSize=<0 0>
      margin=<ltrb 0 0 0 0>
      outlineSettings {
       bUseBrushTransparency=false
       roundingType=FixedRadius
       width=1
       color {colorUseRule=UseColor_ColorTable, specifiedColor=<rgba 0 0 0 0>}
       cornerRadii=<4 4 4 4>
      }
      tintColor {colorUseRule=UseColor_Specified, specifiedColor=<rgba 0.0048 0.0048 0.0048 1>}
      uVRegion {bIsValid=false, max=<0 0>, min=<0 0>}
     }
     backgroundImageHovered {
      bIsDynamicallyLoaded=false
      drawAs=RoundedBox
      imageType=NoImage
      mirroring=NoMirror
      resourceName=None
      resourceObject=null
      tiling=NoTile
      imageSize=<0 0>
      margin=<ltrb 0 0 0 0>
      outlineSettings {
       bUseBrushTransparency=false
       roundingType=FixedRadius
       width=1
       color {colorUseRule=UseColor_ColorTable, specifiedColor=<rgba 0 0 0 0>}
       cornerRadii=<4 4 4 4>
      }
      tintColor {colorUseRule=UseColor_Specified, specifiedColor=<rgba 0.0048 0.0048 0.0048 1>}
      uVRegion {bIsValid=false, max=<0 0>, min=<0 0>}
     }
     backgroundImageNormal {
      bIsDynamicallyLoaded=false
      drawAs=RoundedBox
      imageType=NoImage
      mirroring=NoMirror
      resourceName=None
      resourceObject=null
      tiling=NoTile
      imageSize=<0 0>
      margin=<ltrb 0 0 0 0>
      outlineSettings {
       bUseBrushTransparency=false
       roundingType=FixedRadius
       width=1
       color {colorUseRule=UseColor_ColorTable, specifiedColor=<rgba 0 0 0 0>}
       cornerRadii=<4 4 4 4>
      }
      tintColor {colorUseRule=UseColor_Specified, specifiedColor=<rgba 0.0048 0.0048 0.0048 1>}
      uVRegion {bIsValid=false, max=<0 0>, min=<0 0>}
     }
     backgroundImageReadOnly {
      bIsDynamicallyLoaded=false
      drawAs=RoundedBox
      imageType=NoImage
      mirroring=NoMirror
      resourceName=None
      resourceObject=null
      tiling=NoTile
      imageSize=<0 0>
      margin=<ltrb 0 0 0 0>
      outlineSettings {
       bUseBrushTransparency=false
       roundingType=FixedRadius
       width=1
       color {colorUseRule=UseColor_ColorTable, specifiedColor=<rgba 0 0 0 0>}
       cornerRadii=<4 4 4 4>
      }
      tintColor {colorUseRule=UseColor_Specified, specifiedColor=<rgba 0.0284 0.0284 0.0284 1>}
      uVRegion {bIsValid=false, max=<0 0>, min=<0 0>}
     }
     focusedForegroundColor {colorUseRule=UseColor_Specified, specifiedColor=<rgba 1 1 1 1>}
     foregroundColor {colorUseRule=UseColor_Specified, specifiedColor=<rgba 0.5271 0.5271 0.5271 1>}
     hScrollBarPadding=<ltrb 0 0 0 0>
     padding=<ltrb 8 2 2 2>
     readOnlyForegroundColor {colorUseRule=UseColor_Foreground, specifiedColor=<rgba 1 0 1 1>}
     scrollBarStyle {
      thickness=8
      draggedThumbImage {
       bIsDynamicallyLoaded=false
       drawAs=RoundedBox
       imageType=NoImage
       mirroring=NoMirror
       resourceName=None
       resourceObject=null
       tiling=NoTile
       imageSize=<0 0>
       margin=<ltrb 0 0 0 0>
       outlineSettings {
        bUseBrushTransparency=false
        roundingType=FixedRadius
        width=0
        color {colorUseRule=UseColor_Specified, specifiedColor=<rgba 0 0 0 0>}
        cornerRadii=<4 4 4 4>
       }
       tintColor {colorUseRule=UseColor_Specified, specifiedColor=<rgba 0.2159 0.2159 0.2159 1>}
       uVRegion {bIsValid=false, max=<0 0>, min=<0 0>}
      }
      horizontalBackgroundImage {
       bIsDynamicallyLoaded=false
       drawAs=NoDrawType
       imageType=NoImage
       mirroring=NoMirror
       resourceName=None
       resourceObject=null
       tiling=NoTile
       imageSize=<0 0>
       margin=<ltrb 0 0 0 0>
       outlineSettings {
        bUseBrushTransparency=false
        roundingType=HalfHeightRadius
        width=0
        color {colorUseRule=UseColor_Specified, specifiedColor=<rgba 0 0 0 0>}
        cornerRadii=<0 0 0 0>
       }
       tintColor {colorUseRule=UseColor_Specified, specifiedColor=<rgba 1 1 1 1>}
       uVRegion {bIsValid=false, max=<0 0>, min=<0 0>}
      }
      horizontalBottomSlotImage {
       bIsDynamicallyLoaded=false
       drawAs=NoDrawType
       imageType=NoImage
       mirroring=NoMirror
       resourceName=None
       resourceObject=null
       tiling=NoTile
       imageSize=<0 0>
       margin=<ltrb 0 0 0 0>
       outlineSettings {
        bUseBrushTransparency=false
        roundingType=HalfHeightRadius
        width=0
        color {colorUseRule=UseColor_Specified, specifiedColor=<rgba 0 0 0 0>}
        cornerRadii=<0 0 0 0>
       }
       tintColor {colorUseRule=UseColor_Specified, specifiedColor=<rgba 1 1 1 1>}
       uVRegion {bIsValid=false, max=<0 0>, min=<0 0>}
      }
      horizontalTopSlotImage {
       bIsDynamicallyLoaded=false
       drawAs=NoDrawType
       imageType=NoImage
       mirroring=NoMirror
       resourceName=None
       resourceObject=null
       tiling=NoTile
       imageSize=<0 0>
       margin=<ltrb 0 0 0 0>
       outlineSettings {
        bUseBrushTransparency=false
        roundingType=HalfHeightRadius
        width=0
        color {colorUseRule=UseColor_Specified, specifiedColor=<rgba 0 0 0 0>}
        cornerRadii=<0 0 0 0>
       }
       tintColor {colorUseRule=UseColor_Specified, specifiedColor=<rgba 1 1 1 1>}
       uVRegion {bIsValid=false, max=<0 0>, min=<0 0>}
      }
      hoveredThumbImage {
       bIsDynamicallyLoaded=false
       drawAs=RoundedBox
       imageType=NoImage
       mirroring=NoMirror
       resourceName=None
       resourceObject=null
       tiling=NoTile
       imageSize=<0 0>
       margin=<ltrb 0 0 0 0>
       outlineSettings {
        bUseBrushTransparency=false
        roundingType=FixedRadius
        width=0
        color {colorUseRule=UseColor_Specified, specifiedColor=<rgba 0 0 0 0>}
        cornerRadii=<4 4 4 4>
       }
       tintColor {colorUseRule=UseColor_Specified, specifiedColor=<rgba 0.2159 0.2159 0.2159 1>}
       uVRegion {bIsValid=false, max=<0 0>, min=<0 0>}
      }
      normalThumbImage {
       bIsDynamicallyLoaded=false
       drawAs=RoundedBox
       imageType=NoImage
       mirroring=NoMirror
       resourceName=None
       resourceObject=null
       tiling=NoTile
       imageSize=<0 0>
       margin=<ltrb 0 0 0 0>
       outlineSettings {
        bUseBrushTransparency=false
        roundingType=FixedRadius
        width=0
        color {colorUseRule=UseColor_Specified, specifiedColor=<rgba 0 0 0 0>}
        cornerRadii=<4 4 4 4>
       }
       tintColor {colorUseRule=UseColor_Specified, specifiedColor=<rgba 0.0953 0.0953 0.0953 1>}
       uVRegion {bIsValid=false, max=<0 0>, min=<0 0>}
      }
      verticalBackgroundImage {
       bIsDynamicallyLoaded=false
       drawAs=NoDrawType
       imageType=NoImage
       mirroring=NoMirror
       resourceName=None
       resourceObject=null
       tiling=NoTile
       imageSize=<0 0>
       margin=<ltrb 0 0 0 0>
       outlineSettings {
        bUseBrushTransparency=false
        roundingType=HalfHeightRadius
        width=0
        color {colorUseRule=UseColor_Specified, specifiedColor=<rgba 0 0 0 0>}
        cornerRadii=<0 0 0 0>
       }
       tintColor {colorUseRule=UseColor_Specified, specifiedColor=<rgba 1 1 1 1>}
       uVRegion {bIsValid=false, max=<0 0>, min=<0 0>}
      }
      verticalBottomSlotImage {
       bIsDynamicallyLoaded=false
       drawAs=NoDrawType
       imageType=NoImage
       mirroring=NoMirror
       resourceName=None
       resourceObject=null
       tiling=NoTile
       imageSize=<0 0>
       margin=<ltrb 0 0 0 0>
       outlineSettings {
        bUseBrushTransparency=false
        roundingType=HalfHeightRadius
        width=0
        color {colorUseRule=UseColor_Specified, specifiedColor=<rgba 0 0 0 0>}
        cornerRadii=<0 0 0 0>
       }
       tintColor {colorUseRule=UseColor_Specified, specifiedColor=<rgba 1 1 1 1>}
       uVRegion {bIsValid=false, max=<0 0>, min=<0 0>}
      }
      verticalTopSlotImage {
       bIsDynamicallyLoaded=false
       drawAs=NoDrawType
       imageType=NoImage
       mirroring=NoMirror
       resourceName=None
       resourceObject=null
       tiling=NoTile
       imageSize=<0 0>
       margin=<ltrb 0 0 0 0>
       outlineSettings {
        bUseBrushTransparency=false
        roundingType=HalfHeightRadius
        width=0
        color {colorUseRule=UseColor_Specified, specifiedColor=<rgba 0 0 0 0>}
        cornerRadii=<0 0 0 0>
       }
       tintColor {colorUseRule=UseColor_Specified, specifiedColor=<rgba 1 1 1 1>}
       uVRegion {bIsValid=false, max=<0 0>, min=<0 0>}
      }
     }
     textStyle {
      overflowPolicy=Clip
      transformPolicy=None
      colorAndOpacity {colorUseRule=UseColor_Foreground, specifiedColor=<rgba 1 0 1 1>}
      font {
       bForceMonospaced=false
       bMaterialIsStencil=false
       fontMaterial=null
       fontObject=null
       letterSpacing=0
       monospacedWidth=1
       size=10
       skewAmount=0
       typefaceFontName=Regular
       outlineSettings {
        bApplyOutlineToDropShadows=false
        bMiteredCorners=false
        bSeparateFillAlpha=false
        outlineMaterial=null
        outlineSize=0
        outlineColor=<rgba 0 0 0 1>
       }
      }
      highlightColor {colorUseRule=UseColor_ColorTable, specifiedColor=<rgba 0 0 0 0>}
      highlightShape {
       bIsDynamicallyLoaded=false
       drawAs=Image
       imageType=NoImage
       mirroring=NoMirror
       resourceName=None
       resourceObject=null
       tiling=NoTile
       imageSize=<0 0>
       margin=<ltrb 0 0 0 0>
       outlineSettings {
        bUseBrushTransparency=false
        roundingType=HalfHeightRadius
        width=0
        color {colorUseRule=UseColor_Specified, specifiedColor=<rgba 0 0 0 0>}
        cornerRadii=<0 0 0 0>
       }
       tintColor {colorUseRule=UseColor_ColorTable, specifiedColor=<rgba 0 0 0 0>}
       uVRegion {bIsValid=false, max=<0 0>, min=<0 0>}
      }
      selectedBackgroundColor {colorUseRule=UseColor_ColorTable, specifiedColor=<rgba 0 0 0 0>}
      shadowColorAndOpacity=<rgba 0 0 0 1>
      shadowOffset=<0 0>
      strikeBrush {
       bIsDynamicallyLoaded=false
       drawAs=Image
       imageType=NoImage
       mirroring=NoMirror
       resourceName=None
       resourceObject=null
       tiling=NoTile
       imageSize=<32 32>
       margin=<ltrb 0 0 0 0>
       outlineSettings {
        bUseBrushTransparency=false
        roundingType=HalfHeightRadius
        width=0
        color {colorUseRule=UseColor_Specified, specifiedColor=<rgba 0 0 0 0>}
        cornerRadii=<0 0 0 0>
       }
       tintColor {colorUseRule=UseColor_Specified, specifiedColor=<rgba 1 1 1 1>}
       uVRegion {bIsValid=false, max=<0 0>, min=<0 0>}
      }
      underlineBrush {
       bIsDynamicallyLoaded=false
       drawAs=Image
       imageType=NoImage
       mirroring=NoMirror
       resourceName=None
       resourceObject=null
       tiling=NoTile
       imageSize=<32 32>
       margin=<ltrb 0 0 0 0>
       outlineSettings {
        bUseBrushTransparency=false
        roundingType=HalfHeightRadius
        width=0
        color {colorUseRule=UseColor_Specified, specifiedColor=<rgba 0 0 0 0>}
        cornerRadii=<0 0 0 0>
       }
       tintColor {colorUseRule=UseColor_Specified, specifiedColor=<rgba 1 1 1 1>}
       uVRegion {bIsValid=false, max=<0 0>, min=<0 0>}
      }
     }
     vScrollBarPadding=<ltrb 0 0 0 0>
    }
   }
   slot {class=GridSlot, properties {column=3, horizontalAlignment=HAlign_Center, verticalAlignment=VAlign_Center}}
  }
  SinglePropertyView_BookMarkColor {
   widgetClassPath=&SinglePropertyView
   parent=GridPanel_195
   widget=SinglePropertyView_BookMarkColor
   properties {nameOverride=" ", propertyName=BookmarkColor, renderTransform {angle=0, scale=<0.84 0.84>, shear=<0 0>, translation=<0 0>}}
   slot {class=GridSlot, properties {column=4, horizontalAlignment=HAlign_Left, verticalAlignment=VAlign_Center, padding=<ltrb -20 2 -24 2>}}
  }
  Add_Track_Text_2 {
   widgetClassPath=&TextBlock
   bIsVariable=true
   parent=GridPanel_195
   widget=Add_Track_Text_2
   properties {
    text="(Optional)   End:"
    toolTipText="Name of the end bookmark (that marks the END of the export segment)\r\n\r\nUsually you don't need this. It is used to create overalpping segmetns. "
    colorAndOpacity {colorUseRule=UseColor_Specified, specifiedColor=<rgba 0.1545 0.1545 0.1545 1>}
    font {
     bForceMonospaced=false
     bMaterialIsStencil=false
     fontMaterial=null
     fontObject=&Roboto
     letterSpacing=0
     monospacedWidth=1
     size=9
     skewAmount=0
     typefaceFontName=Bold
     outlineSettings {
      bApplyOutlineToDropShadows=false
      bMiteredCorners=false
      bSeparateFillAlpha=false
      outlineMaterial=null
      outlineSize=0
      outlineColor=<rgba 0 0 0 1>
     }
    }
   }
   slot {class=GridSlot, properties {horizontalAlignment=HAlign_Left, row=1, verticalAlignment=VAlign_Center, padding=<ltrb 0 0 12 0>}}
  }
  EditorUtilityEditableTextBoxEndBookmark {
   widgetClassPath=&EditorUtilityEditableTextBox
   bIsVariable=true
   parent=GridPanel_195
   widget=EditorUtilityEditableTextBoxEndBookmark
   properties {
    hintText=( optional )
    toolTipText="Name of the end bookmark (that marks the END of the export segment)\r\n\r\nYOusually you don't need this. It is used to create overalpping segmetns. "
    visibility=Collapsed
    widgetStyle {
     backgroundColor {colorUseRule=UseColor_Specified, specifiedColor=<rgba 1 1 1 1>}
     backgroundImageFocused {
      bIsDynamicallyLoaded=false
      drawAs=RoundedBox
      imageType=NoImage
      mirroring=NoMirror
      resourceName=None
      resourceObject=null
      tiling=NoTile
      imageSize=<0 0>
      margin=<ltrb 0 0 0 0>
      outlineSettings {
       bUseBrushTransparency=false
       roundingType=FixedRadius
       width=1
       color {colorUseRule=UseColor_ColorTable, specifiedColor=<rgba 0 0 0 0>}
       cornerRadii=<4 4 4 4>
      }
      tintColor {colorUseRule=UseColor_Specified, specifiedColor=<rgba 0.0048 0.0048 0.0048 1>}
      uVRegion {bIsValid=false, max=<0 0>, min=<0 0>}
     }
     backgroundImageHovered {
      bIsDynamicallyLoaded=false
      drawAs=RoundedBox
      imageType=NoImage
      mirroring=NoMirror
      resourceName=None
      resourceObject=null
      tiling=NoTile
      imageSize=<0 0>
      margin=<ltrb 0 0 0 0>
      outlineSettings {
       bUseBrushTransparency=false
       roundingType=FixedRadius
       width=1
       color {colorUseRule=UseColor_ColorTable, specifiedColor=<rgba 0 0 0 0>}
       cornerRadii=<4 4 4 4>
      }
      tintColor {colorUseRule=UseColor_Specified, specifiedColor=<rgba 0.0048 0.0048 0.0048 1>}
      uVRegion {bIsValid=false, max=<0 0>, min=<0 0>}
     }
     backgroundImageNormal {
      bIsDynamicallyLoaded=false
      drawAs=RoundedBox
      imageType=NoImage
      mirroring=NoMirror
      resourceName=None
      resourceObject=null
      tiling=NoTile
      imageSize=<0 0>
      margin=<ltrb 0 0 0 0>
      outlineSettings {
       bUseBrushTransparency=false
       roundingType=FixedRadius
       width=1
       color {colorUseRule=UseColor_ColorTable, specifiedColor=<rgba 0 0 0 0>}
       cornerRadii=<4 4 4 4>
      }
      tintColor {colorUseRule=UseColor_Specified, specifiedColor=<rgba 0.0048 0.0048 0.0048 1>}
      uVRegion {bIsValid=false, max=<0 0>, min=<0 0>}
     }
     backgroundImageReadOnly {
      bIsDynamicallyLoaded=false
      drawAs=RoundedBox
      imageType=NoImage
      mirroring=NoMirror
      resourceName=None
      resourceObject=null
      tiling=NoTile
      imageSize=<0 0>
      margin=<ltrb 0 0 0 0>
      outlineSettings {
       bUseBrushTransparency=false
       roundingType=FixedRadius
       width=1
       color {colorUseRule=UseColor_ColorTable, specifiedColor=<rgba 0 0 0 0>}
       cornerRadii=<4 4 4 4>
      }
      tintColor {colorUseRule=UseColor_Specified, specifiedColor=<rgba 0.0284 0.0284 0.0284 1>}
      uVRegion {bIsValid=false, max=<0 0>, min=<0 0>}
     }
     focusedForegroundColor {colorUseRule=UseColor_Specified, specifiedColor=<rgba 1 1 1 1>}
     foregroundColor {colorUseRule=UseColor_Specified, specifiedColor=<rgba 0.5271 0.5271 0.5271 1>}
     hScrollBarPadding=<ltrb 0 0 0 0>
     padding=<ltrb 8 2 2 2>
     readOnlyForegroundColor {colorUseRule=UseColor_Foreground, specifiedColor=<rgba 1 0 1 1>}
     scrollBarStyle {
      thickness=8
      draggedThumbImage {
       bIsDynamicallyLoaded=false
       drawAs=RoundedBox
       imageType=NoImage
       mirroring=NoMirror
       resourceName=None
       resourceObject=null
       tiling=NoTile
       imageSize=<0 0>
       margin=<ltrb 0 0 0 0>
       outlineSettings {
        bUseBrushTransparency=false
        roundingType=FixedRadius
        width=0
        color {colorUseRule=UseColor_Specified, specifiedColor=<rgba 0 0 0 0>}
        cornerRadii=<4 4 4 4>
       }
       tintColor {colorUseRule=UseColor_Specified, specifiedColor=<rgba 0.2159 0.2159 0.2159 1>}
       uVRegion {bIsValid=false, max=<0 0>, min=<0 0>}
      }
      horizontalBackgroundImage {
       bIsDynamicallyLoaded=false
       drawAs=NoDrawType
       imageType=NoImage
       mirroring=NoMirror
       resourceName=None
       resourceObject=null
       tiling=NoTile
       imageSize=<0 0>
       margin=<ltrb 0 0 0 0>
       outlineSettings {
        bUseBrushTransparency=false
        roundingType=HalfHeightRadius
        width=0
        color {colorUseRule=UseColor_Specified, specifiedColor=<rgba 0 0 0 0>}
        cornerRadii=<0 0 0 0>
       }
       tintColor {colorUseRule=UseColor_Specified, specifiedColor=<rgba 1 1 1 1>}
       uVRegion {bIsValid=false, max=<0 0>, min=<0 0>}
      }
      horizontalBottomSlotImage {
       bIsDynamicallyLoaded=false
       drawAs=NoDrawType
       imageType=NoImage
       mirroring=NoMirror
       resourceName=None
       resourceObject=null
       tiling=NoTile
       imageSize=<0 0>
       margin=<ltrb 0 0 0 0>
       outlineSettings {
        bUseBrushTransparency=false
        roundingType=HalfHeightRadius
        width=0
        color {colorUseRule=UseColor_Specified, specifiedColor=<rgba 0 0 0 0>}
        cornerRadii=<0 0 0 0>
       }
       tintColor {colorUseRule=UseColor_Specified, specifiedColor=<rgba 1 1 1 1>}
       uVRegion {bIsValid=false, max=<0 0>, min=<0 0>}
      }
      horizontalTopSlotImage {
       bIsDynamicallyLoaded=false
       drawAs=NoDrawType
       imageType=NoImage
       mirroring=NoMirror
       resourceName=None
       resourceObject=null
       tiling=NoTile
       imageSize=<0 0>
       margin=<ltrb 0 0 0 0>
       outlineSettings {
        bUseBrushTransparency=false
        roundingType=HalfHeightRadius
        width=0
        color {colorUseRule=UseColor_Specified, specifiedColor=<rgba 0 0 0 0>}
        cornerRadii=<0 0 0 0>
       }
       tintColor {colorUseRule=UseColor_Specified, specifiedColor=<rgba 1 1 1 1>}
       uVRegion {bIsValid=false, max=<0 0>, min=<0 0>}
      }
      hoveredThumbImage {
       bIsDynamicallyLoaded=false
       drawAs=RoundedBox
       imageType=NoImage
       mirroring=NoMirror
       resourceName=None
       resourceObject=null
       tiling=NoTile
       imageSize=<0 0>
       margin=<ltrb 0 0 0 0>
       outlineSettings {
        bUseBrushTransparency=false
        roundingType=FixedRadius
        width=0
        color {colorUseRule=UseColor_Specified, specifiedColor=<rgba 0 0 0 0>}
        cornerRadii=<4 4 4 4>
       }
       tintColor {colorUseRule=UseColor_Specified, specifiedColor=<rgba 0.2159 0.2159 0.2159 1>}
       uVRegion {bIsValid=false, max=<0 0>, min=<0 0>}
      }
      normalThumbImage {
       bIsDynamicallyLoaded=false
       drawAs=RoundedBox
       imageType=NoImage
       mirroring=NoMirror
       resourceName=None
       resourceObject=null
       tiling=NoTile
       imageSize=<0 0>
       margin=<ltrb 0 0 0 0>
       outlineSettings {
        bUseBrushTransparency=false
        roundingType=FixedRadius
        width=0
        color {colorUseRule=UseColor_Specified, specifiedColor=<rgba 0 0 0 0>}
        cornerRadii=<4 4 4 4>
       }
       tintColor {colorUseRule=UseColor_Specified, specifiedColor=<rgba 0.0953 0.0953 0.0953 1>}
       uVRegion {bIsValid=false, max=<0 0>, min=<0 0>}
      }
      verticalBackgroundImage {
       bIsDynamicallyLoaded=false
       drawAs=NoDrawType
       imageType=NoImage
       mirroring=NoMirror
       resourceName=None
       resourceObject=null
       tiling=NoTile
       imageSize=<0 0>
       margin=<ltrb 0 0 0 0>
       outlineSettings {
        bUseBrushTransparency=false
        roundingType=HalfHeightRadius
        width=0
        color {colorUseRule=UseColor_Specified, specifiedColor=<rgba 0 0 0 0>}
        cornerRadii=<0 0 0 0>
       }
       tintColor {colorUseRule=UseColor_Specified, specifiedColor=<rgba 1 1 1 1>}
       uVRegion {bIsValid=false, max=<0 0>, min=<0 0>}
      }
      verticalBottomSlotImage {
       bIsDynamicallyLoaded=false
       drawAs=NoDrawType
       imageType=NoImage
       mirroring=NoMirror
       resourceName=None
       resourceObject=null
       tiling=NoTile
       imageSize=<0 0>
       margin=<ltrb 0 0 0 0>
       outlineSettings {
        bUseBrushTransparency=false
        roundingType=HalfHeightRadius
        width=0
        color {colorUseRule=UseColor_Specified, specifiedColor=<rgba 0 0 0 0>}
        cornerRadii=<0 0 0 0>
       }
       tintColor {colorUseRule=UseColor_Specified, specifiedColor=<rgba 1 1 1 1>}
       uVRegion {bIsValid=false, max=<0 0>, min=<0 0>}
      }
      verticalTopSlotImage {
       bIsDynamicallyLoaded=false
       drawAs=NoDrawType
       imageType=NoImage
       mirroring=NoMirror
       resourceName=None
       resourceObject=null
       tiling=NoTile
       imageSize=<0 0>
       margin=<ltrb 0 0 0 0>
       outlineSettings {
        bUseBrushTransparency=false
        roundingType=HalfHeightRadius
        width=0
        color {colorUseRule=UseColor_Specified, specifiedColor=<rgba 0 0 0 0>}
        cornerRadii=<0 0 0 0>
       }
       tintColor {colorUseRule=UseColor_Specified, specifiedColor=<rgba 1 1 1 1>}
       uVRegion {bIsValid=false, max=<0 0>, min=<0 0>}
      }
     }
     textStyle {
      overflowPolicy=Clip
      transformPolicy=None
      colorAndOpacity {colorUseRule=UseColor_Foreground, specifiedColor=<rgba 1 0 1 1>}
      font {
       bForceMonospaced=false
       bMaterialIsStencil=false
       fontMaterial=null
       fontObject=null
       letterSpacing=0
       monospacedWidth=1
       size=10
       skewAmount=0
       typefaceFontName=Regular
       outlineSettings {
        bApplyOutlineToDropShadows=false
        bMiteredCorners=false
        bSeparateFillAlpha=false
        outlineMaterial=null
        outlineSize=0
        outlineColor=<rgba 0 0 0 1>
       }
      }
      highlightColor {colorUseRule=UseColor_ColorTable, specifiedColor=<rgba 0 0 0 0>}
      highlightShape {
       bIsDynamicallyLoaded=false
       drawAs=Image
       imageType=NoImage
       mirroring=NoMirror
       resourceName=None
       resourceObject=null
       tiling=NoTile
       imageSize=<0 0>
       margin=<ltrb 0 0 0 0>
       outlineSettings {
        bUseBrushTransparency=false
        roundingType=HalfHeightRadius
        width=0
        color {colorUseRule=UseColor_Specified, specifiedColor=<rgba 0 0 0 0>}
        cornerRadii=<0 0 0 0>
       }
       tintColor {colorUseRule=UseColor_ColorTable, specifiedColor=<rgba 0 0 0 0>}
       uVRegion {bIsValid=false, max=<0 0>, min=<0 0>}
      }
      selectedBackgroundColor {colorUseRule=UseColor_ColorTable, specifiedColor=<rgba 0 0 0 0>}
      shadowColorAndOpacity=<rgba 0 0 0 1>
      shadowOffset=<0 0>
      strikeBrush {
       bIsDynamicallyLoaded=false
       drawAs=Image
       imageType=NoImage
       mirroring=NoMirror
       resourceName=None
       resourceObject=null
       tiling=NoTile
       imageSize=<32 32>
       margin=<ltrb 0 0 0 0>
       outlineSettings {
        bUseBrushTransparency=false
        roundingType=HalfHeightRadius
        width=0
        color {colorUseRule=UseColor_Specified, specifiedColor=<rgba 0 0 0 0>}
        cornerRadii=<0 0 0 0>
       }
       tintColor {colorUseRule=UseColor_Specified, specifiedColor=<rgba 1 1 1 1>}
       uVRegion {bIsValid=false, max=<0 0>, min=<0 0>}
      }
      underlineBrush {
       bIsDynamicallyLoaded=false
       drawAs=Image
       imageType=NoImage
       mirroring=NoMirror
       resourceName=None
       resourceObject=null
       tiling=NoTile
       imageSize=<32 32>
       margin=<ltrb 0 0 0 0>
       outlineSettings {
        bUseBrushTransparency=false
        roundingType=HalfHeightRadius
        width=0
        color {colorUseRule=UseColor_Specified, specifiedColor=<rgba 0 0 0 0>}
        cornerRadii=<0 0 0 0>
       }
       tintColor {colorUseRule=UseColor_Specified, specifiedColor=<rgba 1 1 1 1>}
       uVRegion {bIsValid=false, max=<0 0>, min=<0 0>}
      }
     }
     vScrollBarPadding=<ltrb 0 0 0 0>
    }
   }
   slot {class=GridSlot, properties {column=2, row=1, verticalAlignment=VAlign_Center}}
  }
  EditorUtilityEditableTextBoxEndFrame {
   widgetClassPath=&EditorUtilityEditableTextBox
   bIsVariable=true
   parent=GridPanel_195
   widget=EditorUtilityEditableTextBoxEndFrame
   properties {
    hintText="  ##  "
    toolTipText=Frame number to place this bookmark
    widgetStyle {
     backgroundColor {colorUseRule=UseColor_Specified, specifiedColor=<rgba 1 1 1 1>}
     backgroundImageFocused {
      bIsDynamicallyLoaded=false
      drawAs=RoundedBox
      imageType=NoImage
      mirroring=NoMirror
      resourceName=None
      resourceObject=null
      tiling=NoTile
      imageSize=<0 0>
      margin=<ltrb 0 0 0 0>
      outlineSettings {
       bUseBrushTransparency=false
       roundingType=FixedRadius
       width=1
       color {colorUseRule=UseColor_ColorTable, specifiedColor=<rgba 0 0 0 0>}
       cornerRadii=<4 4 4 4>
      }
      tintColor {colorUseRule=UseColor_Specified, specifiedColor=<rgba 0.0048 0.0048 0.0048 1>}
      uVRegion {bIsValid=false, max=<0 0>, min=<0 0>}
     }
     backgroundImageHovered {
      bIsDynamicallyLoaded=false
      drawAs=RoundedBox
      imageType=NoImage
      mirroring=NoMirror
      resourceName=None
      resourceObject=null
      tiling=NoTile
      imageSize=<0 0>
      margin=<ltrb 0 0 0 0>
      outlineSettings {
       bUseBrushTransparency=false
       roundingType=FixedRadius
       width=1
       color {colorUseRule=UseColor_ColorTable, specifiedColor=<rgba 0 0 0 0>}
       cornerRadii=<4 4 4 4>
      }
      tintColor {colorUseRule=UseColor_Specified, specifiedColor=<rgba 0.0048 0.0048 0.0048 1>}
      uVRegion {bIsValid=false, max=<0 0>, min=<0 0>}
     }
     backgroundImageNormal {
      bIsDynamicallyLoaded=false
      drawAs=RoundedBox
      imageType=NoImage
      mirroring=NoMirror
      resourceName=None
      resourceObject=null
      tiling=NoTile
      imageSize=<0 0>
      margin=<ltrb 0 0 0 0>
      outlineSettings {
       bUseBrushTransparency=false
       roundingType=FixedRadius
       width=1
       color {colorUseRule=UseColor_ColorTable, specifiedColor=<rgba 0 0 0 0>}
       cornerRadii=<4 4 4 4>
      }
      tintColor {colorUseRule=UseColor_Specified, specifiedColor=<rgba 0.0048 0.0048 0.0048 1>}
      uVRegion {bIsValid=false, max=<0 0>, min=<0 0>}
     }
     backgroundImageReadOnly {
      bIsDynamicallyLoaded=false
      drawAs=RoundedBox
      imageType=NoImage
      mirroring=NoMirror
      resourceName=None
      resourceObject=null
      tiling=NoTile
      imageSize=<0 0>
      margin=<ltrb 0 0 0 0>
      outlineSettings {
       bUseBrushTransparency=false
       roundingType=FixedRadius
       width=1
       color {colorUseRule=UseColor_ColorTable, specifiedColor=<rgba 0 0 0 0>}
       cornerRadii=<4 4 4 4>
      }
      tintColor {colorUseRule=UseColor_Specified, specifiedColor=<rgba 0.0284 0.0284 0.0284 1>}
      uVRegion {bIsValid=false, max=<0 0>, min=<0 0>}
     }
     focusedForegroundColor {colorUseRule=UseColor_Specified, specifiedColor=<rgba 1 1 1 1>}
     foregroundColor {colorUseRule=UseColor_Specified, specifiedColor=<rgba 0.5271 0.5271 0.5271 1>}
     hScrollBarPadding=<ltrb 0 0 0 0>
     padding=<ltrb 8 2 2 2>
     readOnlyForegroundColor {colorUseRule=UseColor_Foreground, specifiedColor=<rgba 1 0 1 1>}
     scrollBarStyle {
      thickness=8
      draggedThumbImage {
       bIsDynamicallyLoaded=false
       drawAs=RoundedBox
       imageType=NoImage
       mirroring=NoMirror
       resourceName=None
       resourceObject=null
       tiling=NoTile
       imageSize=<0 0>
       margin=<ltrb 0 0 0 0>
       outlineSettings {
        bUseBrushTransparency=false
        roundingType=FixedRadius
        width=0
        color {colorUseRule=UseColor_Specified, specifiedColor=<rgba 0 0 0 0>}
        cornerRadii=<4 4 4 4>
       }
       tintColor {colorUseRule=UseColor_Specified, specifiedColor=<rgba 0.2159 0.2159 0.2159 1>}
       uVRegion {bIsValid=false, max=<0 0>, min=<0 0>}
      }
      horizontalBackgroundImage {
       bIsDynamicallyLoaded=false
       drawAs=NoDrawType
       imageType=NoImage
       mirroring=NoMirror
       resourceName=None
       resourceObject=null
       tiling=NoTile
       imageSize=<0 0>
       margin=<ltrb 0 0 0 0>
       outlineSettings {
        bUseBrushTransparency=false
        roundingType=HalfHeightRadius
        width=0
        color {colorUseRule=UseColor_Specified, specifiedColor=<rgba 0 0 0 0>}
        cornerRadii=<0 0 0 0>
       }
       tintColor {colorUseRule=UseColor_Specified, specifiedColor=<rgba 1 1 1 1>}
       uVRegion {bIsValid=false, max=<0 0>, min=<0 0>}
      }
      horizontalBottomSlotImage {
       bIsDynamicallyLoaded=false
       drawAs=NoDrawType
       imageType=NoImage
       mirroring=NoMirror
       resourceName=None
       resourceObject=null
       tiling=NoTile
       imageSize=<0 0>
       margin=<ltrb 0 0 0 0>
       outlineSettings {
        bUseBrushTransparency=false
        roundingType=HalfHeightRadius
        width=0
        color {colorUseRule=UseColor_Specified, specifiedColor=<rgba 0 0 0 0>}
        cornerRadii=<0 0 0 0>
       }
       tintColor {colorUseRule=UseColor_Specified, specifiedColor=<rgba 1 1 1 1>}
       uVRegion {bIsValid=false, max=<0 0>, min=<0 0>}
      }
      horizontalTopSlotImage {
       bIsDynamicallyLoaded=false
       drawAs=NoDrawType
       imageType=NoImage
       mirroring=NoMirror
       resourceName=None
       resourceObject=null
       tiling=NoTile
       imageSize=<0 0>
       margin=<ltrb 0 0 0 0>
       outlineSettings {
        bUseBrushTransparency=false
        roundingType=HalfHeightRadius
        width=0
        color {colorUseRule=UseColor_Specified, specifiedColor=<rgba 0 0 0 0>}
        cornerRadii=<0 0 0 0>
       }
       tintColor {colorUseRule=UseColor_Specified, specifiedColor=<rgba 1 1 1 1>}
       uVRegion {bIsValid=false, max=<0 0>, min=<0 0>}
      }
      hoveredThumbImage {
       bIsDynamicallyLoaded=false
       drawAs=RoundedBox
       imageType=NoImage
       mirroring=NoMirror
       resourceName=None
       resourceObject=null
       tiling=NoTile
       imageSize=<0 0>
       margin=<ltrb 0 0 0 0>
       outlineSettings {
        bUseBrushTransparency=false
        roundingType=FixedRadius
        width=0
        color {colorUseRule=UseColor_Specified, specifiedColor=<rgba 0 0 0 0>}
        cornerRadii=<4 4 4 4>
       }
       tintColor {colorUseRule=UseColor_Specified, specifiedColor=<rgba 0.2159 0.2159 0.2159 1>}
       uVRegion {bIsValid=false, max=<0 0>, min=<0 0>}
      }
      normalThumbImage {
       bIsDynamicallyLoaded=false
       drawAs=RoundedBox
       imageType=NoImage
       mirroring=NoMirror
       resourceName=None
       resourceObject=null
       tiling=NoTile
       imageSize=<0 0>
       margin=<ltrb 0 0 0 0>
       outlineSettings {
        bUseBrushTransparency=false
        roundingType=FixedRadius
        width=0
        color {colorUseRule=UseColor_Specified, specifiedColor=<rgba 0 0 0 0>}
        cornerRadii=<4 4 4 4>
       }
       tintColor {colorUseRule=UseColor_Specified, specifiedColor=<rgba 0.0953 0.0953 0.0953 1>}
       uVRegion {bIsValid=false, max=<0 0>, min=<0 0>}
      }
      verticalBackgroundImage {
       bIsDynamicallyLoaded=false
       drawAs=NoDrawType
       imageType=NoImage
       mirroring=NoMirror
       resourceName=None
       resourceObject=null
       tiling=NoTile
       imageSize=<0 0>
       margin=<ltrb 0 0 0 0>
       outlineSettings {
        bUseBrushTransparency=false
        roundingType=HalfHeightRadius
        width=0
        color {colorUseRule=UseColor_Specified, specifiedColor=<rgba 0 0 0 0>}
        cornerRadii=<0 0 0 0>
       }
       tintColor {colorUseRule=UseColor_Specified, specifiedColor=<rgba 1 1 1 1>}
       uVRegion {bIsValid=false, max=<0 0>, min=<0 0>}
      }
      verticalBottomSlotImage {
       bIsDynamicallyLoaded=false
       drawAs=NoDrawType
       imageType=NoImage
       mirroring=NoMirror
       resourceName=None
       resourceObject=null
       tiling=NoTile
       imageSize=<0 0>
       margin=<ltrb 0 0 0 0>
       outlineSettings {
        bUseBrushTransparency=false
        roundingType=HalfHeightRadius
        width=0
        color {colorUseRule=UseColor_Specified, specifiedColor=<rgba 0 0 0 0>}
        cornerRadii=<0 0 0 0>
       }
       tintColor {colorUseRule=UseColor_Specified, specifiedColor=<rgba 1 1 1 1>}
       uVRegion {bIsValid=false, max=<0 0>, min=<0 0>}
      }
      verticalTopSlotImage {
       bIsDynamicallyLoaded=false
       drawAs=NoDrawType
       imageType=NoImage
       mirroring=NoMirror
       resourceName=None
       resourceObject=null
       tiling=NoTile
       imageSize=<0 0>
       margin=<ltrb 0 0 0 0>
       outlineSettings {
        bUseBrushTransparency=false
        roundingType=HalfHeightRadius
        width=0
        color {colorUseRule=UseColor_Specified, specifiedColor=<rgba 0 0 0 0>}
        cornerRadii=<0 0 0 0>
       }
       tintColor {colorUseRule=UseColor_Specified, specifiedColor=<rgba 1 1 1 1>}
       uVRegion {bIsValid=false, max=<0 0>, min=<0 0>}
      }
     }
     textStyle {
      overflowPolicy=Clip
      transformPolicy=None
      colorAndOpacity {colorUseRule=UseColor_Foreground, specifiedColor=<rgba 1 0 1 1>}
      font {
       bForceMonospaced=false
       bMaterialIsStencil=false
       fontMaterial=null
       fontObject=null
       letterSpacing=0
       monospacedWidth=1
       size=10
       skewAmount=0
       typefaceFontName=Regular
       outlineSettings {
        bApplyOutlineToDropShadows=false
        bMiteredCorners=false
        bSeparateFillAlpha=false
        outlineMaterial=null
        outlineSize=0
        outlineColor=<rgba 0 0 0 1>
       }
      }
      highlightColor {colorUseRule=UseColor_ColorTable, specifiedColor=<rgba 0 0 0 0>}
      highlightShape {
       bIsDynamicallyLoaded=false
       drawAs=Image
       imageType=NoImage
       mirroring=NoMirror
       resourceName=None
       resourceObject=null
       tiling=NoTile
       imageSize=<0 0>
       margin=<ltrb 0 0 0 0>
       outlineSettings {
        bUseBrushTransparency=false
        roundingType=HalfHeightRadius
        width=0
        color {colorUseRule=UseColor_Specified, specifiedColor=<rgba 0 0 0 0>}
        cornerRadii=<0 0 0 0>
       }
       tintColor {colorUseRule=UseColor_ColorTable, specifiedColor=<rgba 0 0 0 0>}
       uVRegion {bIsValid=false, max=<0 0>, min=<0 0>}
      }
      selectedBackgroundColor {colorUseRule=UseColor_ColorTable, specifiedColor=<rgba 0 0 0 0>}
      shadowColorAndOpacity=<rgba 0 0 0 1>
      shadowOffset=<0 0>
      strikeBrush {
       bIsDynamicallyLoaded=false
       drawAs=Image
       imageType=NoImage
       mirroring=NoMirror
       resourceName=None
       resourceObject=null
       tiling=NoTile
       imageSize=<32 32>
       margin=<ltrb 0 0 0 0>
       outlineSettings {
        bUseBrushTransparency=false
        roundingType=HalfHeightRadius
        width=0
        color {colorUseRule=UseColor_Specified, specifiedColor=<rgba 0 0 0 0>}
        cornerRadii=<0 0 0 0>
       }
       tintColor {colorUseRule=UseColor_Specified, specifiedColor=<rgba 1 1 1 1>}
       uVRegion {bIsValid=false, max=<0 0>, min=<0 0>}
      }
      underlineBrush {
       bIsDynamicallyLoaded=false
       drawAs=Image
       imageType=NoImage
       mirroring=NoMirror
       resourceName=None
       resourceObject=null
       tiling=NoTile
       imageSize=<32 32>
       margin=<ltrb 0 0 0 0>
       outlineSettings {
        bUseBrushTransparency=false
        roundingType=HalfHeightRadius
        width=0
        color {colorUseRule=UseColor_Specified, specifiedColor=<rgba 0 0 0 0>}
        cornerRadii=<0 0 0 0>
       }
       tintColor {colorUseRule=UseColor_Specified, specifiedColor=<rgba 1 1 1 1>}
       uVRegion {bIsValid=false, max=<0 0>, min=<0 0>}
      }
     }
     vScrollBarPadding=<ltrb 0 0 0 0>
    }
   }
   slot {class=GridSlot, properties {column=3, horizontalAlignment=HAlign_Center, row=1, verticalAlignment=VAlign_Center}}
  }
  Button_AddBookmark {
   widgetClassPath=&EditorUtilityButton
   parent=HorizontalBoxAddBookmark
   widget=Button_AddBookmark
   properties {toolTipText=Create Bookmarks}
   slot {class=HorizontalBoxSlot, properties {horizontalAlignment=HAlign_Center, verticalAlignment=VAlign_Center, padding=<ltrb 6 0 0 0>}}
  }
  AssetImagePlusButton_1 {
   widgetClassPath=&Image
   parent=Button_AddBookmark
   widget=AssetImagePlusButton_1
   properties {brush {
     bIsDynamicallyLoaded=false
     drawAs=Image
     imageType=FullColor
     mirroring=NoMirror
     resourceName=None
     resourceObject=&Icon_LevelSequence_add
     tiling=NoTile
     imageSize=<14 14>
     margin=<ltrb 0 0 0 0>
     outlineSettings {
      bUseBrushTransparency=false
      roundingType=HalfHeightRadius
      width=0
      color {colorUseRule=UseColor_Specified, specifiedColor=<rgba 0 0 0 0>}
      cornerRadii=<0 0 0 0>
     }
     tintColor {colorUseRule=UseColor_Specified, specifiedColor=<rgba 0.2108 1 0 1>}
     uVRegion {bIsValid=false, max=<0 0>, min=<0 0>}
    }}
   slot {class=ButtonSlot, properties {padding=<ltrb 0 4 0 4>}}
  }
  Spacer_2 {
   widgetClassPath=&Spacer
   parent=CoreLayout
   widget=Spacer_2
   properties {size=<1 20>}
   slot {class=VerticalBoxSlot, properties {horizontalAlignment=HAlign_Center, verticalAlignment=VAlign_Center}}
  }
  Button_Scan {
   widgetClassPath=&EditorUtilityButton
   parent=CoreLayout
   widget=Button_Scan
   slot {class=VerticalBoxSlot, properties {horizontalAlignment=HAlign_Center, verticalAlignment=VAlign_Center}}
  }
  TextBlock_4 {
   widgetClassPath=&TextBlock
   parent=Button_Scan
   widget=TextBlock_4
   properties {
    text=Refresh
    font {
     bForceMonospaced=false
     bMaterialIsStencil=false
     fontMaterial=null
     fontObject=&Roboto
     letterSpacing=0
     monospacedWidth=1
     size=10.5
     skewAmount=0
     typefaceFontName=Bold
     outlineSettings {
      bApplyOutlineToDropShadows=false
      bMiteredCorners=false
      bSeparateFillAlpha=false
      outlineMaterial=null
      outlineSize=0
      outlineColor=<rgba 0 0 0 1>
     }
    }
   }
   slot {class=ButtonSlot, properties {horizontalAlignment=HAlign_Fill, verticalAlignment=VAlign_Fill, padding=<ltrb 20 6 20 6>}}
  }
  Spacer_4 {
   widgetClassPath=&Spacer
   parent=CoreLayout
   widget=Spacer_4
   properties {size=<1 20>}
   slot {class=VerticalBoxSlot, properties {horizontalAlignment=HAlign_Center, verticalAlignment=VAlign_Center}}
  }
  VerticalBox-CroppedSequences {
   widgetClassPath=&VerticalBox
   parent=CoreLayout
   widget=VerticalBox-CroppedSequences
   slot {class=VerticalBoxSlot, properties {horizontalAlignment=HAlign_Center, verticalAlignment=VAlign_Top}}
  }
  Layout_Cropped_Sequences {
   widgetClassPath=&VerticalBox
   bIsVariable=true
   parent=VerticalBox-CroppedSequences
   widget=Layout_Cropped_Sequences
   slot {class=VerticalBoxSlot, properties {verticalAlignment=VAlign_Top, size {sizeRule=Fill, value=1}}}
  }
  INVISIBLE {
   widgetClassPath=&HorizontalBox
   parent=EditorUtilityScrollBox_1
   widget=INVISIBLE
   properties {bIsEnabled=false, visibility=Collapsed}
   slot {class=ScrollBoxSlot}
  }
  TextStyleSammple_Regular {
   widgetClassPath=&TextBlock
   bIsVariable=true
   parent=INVISIBLE
   widget=TextStyleSammple_Regular
   properties {
    text="INVISIBLE ->"
    toolTipText="Does animation contain float channel curves.\r\n\r\n\r\nType name in the filed on the right."
    font {
     bForceMonospaced=false
     bMaterialIsStencil=false
     fontMaterial=null
     fontObject=&Roboto
     letterSpacing=0
     monospacedWidth=1
     size=9
     skewAmount=0
     typefaceFontName=Regular
     outlineSettings {
      bApplyOutlineToDropShadows=false
      bMiteredCorners=false
      bSeparateFillAlpha=false
      outlineMaterial=null
      outlineSize=0
      outlineColor=<rgba 0 0 0 1>
     }
    }
   }
   slot {class=HorizontalBoxSlot, properties {horizontalAlignment=HAlign_Left, verticalAlignment=VAlign_Center}}
  }
  TextStyleSammple_Bold_Big {
   widgetClassPath=&TextBlock
   bIsVariable=true
   parent=INVISIBLE
   widget=TextStyleSammple_Bold_Big
   properties {
    text=Sample
    font {
     bForceMonospaced=false
     bMaterialIsStencil=false
     fontMaterial=null
     fontObject=&Roboto
     letterSpacing=100
     monospacedWidth=1
     size=10.5
     skewAmount=0
     typefaceFontName=Bold
     outlineSettings {
      bApplyOutlineToDropShadows=false
      bMiteredCorners=false
      bSeparateFillAlpha=false
      outlineMaterial=null
      outlineSize=0
      outlineColor=<rgba 0 0 0 1>
     }
    }
   }
   slot {class=HorizontalBoxSlot, properties {horizontalAlignment=HAlign_Left, verticalAlignment=VAlign_Center}}
  }
  Checkbox_Style {
   widgetClassPath=&EditorUtilityCheckBox
   bIsVariable=true
   parent=INVISIBLE
   widget=Checkbox_Style
   properties {horizontalAlignment=HAlign_Left, toolTipText="Does animation contain float channel curves.\r\n\r\n\r\nType name in the filed on the right."}
   slot {class=HorizontalBoxSlot, properties {horizontalAlignment=HAlign_Left, verticalAlignment=VAlign_Top, padding=<ltrb 30 0 10 0>}}
  }
  AddAnimationButton {
   widgetClassPath=&Button
   parent=INVISIBLE
   widget=AddAnimationButton
   properties {
    bIsEnabled=false
    widgetStyle {
     clickedSlateSound {resourceObject=null}
     disabled {
      bIsDynamicallyLoaded=false
      drawAs=RoundedBox
      imageType=NoImage
      mirroring=NoMirror
      resourceName=None
      resourceObject=null
      tiling=NoTile
      imageSize=<0 0>
      margin=<ltrb 0 0 0 0>
      outlineSettings {
       bUseBrushTransparency=false
       roundingType=FixedRadius
       width=1
       color {colorUseRule=UseColor_ColorTable, specifiedColor=<rgba 0 0 0 0>}
       cornerRadii=<4 4 4 4>
      }
      tintColor {colorUseRule=UseColor_Specified, specifiedColor=<rgba 0.0395 0.0395 0.0395 1>}
      uVRegion {bIsValid=false, max=<0 0>, min=<0 0>}
     }
     disabledForeground {colorUseRule=UseColor_ColorTable, specifiedColor=<rgba 0 0 0 0>}
     hovered {
      bIsDynamicallyLoaded=false
      drawAs=RoundedBox
      imageType=NoImage
      mirroring=NoMirror
      resourceName=None
      resourceObject=null
      tiling=NoTile
      imageSize=<32 32>
      margin=<ltrb 0 0 0 0>
      outlineSettings {
       bUseBrushTransparency=false
       roundingType=FixedRadius
       width=0
       color {colorUseRule=UseColor_ColorTable, specifiedColor=<rgba 0 0 0 0>}
       cornerRadii=<4 4 4 4>
      }
      tintColor {colorUseRule=UseColor_Specified, specifiedColor=<rgba 0.0953 0.0953 0.0953 1>}
      uVRegion {bIsValid=false, max=<0 0>, min=<0 0>}
     }
     hoveredForeground {colorUseRule=UseColor_ColorTable, specifiedColor=<rgba 0 0 0 0>}
     hoveredSlateSound {resourceObject=null}
     normal {
      bIsDynamicallyLoaded=false
      drawAs=RoundedBox
      imageType=FullColor
      mirroring=NoMirror
      resourceName=None
      resourceObject=null
      tiling=NoTile
      imageSize=<14 14>
      margin=<ltrb 0 0 0 0>
      outlineSettings {
       bUseBrushTransparency=false
       roundingType=FixedRadius
       width=1
       color {colorUseRule=UseColor_ColorTable, specifiedColor=<rgba 0 0 0 0>}
       cornerRadii=<4 4 4 4>
      }
      tintColor {colorUseRule=UseColor_Specified, specifiedColor=<rgba 0.0395 0.0395 0.0395 1>}
      uVRegion {bIsValid=false, max=<0 0>, min=<0 0>}
     }
     normalForeground {colorUseRule=UseColor_ColorTable, specifiedColor=<rgba 0 0 0 0>}
     normalPadding=<ltrb 12 1.5 12 1.5>
     pressed {
      bIsDynamicallyLoaded=false
      drawAs=RoundedBox
      imageType=NoImage
      mirroring=NoMirror
      resourceName=None
      resourceObject=null
      tiling=NoTile
      imageSize=<32 32>
      margin=<ltrb 0 0 0 0>
      outlineSettings {
       bUseBrushTransparency=false
       roundingType=FixedRadius
       width=0
       color {colorUseRule=UseColor_ColorTable, specifiedColor=<rgba 0 0 0 0>}
       cornerRadii=<4 4 4 4>
      }
      tintColor {colorUseRule=UseColor_Specified, specifiedColor=<rgba 0.0284 0.0284 0.0284 1>}
      uVRegion {bIsValid=false, max=<0 0>, min=<0 0>}
     }
     pressedForeground {colorUseRule=UseColor_ColorTable, specifiedColor=<rgba 0 0 0 0>}
     pressedPadding=<ltrb 12 2.5 12 0.5>
     pressedSlateSound {resourceObject=null}
    }
   }
   slot {class=HorizontalBoxSlot, properties {horizontalAlignment=HAlign_Center, verticalAlignment=VAlign_Center, padding=<ltrb 0 0 5 0>}}
  }
  HorizontalBox {
   widgetClassPath=&HorizontalBox
   parent=AddAnimationButton
   widget=HorizontalBox
   properties {bIsEnabled=false}
   slot {class=ButtonSlot, properties {padding=<ltrb 0 0 0 0>}}
  }
  AssetImagePreviewAddButton {
   widgetClassPath=&Image
   bIsVariable=true
   parent=HorizontalBox
   widget=AssetImagePreviewAddButton
   properties {brush {
     bIsDynamicallyLoaded=false
     drawAs=Image
     imageType=FullColor
     mirroring=NoMirror
     resourceName=None
     resourceObject=&Icon_LevelSequence_add
     tiling=NoTile
     imageSize=<14 14>
     margin=<ltrb 0 0 0 0>
     outlineSettings {
      bUseBrushTransparency=false
      roundingType=HalfHeightRadius
      width=0
      color {colorUseRule=UseColor_Specified, specifiedColor=<rgba 0 0 0 0>}
      cornerRadii=<0 0 0 0>
     }
     tintColor {colorUseRule=UseColor_Specified, specifiedColor=<rgba 0.2108 1 0 1>}
     uVRegion {bIsValid=false, max=<0 0>, min=<0 0>}
    }}
   slot {class=HorizontalBoxSlot, properties {horizontalAlignment=HAlign_Center, verticalAlignment=VAlign_Center, padding=<ltrb 0 0 5 0>}}
  }
  TextBlock_1 {
   widgetClassPath=&TextBlock
   parent=HorizontalBox
   widget=TextBlock_1
   properties {
    autoWrapText=true
    justification=Center
    text=Create
    colorAndOpacity {colorUseRule=UseColor_Specified, specifiedColor=<rgba 0.5255 0.5255 0.5255 1>}
    font {
     bForceMonospaced=false
     bMaterialIsStencil=false
     fontMaterial=null
     fontObject=&Roboto
     letterSpacing=0
     monospacedWidth=1
     size=10
     skewAmount=0
     typefaceFontName=Bold
     outlineSettings {
      bApplyOutlineToDropShadows=false
      bMiteredCorners=false
      bSeparateFillAlpha=false
      outlineMaterial=null
      outlineSize=0
      outlineColor=<rgba 0 0 0 1>
     }
    }
   }
   slot {class=HorizontalBoxSlot}
  }
  Spacer_1 {
   widgetClassPath=&Spacer
   parent=HorizontalBox
   widget=Spacer_1
   properties {size=<20 20>}
   slot {class=HorizontalBoxSlot, properties {horizontalAlignment=HAlign_Left, verticalAlignment=VAlign_Center}}
  }
  AssetImageInivisibleIcon {
   widgetClassPath=&Image
   bIsVariable=true
   parent=HorizontalBox
   widget=AssetImageInivisibleIcon
   properties {brush {
     bIsDynamicallyLoaded=false
     drawAs=Image
     imageType=FullColor
     mirroring=NoMirror
     resourceName=None
     resourceObject=&/icon_layer_not_visible
     tiling=NoTile
     imageSize=<40 40>
     margin=<ltrb 0 0 0 0>
     outlineSettings {
      bUseBrushTransparency=false
      roundingType=HalfHeightRadius
      width=0
      color {colorUseRule=UseColor_Specified, specifiedColor=<rgba 0 0 0 0>}
      cornerRadii=<0 0 0 0>
     }
     tintColor {colorUseRule=UseColor_Specified, specifiedColor=<rgba 0.2108 1 0 1>}
     uVRegion {bIsValid=false, max=<0 0>, min=<0 0>}
    }}
   slot {class=HorizontalBoxSlot, properties {horizontalAlignment=HAlign_Center, verticalAlignment=VAlign_Center}}
  }
  Spacer_5 {
   widgetClassPath=&Spacer
   parent=HorizontalBox
   widget=Spacer_5
   properties {size=<20 20>}
   slot {class=HorizontalBoxSlot, properties {horizontalAlignment=HAlign_Left, verticalAlignment=VAlign_Center}}
  }
  AssetImageVisibleicon {
   widgetClassPath=&Image
   bIsVariable=true
   parent=HorizontalBox
   widget=AssetImageVisibleicon
   properties {brush {
     bIsDynamicallyLoaded=false
     drawAs=Image
     imageType=FullColor
     mirroring=NoMirror
     resourceName=None
     resourceObject=&/icon_layer_visible_16x
     tiling=NoTile
     imageSize=<16 16>
     margin=<ltrb 0 0 0 0>
     outlineSettings {
      bUseBrushTransparency=false
      roundingType=HalfHeightRadius
      width=0
      color {colorUseRule=UseColor_Specified, specifiedColor=<rgba 0 0 0 0>}
      cornerRadii=<0 0 0 0>
     }
     tintColor {colorUseRule=UseColor_Specified, specifiedColor=<rgba 0.2108 1 0 1>}
     uVRegion {bIsValid=false, max=<0 0>, min=<0 0>}
    }}
   slot {class=HorizontalBoxSlot, properties {horizontalAlignment=HAlign_Center, verticalAlignment=VAlign_Center}}
  }
  ZoomToFrames {
   widgetClassPath=&Button
   parent=INVISIBLE
   widget=ZoomToFrames
   properties {
    bIsEnabled=false
    widgetStyle {
     clickedSlateSound {resourceObject=null}
     disabled {
      bIsDynamicallyLoaded=false
      drawAs=RoundedBox
      imageType=NoImage
      mirroring=NoMirror
      resourceName=None
      resourceObject=null
      tiling=NoTile
      imageSize=<0 0>
      margin=<ltrb 0 0 0 0>
      outlineSettings {
       bUseBrushTransparency=false
       roundingType=FixedRadius
       width=1
       color {colorUseRule=UseColor_ColorTable, specifiedColor=<rgba 0 0 0 0>}
       cornerRadii=<4 4 4 4>
      }
      tintColor {colorUseRule=UseColor_Specified, specifiedColor=<rgba 0.0395 0.0395 0.0395 1>}
      uVRegion {bIsValid=false, max=<0 0>, min=<0 0>}
     }
     disabledForeground {colorUseRule=UseColor_ColorTable, specifiedColor=<rgba 0 0 0 0>}
     hovered {
      bIsDynamicallyLoaded=false
      drawAs=RoundedBox
      imageType=NoImage
      mirroring=NoMirror
      resourceName=None
      resourceObject=null
      tiling=NoTile
      imageSize=<32 32>
      margin=<ltrb 0 0 0 0>
      outlineSettings {
       bUseBrushTransparency=false
       roundingType=FixedRadius
       width=0
       color {colorUseRule=UseColor_ColorTable, specifiedColor=<rgba 0 0 0 0>}
       cornerRadii=<4 4 4 4>
      }
      tintColor {colorUseRule=UseColor_Specified, specifiedColor=<rgba 0.0395 0.0395 0.0395 0.5>}
      uVRegion {bIsValid=false, max=<0 0>, min=<0 0>}
     }
     hoveredForeground {colorUseRule=UseColor_ColorTable, specifiedColor=<rgba 0 0 0 0>}
     hoveredSlateSound {resourceObject=null}
     normal {
      bIsDynamicallyLoaded=false
      drawAs=NoDrawType
      imageType=FullColor
      mirroring=NoMirror
      resourceName=None
      resourceObject=null
      tiling=NoTile
      imageSize=<14 14>
      margin=<ltrb 0 0 0 0>
      outlineSettings {
       bUseBrushTransparency=false
       roundingType=FixedRadius
       width=1
       color {colorUseRule=UseColor_ColorTable, specifiedColor=<rgba 0 0 0 0>}
       cornerRadii=<4 4 4 4>
      }
      tintColor {colorUseRule=UseColor_Specified, specifiedColor=<rgba 0.0395 0.0395 0.0395 1>}
      uVRegion {bIsValid=false, max=<0 0>, min=<0 0>}
     }
     normalForeground {colorUseRule=UseColor_ColorTable, specifiedColor=<rgba 0 0 0 0>}
     normalPadding=<ltrb 12 1.5 12 1.5>
     pressed {
      bIsDynamicallyLoaded=false
      drawAs=RoundedBox
      imageType=NoImage
      mirroring=NoMirror
      resourceName=None
      resourceObject=null
      tiling=NoTile
      imageSize=<32 32>
      margin=<ltrb 0 0 0 0>
      outlineSettings {
       bUseBrushTransparency=false
       roundingType=FixedRadius
       width=0
       color {colorUseRule=UseColor_ColorTable, specifiedColor=<rgba 0 0 0 0>}
       cornerRadii=<4 4 4 4>
      }
      tintColor {colorUseRule=UseColor_Specified, specifiedColor=<rgba 0.0284 0.0284 0.0284 1>}
      uVRegion {bIsValid=false, max=<0 0>, min=<0 0>}
     }
     pressedForeground {colorUseRule=UseColor_ColorTable, specifiedColor=<rgba 0 0 0 0>}
     pressedPadding=<ltrb 12 2.5 12 0.5>
     pressedSlateSound {resourceObject=null}
    }
   }
   slot {class=HorizontalBoxSlot, properties {horizontalAlignment=HAlign_Center, verticalAlignment=VAlign_Center, padding=<ltrb 0 0 5 0>}}
  }
  HorizontalBox_2 {
   widgetClassPath=&HorizontalBox
   parent=ZoomToFrames
   widget=HorizontalBox_2
   slot {class=ButtonSlot, properties {padding=<ltrb 0 0 0 0>}}
  }
  TextBlock {
   widgetClassPath=&TextBlock
   parent=HorizontalBox_2
   widget=TextBlock
   properties {
    autoWrapText=true
    justification=Center
    text=SelectRange
    colorAndOpacity {colorUseRule=UseColor_Specified, specifiedColor=<rgba 0.5255 0.5255 0.5255 1>}
    font {
     bForceMonospaced=false
     bMaterialIsStencil=false
     fontMaterial=null
     fontObject=&Roboto
     letterSpacing=0
     monospacedWidth=1
     size=10
     skewAmount=0
     typefaceFontName=Bold
     outlineSettings {
      bApplyOutlineToDropShadows=false
      bMiteredCorners=false
      bSeparateFillAlpha=false
      outlineMaterial=null
      outlineSize=0
      outlineColor=<rgba 0 0 0 1>
     }
    }
   }
   slot {class=HorizontalBoxSlot}
  }
  StyleSample_Border {
   widgetClassPath=&Border
   bIsVariable=true
   parent=INVISIBLE
   widget=StyleSample_Border
   properties {
    horizontalAlignment=HAlign_Left
    verticalAlignment=VAlign_Center
    background {
     bIsDynamicallyLoaded=false
     drawAs=Border
     imageType=NoImage
     mirroring=NoMirror
     resourceName=None
     resourceObject=null
     tiling=NoTile
     imageSize=<32 32>
     margin=<ltrb 0 0 0 1>
     outlineSettings {
      bUseBrushTransparency=false
      roundingType=HalfHeightRadius
      width=0
      color {colorUseRule=UseColor_Specified, specifiedColor=<rgba 0 0 0 0>}
      cornerRadii=<0 0 0 0>
     }
     tintColor {colorUseRule=UseColor_Specified, specifiedColor=<rgba 1 1 1 1>}
     uVRegion {bIsValid=false, max=<0 0>, min=<0 0>}
    }
    brushColor=<rgba 0.1545 0.1545 0.1545 1>
    padding=<ltrb -16 2 0 2>
   }
   slot {class=HorizontalBoxSlot, properties {verticalAlignment=VAlign_Center}}
  }
  BorderCroppedSequences_1 {
   widgetClassPath=&Border
   parent=INVISIBLE
   widget=BorderCroppedSequences_1
   properties {
    bIsEnabled=false
    visibility=Collapsed
    background {
     bIsDynamicallyLoaded=false
     drawAs=Border
     imageType=NoImage
     mirroring=NoMirror
     resourceName=None
     resourceObject=null
     tiling=NoTile
     imageSize=<32 32>
     margin=<ltrb 0 0 0 1>
     outlineSettings {
      bUseBrushTransparency=false
      roundingType=HalfHeightRadius
      width=0
      color {colorUseRule=UseColor_Specified, specifiedColor=<rgba 0 0 0 0>}
      cornerRadii=<0 0 0 0>
     }
     tintColor {colorUseRule=UseColor_Specified, specifiedColor=<rgba 0.1302 0.1302 0.1302 1>}
     uVRegion {bIsValid=false, max=<0 0>, min=<0 0>}
    }
    padding=<ltrb 0 10 0 0>
   }
   slot {class=HorizontalBoxSlot}
  }
  TextSampleHeader {
   widgetClassPath=&TextBlock
   bIsVariable=true
   parent=BorderCroppedSequences_1
   widget=TextSampleHeader
   properties {
    text=Cropped Linked Anim Sequences:
    toolTipText="List of all cropped* animation sequences that are currently linked to this Level Sequnce.\r\n\r\n*cropped - means they export frame range of a timeline, not the whole timeline. For ex. frames 30-50, while whole timeline is 0-100. "
    font {
     bForceMonospaced=false
     bMaterialIsStencil=false
     fontMaterial=null
     fontObject=&Roboto
     letterSpacing=0
     monospacedWidth=1
     size=9
     skewAmount=0
     typefaceFontName=Bold
     outlineSettings {
      bApplyOutlineToDropShadows=false
      bMiteredCorners=false
      bSeparateFillAlpha=false
      outlineMaterial=null
      outlineSize=0
      outlineColor=<rgba 0 0 0 1>
     }
    }
   }
   slot {class=BorderSlot, properties {padding=<ltrb 0 10 0 0>}}
  }
  ModalLayer {
   widgetClassPath=&NamedSlot
   bIsVariable=true
   parent=Overlay_0
   widget=ModalLayer
   properties {bExposeOnInstanceOnly=true}
   slot {class=OverlaySlot, properties {horizontalAlignment=HAlign_Fill, verticalAlignment=VAlign_Fill}}
  }
  Border_Loading {
   widgetClassPath=&Border
   bIsVariable=true
   parent=Overlay_0
   widget=Border_Loading
   properties {horizontalAlignment=HAlign_Center, verticalAlignment=VAlign_Center, brushColor=<rgba 0 0 0 0.8435>}
   slot {class=OverlaySlot, properties {horizontalAlignment=HAlign_Fill, verticalAlignment=VAlign_Fill}}
  }
  HorizontalBox_0 {
   widgetClassPath=&HorizontalBox
   parent=Border_Loading
   widget=HorizontalBox_0
   slot {class=BorderSlot, properties {horizontalAlignment=HAlign_Center, verticalAlignment=VAlign_Center}}
  }
  TextStyleSammple_Bold_Big_1 {
   widgetClassPath=&TextBlock
   bIsVariable=true
   parent=HorizontalBox_0
   widget=TextStyleSammple_Bold_Big_1
   properties {
    text=Loading
    font {
     bForceMonospaced=false
     bMaterialIsStencil=false
     fontMaterial=null
     fontObject=&Roboto
     letterSpacing=100
     monospacedWidth=1
     size=10.5
     skewAmount=0
     typefaceFontName=Bold
     outlineSettings {
      bApplyOutlineToDropShadows=false
      bMiteredCorners=false
      bSeparateFillAlpha=false
      outlineMaterial=null
      outlineSize=0
      outlineColor=<rgba 0 0 0 1>
     }
    }
   }
   slot {class=HorizontalBoxSlot, properties {horizontalAlignment=HAlign_Center, verticalAlignment=VAlign_Center}}
  }
 ]
}
