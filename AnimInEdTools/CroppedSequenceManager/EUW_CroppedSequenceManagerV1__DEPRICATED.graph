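# Deprecated editor utility widget (Cropped Sequence Manager V1) for exporting Level Sequence tracks as AnimSequences, with Standard, Aim Offset and Single Segment modes.
# Buttons and options delegate to the CroppedSequenceManagerV1 Python module, which Construct reloads and hands widget references.
# Scan and most actions rerun init_tool, refreshing the Level Sequence and BlendSpace.
# Curve compression choice persists through a SaveGame slot; Destruct also reverts temporary renames and clears temp subfolders.
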EditorUtilityWidgetBlueprint /AnimInEdTools/CroppedSequenceManager/EUW_CroppedSequenceManagerV1__DEPRICATED
graphs [
 Dispatch {nodes [Dispatch() {is=title K2Node_FunctionEntry}]}
 EventGraph {nodes [
   n1 = "On Clicked (Button_Scan)"() {is=title K2Node_ComponentBoundEvent}
   n2 = Execute Python Script!(
    execute=[@n11.then, @n1.then, @n10.then, @n13.then, @n4.then, @n16.then_0]
    PythonScript="import sys\r\nCSM = sys.modules['CroppedSequenceManagerV1']\r\n\r\nblend_space = CSM.cls.init_tool()"
   ) -> (then, ReturnValue, blend_space:Blend Space Object Reference) {outputs [blend_space]}
   > n3 = GetCurrentLevelSequence!(execute) -> (then, ReturnValue:Level Sequence Object Reference)
   > Set Current Level Sequence(Current Level Sequence:Level Sequence Object Reference=@n3) -> (Output_Get:Level Sequence Object Reference)
   > Set BlendSpace(BlendSpace:Blend Space Object Reference=@n2.blend_space) -> (Output_Get:Blend Space Object Reference)
   "On Clicked (Button_Export)"() {is=title K2Node_ComponentBoundEvent}
   > n4 = Execute Python Script!(execute, PythonScript="import sys\r\nCSM = sys.modules['CroppedSequenceManagerV1']\r\n\r\nCSM.cls.export()") -> (then, ReturnValue)
   n5 = "On Selection Changed (ComboBox_CurveCompression)"!() -> (OutputDelegate, then, SelectedItem:String, SelectionType:ESelectInfo Enum) {is=title K2Node_ComponentBoundEvent}
   > Execute Python Script!(
    execute
    PythonScript="import sys\r\nCSM = sys.modules['CroppedSequenceManagerV1']\r\n\r\nCSM.cls.set_anim_curve_compression_settings(current_project)"
    current_project:String=@n5.SelectedItem
   ) -> (then, ReturnValue) {inputs [current_project]}
   n6 = "On Check State Changed (checkbox_no_ls)"!() -> (OutputDelegate, then, bIsChecked:bool) {is=title K2Node_ComponentBoundEvent}
   > n7 = Branch(Condition=@n6.bIsChecked)
   n8 = Get TextBlock_18() -> (TextBlock_18:Text Object Reference)
   > SetVisibility(self=@n8, InVisibility=Hidden)
   SetVisibility(execute=@n7.else, self=@n8, InVisibility=Visible)
   "On Clicked (Button_AddTrack)"() {is=title K2Node_ComponentBoundEvent}
   > Execute Python Script!(execute, PythonScript="import sys\r\nCSM = sys.modules['CroppedSequenceManagerV1']\r\n\r\nCSM.cls.expand_track_options()") -> (then, ReturnValue)
   n9 = "On Clicked (Button_AddTrack_Cancel)"() {is=title K2Node_ComponentBoundEvent}
   Execute Python Script(
    execute=[@n9.then, @n12.then]
    PythonScript="import sys\r\nCSM = sys.modules['CroppedSequenceManagerV1']\r\n\r\nCSM.cls.collapse_track_options()"
   )
   n10 = "On Selection Changed (ComboBox_ToolMode)"!() -> (OutputDelegate, then, SelectedItem:String, SelectionType:ESelectInfo Enum) {is=title K2Node_ComponentBoundEvent}
   "On Property Changed (SinglePropertyView_LevelSequence)"!() -> (OutputDelegate, then, PropertyName:Name=None) {is=title K2Node_ComponentBoundEvent}
   > n11 = OpenLevelSequence!(
    execute
    LevelSequence:Level Sequence Object Reference=Get Current Level Sequence() -> (Current Level Sequence:Level Sequence Object Reference)
   ) -> (then, ReturnValue:bool)
   "On Clicked (Button_AddTrack_Confirm)"() {is=title K2Node_ComponentBoundEvent}
   > n12 = Execute Python Script(PythonScript="import sys\r\nCSM = sys.modules['CroppedSequenceManagerV1']\r\n\r\nCSM.cls.link_track_to_animation_sequence()")
   "On Clicked (Button_Save)"() {is=title K2Node_ComponentBoundEvent}
   > n13 = Execute Python Script(PythonScript="import sys\r\nCSM = sys.modules['CroppedSequenceManagerV1']\r\n\r\nCSM.cls.save()")
   "On Clicked (Button_?)"() {is=title K2Node_ComponentBoundEvent}
   > LaunchURL!(execute, URL:String=https://confluence.it.epicgames.com/display/PP/Cropped+Sequence+Manager) -> (then)
   n14 = "On Check State Changed (ForceCreate)"!() -> (OutputDelegate, then, bIsChecked:bool) {is=title K2Node_ComponentBoundEvent}
   > Execute Python Script!(
    execute
    PythonScript="import sys\r\nCSM = sys.modules['CroppedSequenceManagerV1']\r\n\r\nCSM.cls.set_force_export(is_checked)"
    is_checked:bool=@n14.bIsChecked
   ) -> (then, ReturnValue) {inputs [is_checked]}
   Event PreConstruct()
   n15 = Self-Reference()
   > SetObject(self=Get SinglePropertyView_LevelSequence() -> (SinglePropertyView_LevelSequence:Single Property View Object Reference), NewObject=@n15)
   > SetObject(self=Get SinglePropertyView_BlendSpace() -> (SinglePropertyView_BlendSpace:Single Property View Object Reference), NewObject=@n15)
   Event Construct()
   > Execute Python Script!(execute, PythonScript="import CroppedSequenceManagerV1\r\nimport importlib\r\nimportlib.reload(CroppedSequenceManagerV1)") -> (then, ReturnValue)
   > Execute Python Script!(
    execute
    PythonScript="import sys\r\nCSM = sys.modules['CroppedSequenceManagerV1']\r\n\r\nCSM.cls.init_widget_export_button(button_export)\r\nCSM.cls.init_widget_save_button(button_save)\r\nCSM.cls.init_widget_scan_button(button_scan)\r\nCSM.cls.init_widget_layouts(layout_core, layout_1, layout_2, layout_track_options)\r\nCSM.cls.init_widget_combo_box_tool_mode(combo_box_tool_mode)\r\nCSM.cls.init_fonts(text_regular, text_bold_big, text_add_track)\r\nCSM.cls.init_text_edit_style(text_edit)\r\nCSM.cls.init_checkbox_style(checkbox_style)\r\nCSM.cls.init_animation_types(animation_types)\r\nCSM.cls.init_widget_no_ls_checkbox(checkbox_no_ls)"
    button_export:Editor Utility Button Object Reference=Get Button_Export() -> (Button_Export:Editor Utility Button Object Reference)
    layout_core:Vertical Box Object Reference=Get CoreLayout() -> (CoreLayout:Vertical Box Object Reference)
    button_save:Editor Utility Button Object Reference=Get Button_Save() -> (Button_Save:Editor Utility Button Object Reference)
    layout_1:Vertical Box Object Reference=Get Layout1() -> (Layout1:Vertical Box Object Reference)
    text_regular:Text Object Reference=Get TextStyleSammple_Regular() -> (TextStyleSammple_Regular:Text Object Reference)
    text_edit:Editor Utility Editable Text Object Reference=Get EditableText_Style() -> (EditableText_Style:Editor Utility Editable Text Object Reference)
    checkbox_style:Editor Utility Check Box Object Reference=Get Checkbox_Style() -> (Checkbox_Style:Editor Utility Check Box Object Reference)
    button_scan:Editor Utility Button Object Reference=Get Button_Scan() -> (Button_Scan:Editor Utility Button Object Reference)
    layout_2:Vertical Box Object Reference=Get Layout2() -> (Layout2:Vertical Box Object Reference)
    text_bold_big:Text Object Reference=Get TextStyleSammple_Bold_Big() -> (TextStyleSammple_Bold_Big:Text Object Reference)
    layout_track_options:Vertical Box Object Reference=Get Layout1b() -> (Layout1b:Vertical Box Object Reference)
    combo_box_tool_mode:Editor Utility Combo Box String Object Reference=Get ComboBox_ToolMode() -> (ComboBox_ToolMode:Editor Utility Combo Box String Object Reference)
    text_add_track:Text Object Reference=Get Add_Track_Text() -> (Add_Track_Text:Text Object Reference)
    animation_types:Map of Strings to Struct Animation Browser Animation Type Structures=Get Animation Types!(self:DA Animation Browser Object Reference=Get Data Asset() -> (Data Asset:DA Animation Browser Object Reference)) -> (Animation Types:Map of Strings to Struct Animation Browser Animation Type Structures)
    checkbox_no_ls:Editor Utility Check Box Object Reference=Get checkbox_no_ls() -> (checkbox_no_ls:Editor Utility Check Box Object Reference)
   ) -> (then, ReturnValue) {inputs [
     button_export
     layout_core
     button_save
     layout_1
     text_regular
     text_edit
     checkbox_style
     button_scan
     layout_2
     text_bold_big
     layout_track_options
     combo_box_tool_mode
     text_add_track
     animation_types
     checkbox_no_ls
    ]}
   > Execute Python Script!(
    execute
    PythonScript="import sys\r\nCSM = sys.modules['CroppedSequenceManagerV1']\r\n\r\nCSM.cls.init_widget_combo_box_tracks(combo_box_tracks)\r\nCSM.cls.init_track_option_widgets(e_text_track_path,  e_text_track_new_prefix, e_text_track_preview)\r\nCSM.cls.init_widget_layout_add_track(layout)"
    combo_box_tracks:Editor Utility Combo Box String Object Reference=Get ComboBox_AddTrack() -> (ComboBox_AddTrack:Editor Utility Combo Box String Object Reference)
    e_text_track_path:Editor Utility Editable Text Object Reference=Get EditableText_Track_Path() -> (EditableText_Track_Path:Editor Utility Editable Text Object Reference)
    e_text_track_new_prefix:Editor Utility Editable Text Object Reference=Get EditableText_New_Prefix() -> (EditableText_New_Prefix:Editor Utility Editable Text Object Reference)
    e_text_track_preview:Editor Utility Editable Text Object Reference=Get EditableText_Track_Preview() -> (EditableText_Track_Preview:Editor Utility Editable Text Object Reference)
    layout:Horizontal Box Object Reference=Get HorizontalBoxAddTrack() -> (HorizontalBoxAddTrack:Horizontal Box Object Reference)
   ) -> (then, ReturnValue) {inputs [combo_box_tracks, e_text_track_path, e_text_track_new_prefix, e_text_track_preview, layout]}
   > Execute Python Script!(
    execute
    PythonScript="import sys\r\nCSM = sys.modules['CroppedSequenceManagerV1']\r\n\r\nCSM.cls.init_widget_layout_blendspace(layout)\r\nCSM.cls.init_widget_combo_box_anim_type(combo_box)"
    layout:Vertical Box Object Reference=Get HorizontalBoxBlendSpace() -> (HorizontalBoxBlendSpace:Vertical Box Object Reference)
    combo_box:Editor Utility Combo Box String Object Reference=Get ComboBox_AnimType() -> (ComboBox_AnimType:Editor Utility Combo Box String Object Reference)
   ) -> (then, ReturnValue) {inputs [layout,combo_box]}
   > n16 = Sequence()
   n17 = LoadGameFromSlot!(execute=@n16.then_1, SlotName:String=CroppedSequencemanager_V1, UserIndex:int=0) -> (then, ReturnValue:Save Game Object Reference)
   > n18 = Cast To EUW_CroppedSequenceManager_SaveData!(execute, Object=@n17) -> (then, CastFailed, AsEUW Cropped Sequence Manager Save Data:EUW Cropped Sequence Manager Save Data Object Reference) {is=title K2Node_DynamicCast}
   n19 = Get CurveCompression!(self:EUW Cropped Sequence Manager Save Data Object Reference=@n18) -> (CurveCompression:String)
   > SetSelectedOption!(
    execute
    self:"Combo Box (String) Object Reference"=Get ComboBox_CurveCompression() -> (ComboBox_CurveCompression:Editor Utility Combo Box String Object Reference)
    Option:String=@n19
   ) -> (then)
   > Execute Python Script!(
    execute
    PythonScript="import sys\r\nCSM = sys.modules['CroppedSequenceManagerV1']\r\n\r\nCSM.cls.set_anim_curve_compression_settings(current_project)"
    current_project:String=@n19
   ) -> (then, ReturnValue) {inputs [current_project]}
   Event Destruct()
   > n20 = Sequence()
   > Execute Python Script!(
    execute
    PythonScript="import sys\r\nCSM = sys.modules['CroppedSequenceManagerV1']\r\n\r\nCSM.cls.CSRenamer.revert_name(False)\r\nCSM.cls.CSRenamer.clear_temp_sub_folders()"
   ) -> (then, ReturnValue)
   n21 = CreateSaveGameObject!(
    execute=@n20.then_1
    SaveGameClass:Save Game Class Reference=&/EUW_CroppedSequenceManager_SaveData.EUW_CroppedSequenceManager_SaveData_C
   ) -> (then, ReturnValue:EUW Cropped Sequence Manager Save Data Object Reference)
   > Set CurveCompression!(
    execute
    CurveCompression:String=GetSelectedOption!(self:"Combo Box (String) Object Reference"=Get ComboBox_CurveCompression()) -> (ReturnValue:String)
    self:EUW Cropped Sequence Manager Save Data Object Reference=@n21
   ) -> (then, Output_Get:String)
   > AsyncSaveGameToSlot!(
    execute:Exec
    WorldContextObject:Object Reference
    SaveGameObject:Save Game Object Reference=@n21
    SlotName:String=CroppedSequencemanager_V1
    UserIndex:int=0
   ) -> (then:Exec, Completed:Exec, SaveGame:Save Game Object Reference, bSuccess:bool=~) {is=title K2Node_AsyncAction}
   Bind Event to On Post Save World(self:Level Editor Subsystem Object Reference=Get LevelEditorSubsystem()) {is=title K2Node_AddDelegate}
   Event Tick()
  ]}
 LevelSaved {nodes [LevelSaved() {is=title K2Node_FunctionEntry}]}
]
variables [
 BlendSpace {type=BlendSpace, default=null}
 BoneCompressionSettings {type=Object, default=null}
 Current Level Sequence {type=LevelSequence, default=null}
 Data Asset {type=DA_AnimationBrowser_C, default=/AnimInEdTools/AnimationBrowser/DA_Play_AnimationBrowser}
 Dispatch {default=()}
 LevelSaved {default=()}
]
widgets {
 info {rootWidgetClass=&EditorUtilityScrollBox, widgetCount=86}
 widgets [
  EditorUtilityScrollBox_1 {widgetClassPath=&EditorUtilityScrollBox, widget=EditorUtilityScrollBox_1}
  CoreLayout {
   widgetClassPath=&VerticalBox
   bIsVariable=true
   parent=EditorUtilityScrollBox_1
   widget=CoreLayout
   properties {renderTransformPivot=<0 0>}
   slot {class=ScrollBoxSlot, properties {verticalAlignment=VAlign_Top, padding=<ltrb 4 0 4 4>, size {sizeRule=Fill, value=1}}}
  }
  HorizontalBox_1 {
   widgetClassPath=&HorizontalBox
   parent=CoreLayout
   widget=HorizontalBox_1
   slot {class=VerticalBoxSlot, properties {verticalAlignment=VAlign_Center}}
  }
  Spacer_6 {
   widgetClassPath=&Spacer
   parent=HorizontalBox_1
   widget=Spacer_6
   properties {size=<1 0>}
   slot {class=HorizontalBoxSlot, properties {verticalAlignment=VAlign_Top, size {sizeRule=Fill, value=1}}}
  }
  SinglePropertyView_LevelSequence {
   widgetClassPath=&SinglePropertyView
   parent=HorizontalBox_1
   widget=SinglePropertyView_LevelSequence
   properties {propertyName=Current Level Sequence}
   slot {class=HorizontalBoxSlot}
  }
  Spacer_7 {
   widgetClassPath=&Spacer
   parent=HorizontalBox_1
   widget=Spacer_7
   properties {size=<1 0>}
   slot {class=HorizontalBoxSlot, properties {verticalAlignment=VAlign_Top, size {sizeRule=Fill, value=1}}}
  }
  Button_? {
   widgetClassPath=&EditorUtilityButton
   parent=HorizontalBox_1
   widget=Button_?
   properties {
    toolTipText=Open Confuelnce Page
    widgetStyle {
     clickedSlateSound {resourceObject=null}
     disabled {
      bIsDynamicallyLoaded=false
      drawAs=RoundedBox
      imageType=NoImage
      mirroring=NoMirror
      resourceName=None
      resourceObject=null
      tiling=NoTile
      imageSize=<0 0>
      margin=<ltrb 0 0 0 0>
      outlineSettings {
       bUseBrushTransparency=false
       roundingType=FixedRadius
       width=1
       color {colorUseRule=UseColor_ColorTable, specifiedColor=<rgba 0 0 0 0>}
       cornerRadii=<4 4 4 4>
      }
      tintColor {colorUseRule=UseColor_Specified, specifiedColor=<rgba 0.0395 0.0395 0.0395 1>}
      uVRegion {bIsValid=false, max=<0 0>, min=<0 0>}
     }
     disabledForeground {colorUseRule=UseColor_ColorTable, specifiedColor=<rgba 0 0 0 0>}
     hovered {
      bIsDynamicallyLoaded=false
      drawAs=RoundedBox
      imageType=NoImage
      mirroring=NoMirror
      resourceName=None
      resourceObject=null
      tiling=NoTile
      imageSize=<32 32>
      margin=<ltrb 0 0 0 0>
      outlineSettings {
       bUseBrushTransparency=false
       roundingType=FixedRadius
       width=1
       color {colorUseRule=UseColor_ColorTable, specifiedColor=<rgba 0 0 0 0>}
       cornerRadii=<4 4 4 4>
      }
      tintColor {colorUseRule=UseColor_Specified, specifiedColor=<rgba 0.0953 0.0953 0.0953 1>}
      uVRegion {bIsValid=false, max=<0 0>, min=<0 0>}
     }
     hoveredForeground {colorUseRule=UseColor_ColorTable, specifiedColor=<rgba 0 0 0 0>}
     hoveredSlateSound {resourceObject=null}
     normal {
      bIsDynamicallyLoaded=false
      drawAs=RoundedBox
      imageType=NoImage
      mirroring=NoMirror
      resourceName=None
      resourceObject=null
      tiling=NoTile
      imageSize=<32 32>
      margin=<ltrb 0 0 0 0>
      outlineSettings {
       bUseBrushTransparency=false
       roundingType=FixedRadius
       width=1
       color {colorUseRule=UseColor_ColorTable, specifiedColor=<rgba 0 0 0 0>}
       cornerRadii=<4 4 4 4>
      }
      tintColor {colorUseRule=UseColor_Specified, specifiedColor=<rgba 0.0395 0.0395 0.0395 1>}
      uVRegion {bIsValid=false, max=<0 0>, min=<0 0>}
     }
     normalForeground {colorUseRule=UseColor_ColorTable, specifiedColor=<rgba 0 0 0 0>}
     normalPadding=<ltrb 8 2 6 2>
     pressed {
      bIsDynamicallyLoaded=false
      drawAs=RoundedBox
      imageType=NoImage
      mirroring=NoMirror
      resourceName=None
      resourceObject=null
      tiling=NoTile
      imageSize=<32 32>
      margin=<ltrb 0 0 0 0>
      outlineSettings {
       bUseBrushTransparency=false
       roundingType=FixedRadius
       width=1
       color {colorUseRule=UseColor_ColorTable, specifiedColor=<rgba 0 0 0 0>}
       cornerRadii=<4 4 4 4>
      }
      tintColor {colorUseRule=UseColor_Specified, specifiedColor=<rgba 0.0284 0.0284 0.0284 1>}
      uVRegion {bIsValid=false, max=<0 0>, min=<0 0>}
     }
     pressedForeground {colorUseRule=UseColor_ColorTable, specifiedColor=<rgba 0 0 0 0>}
     pressedPadding=<ltrb 8 2.5 6 2.5>
     pressedSlateSound {resourceObject=null}
    }
   }
   slot {class=HorizontalBoxSlot, properties {horizontalAlignment=HAlign_Center, verticalAlignment=VAlign_Center, padding=<ltrb 4 0 10 0>}}
  }
  TextBlock_14 {
   widgetClassPath=&TextBlock
   parent=Button_?
   widget=TextBlock_14
   properties {
    text=?
    colorAndOpacity {colorUseRule=UseColor_Specified, specifiedColor=<rgba 0 1 0 1>}
    font {
     bForceMonospaced=false
     bMaterialIsStencil=false
     fontMaterial=null
     fontObject=&Roboto
     letterSpacing=0
     monospacedWidth=1
     size=12
     skewAmount=0
     typefaceFontName=Bold
     outlineSettings {
      bApplyOutlineToDropShadows=false
      bMiteredCorners=false
      bSeparateFillAlpha=false
      outlineMaterial=null
      outlineSize=0
      outlineColor=<rgba 0 0 0 1>
     }
    }
   }
   slot {class=ButtonSlot, properties {horizontalAlignment=HAlign_Left, padding=<ltrb 0 0 0 0>}}
  }
  VerticallBoxDepricatd {
   widgetClassPath=&VerticalBox
   parent=CoreLayout
   widget=VerticallBoxDepricatd
   properties {bOverride_Cursor=true}
   slot {class=VerticalBoxSlot, properties {horizontalAlignment=HAlign_Center, verticalAlignment=VAlign_Top}}
  }
  TextBlock_20 {
   widgetClassPath=&TextBlock
   parent=VerticallBoxDepricatd
   widget=TextBlock_20
   properties {
    text=This tool is depricated and will be premanently deleted soon.
    toolTipText=Contact Maksym Zhuravlov if any questions
    colorAndOpacity {colorUseRule=UseColor_Specified, specifiedColor=<rgba 1 0 0 1>}
    font {
     bForceMonospaced=false
     bMaterialIsStencil=false
     fontMaterial=null
     fontObject=&Roboto
     letterSpacing=0
     monospacedWidth=1
     size=9
     skewAmount=0
     typefaceFontName=Bold
     outlineSettings {
      bApplyOutlineToDropShadows=false
      bMiteredCorners=false
      bSeparateFillAlpha=false
      outlineMaterial=null
      outlineSize=0
      outlineColor=<rgba 0 0 0 1>
     }
    }
   }
   slot {class=VerticalBoxSlot, properties {horizontalAlignment=HAlign_Center, verticalAlignment=VAlign_Center, padding=<ltrb 0 5 0 0>}}
  }
  TextBlock_21 {
   widgetClassPath=&TextBlock
   parent=VerticallBoxDepricatd
   widget=TextBlock_21
   properties {
    justification=Center
    text="\r\nUse '&/EUW_CroppedSequenceManager' instead.\r\nTop Menu shortcut: FN Tools > FN1 Anim Tools > Cropped Sequence Manager"
    toolTipText=Contact Maksym Zhuravlov if any questions
    colorAndOpacity {colorUseRule=UseColor_Specified, specifiedColor=<rgba 1 0 0 1>}
    font {
     bForceMonospaced=false
     bMaterialIsStencil=false
     fontMaterial=null
     fontObject=&Roboto
     letterSpacing=0
     monospacedWidth=1
     size=9
     skewAmount=0
     typefaceFontName=Regular
     outlineSettings {
      bApplyOutlineToDropShadows=false
      bMiteredCorners=false
      bSeparateFillAlpha=false
      outlineMaterial=null
      outlineSize=0
      outlineColor=<rgba 0 0 0 1>
     }
    }
   }
   slot {class=VerticalBoxSlot, properties {horizontalAlignment=HAlign_Center, verticalAlignment=VAlign_Center, padding=<ltrb 0 5 0 10>}}
  }
  Border {
   widgetClassPath=&Border
   parent=CoreLayout
   widget=Border
   properties {
    background {
     bIsDynamicallyLoaded=false
     drawAs=Border
     imageType=NoImage
     mirroring=NoMirror
     resourceName=None
     resourceObject=null
     tiling=NoTile
     imageSize=<32 32>
     margin=<ltrb 0 0 0 1>
     outlineSettings {
      bUseBrushTransparency=false
      roundingType=HalfHeightRadius
      width=0
      color {colorUseRule=UseColor_Specified, specifiedColor=<rgba 0 0 0 0>}
      cornerRadii=<0 0 0 0>
     }
     tintColor {colorUseRule=UseColor_Specified, specifiedColor=<rgba 0.1302 0.1302 0.1302 1>}
     uVRegion {bIsValid=false, max=<0 0>, min=<0 0>}
    }
    padding=<ltrb 0 0 0 0>
   }
   slot {class=VerticalBoxSlot}
  }
  TextBlock_2 {
   widgetClassPath=&TextBlock
   parent=Border
   widget=TextBlock_2
   properties {
    text=Export Tracks:
    toolTipText="Using Checkboxed select one or several criterias to scan for.\r\n"
    font {
     bForceMonospaced=false
     bMaterialIsStencil=false
     fontMaterial=null
     fontObject=&Roboto
     letterSpacing=0
     monospacedWidth=1
     size=9
     skewAmount=0
     typefaceFontName=Bold
     outlineSettings {
      bApplyOutlineToDropShadows=false
      bMiteredCorners=false
      bSeparateFillAlpha=false
      outlineMaterial=null
      outlineSize=0
      outlineColor=<rgba 0 0 0 1>
     }
    }
   }
   slot {class=BorderSlot, properties {padding=<ltrb 0 0 0 0>}}
  }
  HorizontalBox_Mode {
   widgetClassPath=&HorizontalBox
   parent=CoreLayout
   widget=HorizontalBox_Mode
   slot {class=VerticalBoxSlot, properties {horizontalAlignment=HAlign_Right, verticalAlignment=VAlign_Top}}
  }
  TextBlock_3 {
   widgetClassPath=&TextBlock
   parent=HorizontalBox_Mode
   widget=TextBlock_3
   properties {
    text=Mode:
    toolTipText="Using Checkboxed select one or several criterias to scan for.\r\n"
    font {
     bForceMonospaced=false
     bMaterialIsStencil=false
     fontMaterial=null
     fontObject=&Roboto
     letterSpacing=0
     monospacedWidth=1
     size=9
     skewAmount=0
     typefaceFontName=Bold
     outlineSettings {
      bApplyOutlineToDropShadows=false
      bMiteredCorners=false
      bSeparateFillAlpha=false
      outlineMaterial=null
      outlineSize=0
      outlineColor=<rgba 0 0 0 1>
     }
    }
   }
   slot {class=HorizontalBoxSlot, properties {verticalAlignment=VAlign_Center}}
  }
  ComboBox_ToolMode {
   widgetClassPath=/Script/Blutility.EditorUtilityComboBoxString
   bIsVariable=true
   parent=HorizontalBox_Mode
   widget=ComboBox_ToolMode
   properties {
    selectedOption=Standard
    contentPadding=<ltrb 4 2 10 2>
    defaultOptions [Standard, Aim Offset, Single Segment]
    font {
     bForceMonospaced=false
     bMaterialIsStencil=false
     fontMaterial=null
     fontObject=&Roboto
     letterSpacing=0
     monospacedWidth=1
     size=9
     skewAmount=0
     typefaceFontName=Regular
     outlineSettings {
      bApplyOutlineToDropShadows=false
      bMiteredCorners=false
      bSeparateFillAlpha=false
      outlineMaterial=null
      outlineSize=0
      outlineColor=<rgba 0 0 0 1>
     }
    }
    foregroundColor {colorUseRule=UseColor_Specified, specifiedColor=<rgba 0.5271 0.5271 0.5271 1>}
   }
   slot {class=HorizontalBoxSlot, properties {horizontalAlignment=HAlign_Right, verticalAlignment=VAlign_Center}}
  }
  Layout1 {
   widgetClassPath=&VerticalBox
   bIsVariable=true
   parent=CoreLayout
   widget=Layout1
   slot {class=VerticalBoxSlot, properties {verticalAlignment=VAlign_Top, padding=<ltrb 20 12 0 8>}}
  }
  Spacer_3 {
   widgetClassPath=&Spacer
   parent=CoreLayout
   widget=Spacer_3
   properties {size=<0 4>}
   slot {class=VerticalBoxSlot, properties {verticalAlignment=VAlign_Top}}
  }
  Layout1a {
   widgetClassPath=&HorizontalBox
   parent=CoreLayout
   widget=Layout1a
   slot {class=VerticalBoxSlot, properties {horizontalAlignment=HAlign_Left, verticalAlignment=VAlign_Center, padding=<ltrb 20 0 0 0>}}
  }
  HorizontalBoxAddTrack {
   widgetClassPath=&HorizontalBox
   bIsVariable=true
   parent=Layout1a
   widget=HorizontalBoxAddTrack
   slot {class=HorizontalBoxSlot, properties {horizontalAlignment=HAlign_Left, verticalAlignment=VAlign_Top}}
  }
  Add_Track_Text {
   widgetClassPath=&TextBlock
   bIsVariable=true
   parent=HorizontalBoxAddTrack
   widget=Add_Track_Text
   properties {
    text=Add Track:
    toolTipText="Setup Path, Rename and press CREATE button. "
    colorAndOpacity {colorUseRule=UseColor_Specified, specifiedColor=<rgba 0.1545 0.1545 0.1545 1>}
    font {
     bForceMonospaced=false
     bMaterialIsStencil=false
     fontMaterial=null
     fontObject=&Roboto
     letterSpacing=0
     monospacedWidth=1
     size=9
     skewAmount=0
     typefaceFontName=Bold
     outlineSettings {
      bApplyOutlineToDropShadows=false
      bMiteredCorners=false
      bSeparateFillAlpha=false
      outlineMaterial=null
      outlineSize=0
      outlineColor=<rgba 0 0 0 1>
     }
    }
   }
   slot {class=HorizontalBoxSlot, properties {horizontalAlignment=HAlign_Left, verticalAlignment=VAlign_Center, padding=<ltrb 0 0 12 0>}}
  }
  ComboBox_AddTrack {
   widgetClassPath=&EditorUtilityComboBoxString
   bIsVariable=true
   parent=HorizontalBoxAddTrack
   widget=ComboBox_AddTrack
   properties {
    toolTipText=Select Track to add to exporting and press + button.
    contentPadding=<ltrb 0 0 0 0>
    font {
     bForceMonospaced=false
     bMaterialIsStencil=false
     fontMaterial=null
     fontObject=&Roboto
     letterSpacing=0
     monospacedWidth=1
     size=9
     skewAmount=0
     typefaceFontName=Regular
     outlineSettings {
      bApplyOutlineToDropShadows=false
      bMiteredCorners=false
      bSeparateFillAlpha=false
      outlineMaterial=null
      outlineSize=0
      outlineColor=<rgba 0 0 0 1>
     }
    }
    foregroundColor {colorUseRule=UseColor_Specified, specifiedColor=<rgba 0.5271 0.5271 0.5271 1>}
   }
   slot {class=HorizontalBoxSlot}
  }
  Button_AddTrack {
   widgetClassPath=&EditorUtilityButton
   parent=HorizontalBoxAddTrack
   widget=Button_AddTrack
   properties {toolTipText="Adds selected in drop down list track as an exportable.\r\n\r\nMeant to be used for props."}
   slot {class=HorizontalBoxSlot, properties {horizontalAlignment=HAlign_Center, verticalAlignment=VAlign_Top, padding=<ltrb 4 0 10 0>}}
  }
  TextBlock_5 {
   widgetClassPath=&TextBlock
   parent=Button_AddTrack
   widget=TextBlock_5
   properties {
    text=+
    font {
     bForceMonospaced=false
     bMaterialIsStencil=false
     fontMaterial=null
     fontObject=&Roboto
     letterSpacing=0
     monospacedWidth=1
     size=12
     skewAmount=0
     typefaceFontName=Regular
     outlineSettings {
      bApplyOutlineToDropShadows=false
      bMiteredCorners=false
      bSeparateFillAlpha=false
      outlineMaterial=null
      outlineSize=0
      outlineColor=<rgba 0 0 0 1>
     }
    }
   }
   slot {class=ButtonSlot, properties {horizontalAlignment=HAlign_Left, padding=<ltrb 0 0 0 0>}}
  }
  Layout1b {
   widgetClassPath=&VerticalBox
   bIsVariable=true
   parent=Layout1a
   widget=Layout1b
   properties {visibility=Collapsed}
   slot {class=HorizontalBoxSlot, properties {horizontalAlignment=HAlign_Left, verticalAlignment=VAlign_Top, padding=<ltrb 10 6 0 0>}}
  }
  TextBlock_9 {
   widgetClassPath=&TextBlock
   parent=Layout1b
   widget=TextBlock_9
   properties {
    text=Export Track Options:
    toolTipText="Setup Path, Rename and press CREATE button. "
    colorAndOpacity {colorUseRule=UseColor_Specified, specifiedColor=<rgba 0.1545 0.1545 0.1545 1>}
    font {
     bForceMonospaced=false
     bMaterialIsStencil=false
     fontMaterial=null
     fontObject=&Roboto
     letterSpacing=0
     monospacedWidth=1
     size=9
     skewAmount=0
     typefaceFontName=Bold
     outlineSettings {
      bApplyOutlineToDropShadows=false
      bMiteredCorners=false
      bSeparateFillAlpha=false
      outlineMaterial=null
      outlineSize=0
      outlineColor=<rgba 0 0 0 1>
     }
    }
   }
   slot {class=VerticalBoxSlot, properties {horizontalAlignment=HAlign_Left, verticalAlignment=VAlign_Center, padding=<ltrb 0 0 12 0>}}
  }
  HorizontalBox_4 {
   widgetClassPath=&HorizontalBox
   parent=Layout1b
   widget=HorizontalBox_4
   slot {class=VerticalBoxSlot, properties {horizontalAlignment=HAlign_Left, verticalAlignment=VAlign_Center, padding=<ltrb 10 4 0 0>}}
  }
  Layout1b1 {widgetClassPath=&VerticalBox, parent=HorizontalBox_4, widget=Layout1b1, slot {class=HorizontalBoxSlot}}
  TextBlock_6 {
   widgetClassPath=&TextBlock
   parent=Layout1b1
   widget=TextBlock_6
   properties {
    text=New Prefix:
    toolTipText="Rename prefix for exported AnimationSequences.\r\n\r\nFor ex. 'Minifig' to 'FishingRod'"
    colorAndOpacity {colorUseRule=UseColor_Specified, specifiedColor=<rgba 0.1545 0.1545 0.1545 1>}
    font {
     bForceMonospaced=false
     bMaterialIsStencil=false
     fontMaterial=null
     fontObject=&Roboto
     letterSpacing=0
     monospacedWidth=1
     size=9
     skewAmount=0
     typefaceFontName=Bold
     outlineSettings {
      bApplyOutlineToDropShadows=false
      bMiteredCorners=false
      bSeparateFillAlpha=false
      outlineMaterial=null
      outlineSize=0
      outlineColor=<rgba 0 0 0 1>
     }
    }
   }
   slot {class=VerticalBoxSlot, properties {horizontalAlignment=HAlign_Right, verticalAlignment=VAlign_Center}}
  }
  TextBlock_8 {
   widgetClassPath=&TextBlock
   parent=Layout1b1
   widget=TextBlock_8
   properties {
    text=AnimSeq Path:
    toolTipText="Path to store exported AnimationSequences for this track.\r\n\r\nFor ex. /JunoFishing/FishingRod/Animation/Actions/"
    colorAndOpacity {colorUseRule=UseColor_Specified, specifiedColor=<rgba 0.1545 0.1545 0.1545 1>}
    font {
     bForceMonospaced=false
     bMaterialIsStencil=false
     fontMaterial=null
     fontObject=&Roboto
     letterSpacing=0
     monospacedWidth=1
     size=9
     skewAmount=0
     typefaceFontName=Bold
     outlineSettings {
      bApplyOutlineToDropShadows=false
      bMiteredCorners=false
      bSeparateFillAlpha=false
      outlineMaterial=null
      outlineSize=0
      outlineColor=<rgba 0 0 0 1>
     }
    }
   }
   slot {class=VerticalBoxSlot, properties {verticalAlignment=VAlign_Center}}
  }
  TextBlock_10 {
   widgetClassPath=&TextBlock
   parent=Layout1b1
   widget=TextBlock_10
   properties {
    text=Preview:
    toolTipText="Preview shows how the name of future AnimSequence(s) for this track. "
    colorAndOpacity {colorUseRule=UseColor_Specified, specifiedColor=<rgba 0.1545 0.1545 0.1545 1>}
    font {
     bForceMonospaced=false
     bMaterialIsStencil=false
     fontMaterial=null
     fontObject=&Roboto
     letterSpacing=0
     monospacedWidth=1
     size=9
     skewAmount=0
     typefaceFontName=Bold
     outlineSettings {
      bApplyOutlineToDropShadows=false
      bMiteredCorners=false
      bSeparateFillAlpha=false
      outlineMaterial=null
      outlineSize=0
      outlineColor=<rgba 0 0 0 1>
     }
    }
   }
   slot {class=VerticalBoxSlot, properties {horizontalAlignment=HAlign_Right, verticalAlignment=VAlign_Center}}
  }
  Layout1b3 {
   widgetClassPath=&VerticalBox
   parent=HorizontalBox_4
   widget=Layout1b3
   slot {class=HorizontalBoxSlot, properties {horizontalAlignment=HAlign_Left, verticalAlignment=VAlign_Center, padding=<ltrb 10 0 0 0>}}
  }
  EditableText_Track_Path {
   widgetClassPath=/Script/Blutility.EditorUtilityEditableText
   bIsVariable=true
   parent=Layout1b3
   widget=EditableText_Track_Path
   properties {
    hintText="[  Folder to export AnimSequences  ]"
    toolTipText="Path to store exported AnimationSequences for this track.\r\n\r\nFor ex. &Actions/"
    widgetStyle {
     backgroundImageComposing {
      bIsDynamicallyLoaded=false
      drawAs=Border
      imageType=FullColor
      mirroring=NoMirror
      resourceName=../../../Engine/Content/Slate/Old/HyperlinkDotted.png
      resourceObject=null
      tiling=Both
      imageSize=<0 0>
      margin=<ltrb 0 0 0 0.1875>
      outlineSettings {
       bUseBrushTransparency=false
       roundingType=HalfHeightRadius
       width=0
       color {colorUseRule=UseColor_Specified, specifiedColor=<rgba 0 0 0 0>}
       cornerRadii=<0 0 0 0>
      }
      tintColor {colorUseRule=UseColor_Specified, specifiedColor=<rgba 1 1 1 1>}
      uVRegion {bIsValid=false, max=<0 0>, min=<0 0>}
     }
     backgroundImageSelected {
      bIsDynamicallyLoaded=false
      drawAs=Image
      imageType=NoImage
      mirroring=NoMirror
      resourceName=None
      resourceObject=null
      tiling=NoTile
      imageSize=<0 0>
      margin=<ltrb 0 0 0 0>
      outlineSettings {
       bUseBrushTransparency=false
       roundingType=HalfHeightRadius
       width=0
       color {colorUseRule=UseColor_Specified, specifiedColor=<rgba 0 0 0 0>}
       cornerRadii=<0 0 0 0>
      }
      tintColor {colorUseRule=UseColor_Specified, specifiedColor=<rgba 0 0.162 0.7454 1>}
      uVRegion {bIsValid=false, max=<0 0>, min=<0 0>}
     }
     caretImage {
      bIsDynamicallyLoaded=false
      drawAs=Image
      imageType=NoImage
      mirroring=NoMirror
      resourceName=None
      resourceObject=null
      tiling=NoTile
      imageSize=<0 0>
      margin=<ltrb 0 0 0 0>
      outlineSettings {
       bUseBrushTransparency=false
       roundingType=HalfHeightRadius
       width=0
       color {colorUseRule=UseColor_Specified, specifiedColor=<rgba 0 0 0 0>}
       cornerRadii=<0 0 0 0>
      }
      tintColor {colorUseRule=UseColor_Specified, specifiedColor=<rgba 1 1 1 1>}
      uVRegion {bIsValid=false, max=<0 0>, min=<0 0>}
     }
     colorAndOpacity {colorUseRule=UseColor_Specified, specifiedColor=<rgba 1 1 1 1>}
     font {
      bForceMonospaced=false
      bMaterialIsStencil=false
      fontMaterial=null
      fontObject=&Roboto
      letterSpacing=10
      monospacedWidth=1
      size=9
      skewAmount=0.12
      typefaceFontName=Regular
      outlineSettings {
       bApplyOutlineToDropShadows=false
       bMiteredCorners=false
       bSeparateFillAlpha=false
       outlineMaterial=null
       outlineSize=0
       outlineColor=<rgba 0 0 0 1>
      }
     }
     hintColorAndOpacity {colorUseRule=UseColor_Foreground, specifiedColor=<rgba 1 0 1 1>}
     selectedBackgroundColor {colorUseRule=UseColor_Foreground, specifiedColor=<rgba 1 0 1 1>}
    }
   }
   slot {class=VerticalBoxSlot, properties {horizontalAlignment=HAlign_Left, verticalAlignment=VAlign_Center}}
  }
  Layout1b2 {widgetClassPath=&HorizontalBox, parent=Layout1b3, widget=Layout1b2, slot {class=VerticalBoxSlot}}
  EditableText_New_Prefix {
   widgetClassPath=&EditorUtilityEditableText
   bIsVariable=true
   parent=Layout1b2
   widget=EditableText_New_Prefix
   properties {
    hintText="[  Prefix  ]"
    toolTipText=Starting part of the filename to replace.
    widgetStyle {
     backgroundImageComposing {
      bIsDynamicallyLoaded=false
      drawAs=Border
      imageType=FullColor
      mirroring=NoMirror
      resourceName=../../../Engine/Content/Slate/Old/HyperlinkDotted.png
      resourceObject=null
      tiling=Both
      imageSize=<0 0>
      margin=<ltrb 0 0 0 0.1875>
      outlineSettings {
       bUseBrushTransparency=false
       roundingType=HalfHeightRadius
       width=0
       color {colorUseRule=UseColor_Specified, specifiedColor=<rgba 0 0 0 0>}
       cornerRadii=<0 0 0 0>
      }
      tintColor {colorUseRule=UseColor_Specified, specifiedColor=<rgba 1 1 1 1>}
      uVRegion {bIsValid=false, max=<0 0>, min=<0 0>}
     }
     backgroundImageSelected {
      bIsDynamicallyLoaded=false
      drawAs=Image
      imageType=NoImage
      mirroring=NoMirror
      resourceName=None
      resourceObject=null
      tiling=NoTile
      imageSize=<0 0>
      margin=<ltrb 0 0 0 0>
      outlineSettings {
       bUseBrushTransparency=false
       roundingType=HalfHeightRadius
       width=0
       color {colorUseRule=UseColor_Specified, specifiedColor=<rgba 0 0 0 0>}
       cornerRadii=<0 0 0 0>
      }
      tintColor {colorUseRule=UseColor_Specified, specifiedColor=<rgba 0 0.162 0.7454 1>}
      uVRegion {bIsValid=false, max=<0 0>, min=<0 0>}
     }
     caretImage {
      bIsDynamicallyLoaded=false
      drawAs=Image
      imageType=NoImage
      mirroring=NoMirror
      resourceName=None
      resourceObject=null
      tiling=NoTile
      imageSize=<0 0>
      margin=<ltrb 0 0 0 0>
      outlineSettings {
       bUseBrushTransparency=false
       roundingType=HalfHeightRadius
       width=0
       color {colorUseRule=UseColor_Specified, specifiedColor=<rgba 0 0 0 0>}
       cornerRadii=<0 0 0 0>
      }
      tintColor {colorUseRule=UseColor_Specified, specifiedColor=<rgba 1 1 1 1>}
      uVRegion {bIsValid=false, max=<0 0>, min=<0 0>}
     }
     colorAndOpacity {colorUseRule=UseColor_Specified, specifiedColor=<rgba 1 1 1 1>}
     font {
      bForceMonospaced=false
      bMaterialIsStencil=false
      fontMaterial=null
      fontObject=&Roboto
      letterSpacing=10
      monospacedWidth=1
      size=9
      skewAmount=0.12
      typefaceFontName=Regular
      outlineSettings {
       bApplyOutlineToDropShadows=false
       bMiteredCorners=false
       bSeparateFillAlpha=false
       outlineMaterial=null
       outlineSize=0
       outlineColor=<rgba 0 0 0 1>
      }
     }
     hintColorAndOpacity {colorUseRule=UseColor_Foreground, specifiedColor=<rgba 1 0 1 1>}
     selectedBackgroundColor {colorUseRule=UseColor_Foreground, specifiedColor=<rgba 1 0 1 1>}
    }
   }
   slot {class=HorizontalBoxSlot, properties {horizontalAlignment=HAlign_Left, verticalAlignment=VAlign_Center}}
  }
  HorizontalBox_88 {
   widgetClassPath=&HorizontalBox
   parent=Layout1b3
   widget=HorizontalBox_88
   slot {class=VerticalBoxSlot, properties {verticalAlignment=VAlign_Center}}
  }
  EditableText_Track_Preview {
   widgetClassPath=&EditorUtilityEditableText
   bIsVariable=true
   parent=HorizontalBox_88
   widget=EditableText_Track_Preview
   properties {
    hintText="[  Preview of the AnimSequnce name  ]"
    isReadOnly=true
    toolTipText=Name of mastre file AnimSequnce that will be created and linked.
    widgetStyle {
     backgroundImageComposing {
      bIsDynamicallyLoaded=false
      drawAs=Border
      imageType=FullColor
      mirroring=NoMirror
      resourceName=../../../Engine/Content/Slate/Old/HyperlinkDotted.png
      resourceObject=null
      tiling=Both
      imageSize=<0 0>
      margin=<ltrb 0 0 0 0.1875>
      outlineSettings {
       bUseBrushTransparency=false
       roundingType=HalfHeightRadius
       width=0
       color {colorUseRule=UseColor_Specified, specifiedColor=<rgba 0 0 0 0>}
       cornerRadii=<0 0 0 0>
      }
      tintColor {colorUseRule=UseColor_Specified, specifiedColor=<rgba 1 1 1 1>}
      uVRegion {bIsValid=false, max=<0 0>, min=<0 0>}
     }
     backgroundImageSelected {
      bIsDynamicallyLoaded=false
      drawAs=Image
      imageType=NoImage
      mirroring=NoMirror
      resourceName=None
      resourceObject=null
      tiling=NoTile
      imageSize=<0 0>
      margin=<ltrb 0 0 0 0>
      outlineSettings {
       bUseBrushTransparency=false
       roundingType=HalfHeightRadius
       width=0
       color {colorUseRule=UseColor_Specified, specifiedColor=<rgba 0 0 0 0>}
       cornerRadii=<0 0 0 0>
      }
      tintColor {colorUseRule=UseColor_Specified, specifiedColor=<rgba 0 0.162 0.7454 1>}
      uVRegion {bIsValid=false, max=<0 0>, min=<0 0>}
     }
     caretImage {
      bIsDynamicallyLoaded=false
      drawAs=Image
      imageType=NoImage
      mirroring=NoMirror
      resourceName=None
      resourceObject=null
      tiling=NoTile
      imageSize=<0 0>
      margin=<ltrb 0 0 0 0>
      outlineSettings {
       bUseBrushTransparency=false
       roundingType=HalfHeightRadius
       width=0
       color {colorUseRule=UseColor_Specified, specifiedColor=<rgba 0 0 0 0>}
       cornerRadii=<0 0 0 0>
      }
      tintColor {colorUseRule=UseColor_Specified, specifiedColor=<rgba 1 1 1 1>}
      uVRegion {bIsValid=false, max=<0 0>, min=<0 0>}
     }
     colorAndOpacity {colorUseRule=UseColor_Specified, specifiedColor=<rgba 1 1 1 1>}
     font {
      bForceMonospaced=false
      bMaterialIsStencil=false
      fontMaterial=null
      fontObject=&Roboto
      letterSpacing=10
      monospacedWidth=1
      size=9
      skewAmount=0.12
      typefaceFontName=Regular
      outlineSettings {
       bApplyOutlineToDropShadows=false
       bMiteredCorners=false
       bSeparateFillAlpha=false
       outlineMaterial=null
       outlineSize=0
       outlineColor=<rgba 0 0 0 1>
      }
     }
     hintColorAndOpacity {colorUseRule=UseColor_Foreground, specifiedColor=<rgba 1 0 1 1>}
     selectedBackgroundColor {colorUseRule=UseColor_Foreground, specifiedColor=<rgba 1 0 1 1>}
    }
   }
   slot {class=HorizontalBoxSlot}
  }
  TextBlock_7 {
   widgetClassPath=&TextBlock
   parent=HorizontalBox_88
   widget=TextBlock_7
   properties {
    text=_SegmentName
    toolTipText=Generated Segments' AnimSequences will additionaly have the suffix with the name of the segment added.
    colorAndOpacity {colorUseRule=UseColor_Specified, specifiedColor=<rgba 0.1545 0.1545 0.1545 1>}
    font {
     bForceMonospaced=false
     bMaterialIsStencil=false
     fontMaterial=null
     fontObject=&Roboto
     letterSpacing=0
     monospacedWidth=1
     size=9
     skewAmount=0
     typefaceFontName=Regular
     outlineSettings {
      bApplyOutlineToDropShadows=false
      bMiteredCorners=false
      bSeparateFillAlpha=false
      outlineMaterial=null
      outlineSize=0
      outlineColor=<rgba 0 0 0 1>
     }
    }
   }
   slot {class=HorizontalBoxSlot, properties {horizontalAlignment=HAlign_Right, verticalAlignment=VAlign_Center}}
  }
  HorizontalBox_83 {
   widgetClassPath=&HorizontalBox
   parent=Layout1b
   widget=HorizontalBox_83
   slot {class=VerticalBoxSlot, properties {verticalAlignment=VAlign_Top, padding=<ltrb 0 6 0 0>}}
  }
  Button_AddTrack_Confirm {
   widgetClassPath=&EditorUtilityButton
   parent=HorizontalBox_83
   widget=Button_AddTrack_Confirm
   properties {toolTipText="Adds selected in drop down list track as an exportable.\r\n\r\nMeant to be used for props."}
   slot {class=HorizontalBoxSlot, properties {horizontalAlignment=HAlign_Center, verticalAlignment=VAlign_Top, padding=<ltrb 30 0 10 0>}}
  }
  TextBlock_11 {
   widgetClassPath=&TextBlock
   parent=Button_AddTrack_Confirm
   widget=TextBlock_11
   properties {
    text=Create
    colorAndOpacity {colorUseRule=UseColor_Specified, specifiedColor=<rgba 0.161 1 0.161 1>}
    font {
     bForceMonospaced=false
     bMaterialIsStencil=false
     fontMaterial=null
     fontObject=&Roboto
     letterSpacing=100
     monospacedWidth=1
     size=10.5
     skewAmount=0
     typefaceFontName=Regular
     outlineSettings {
      bApplyOutlineToDropShadows=false
      bMiteredCorners=false
      bSeparateFillAlpha=false
      outlineMaterial=null
      outlineSize=0
      outlineColor=<rgba 0 0 0 1>
     }
    }
   }
   slot {class=ButtonSlot, properties {horizontalAlignment=HAlign_Left, padding=<ltrb 3 3 3 3>}}
  }
  Spacer_5 {
   widgetClassPath=&Spacer
   parent=HorizontalBox_83
   widget=Spacer_5
   properties {size=<0 0>}
   slot {class=HorizontalBoxSlot, properties {verticalAlignment=VAlign_Top, size {sizeRule=Fill, value=1}}}
  }
  Button_AddTrack_Cancel {
   widgetClassPath=&EditorUtilityButton
   parent=HorizontalBox_83
   widget=Button_AddTrack_Cancel
   slot {class=HorizontalBoxSlot, properties {horizontalAlignment=HAlign_Center, verticalAlignment=VAlign_Top, padding=<ltrb 10 0 30 0>}}
  }
  TextBlock_12 {
   widgetClassPath=&TextBlock
   parent=Button_AddTrack_Cancel
   widget=TextBlock_12
   properties {
    text=Cancel
    colorAndOpacity {colorUseRule=UseColor_Specified, specifiedColor=<rgba 0.764 0 0 1>}
    font {
     bForceMonospaced=false
     bMaterialIsStencil=false
     fontMaterial=null
     fontObject=&Roboto
     letterSpacing=100
     monospacedWidth=1
     size=10.5
     skewAmount=0
     typefaceFontName=Regular
     outlineSettings {
      bApplyOutlineToDropShadows=false
      bMiteredCorners=false
      bSeparateFillAlpha=false
      outlineMaterial=null
      outlineSize=0
      outlineColor=<rgba 0 0 0 1>
     }
    }
   }
   slot {class=ButtonSlot, properties {horizontalAlignment=HAlign_Left, padding=<ltrb 3 3 3 3>}}
  }
  Spacer_2 {
   widgetClassPath=&Spacer
   parent=CoreLayout
   widget=Spacer_2
   properties {size=<1 20>}
   slot {class=VerticalBoxSlot, properties {horizontalAlignment=HAlign_Left, verticalAlignment=VAlign_Center}}
  }
  Button_Scan {
   widgetClassPath=&EditorUtilityButton
   parent=CoreLayout
   widget=Button_Scan
   properties {toolTipText="{None\r\n\r\nSet in python}"}
   slot {class=VerticalBoxSlot, properties {horizontalAlignment=HAlign_Center, verticalAlignment=VAlign_Center}}
  }
  TextBlock_4 {
   widgetClassPath=&TextBlock
   parent=Button_Scan
   widget=TextBlock_4
   properties {
    text=Refresh
    font {
     bForceMonospaced=false
     bMaterialIsStencil=false
     fontMaterial=null
     fontObject=&Roboto
     letterSpacing=0
     monospacedWidth=1
     size=10.5
     skewAmount=0
     typefaceFontName=Bold
     outlineSettings {
      bApplyOutlineToDropShadows=false
      bMiteredCorners=false
      bSeparateFillAlpha=false
      outlineMaterial=null
      outlineSize=0
      outlineColor=<rgba 0 0 0 1>
     }
    }
   }
   slot {class=ButtonSlot, properties {horizontalAlignment=HAlign_Fill, verticalAlignment=VAlign_Fill, padding=<ltrb 20 6 20 6>}}
  }
  HorizontalBox_163 {
   widgetClassPath=&HorizontalBox
   parent=CoreLayout
   widget=HorizontalBox_163
   slot {class=VerticalBoxSlot, properties {horizontalAlignment=HAlign_Center, verticalAlignment=VAlign_Center, padding=<ltrb 0 4 0 4>}}
  }
  TextBlock_19 {
   widgetClassPath=&TextBlock
   parent=HorizontalBox_163
   widget=TextBlock_19
   properties {
    text=Depricated
    toolTipText="This is the old way, it's much slower and prone to bugs. \r\n\r\nWe keep it to quickly edit master files that were made before support of direct linkage of multiple AnimSequences. \r\n\r\nIn the future it will be removed completely. "
    visibility=Hidden
    colorAndOpacity {colorUseRule=UseColor_Specified, specifiedColor=<rgba 1 0 0 1>}
    font {
     bForceMonospaced=false
     bMaterialIsStencil=false
     fontMaterial=null
     fontObject=&Roboto
     letterSpacing=0
     monospacedWidth=1
     size=9
     skewAmount=0
     typefaceFontName=Regular
     outlineSettings {
      bApplyOutlineToDropShadows=false
      bMiteredCorners=false
      bSeparateFillAlpha=false
      outlineMaterial=null
      outlineSize=0
      outlineColor=<rgba 0 0 0 1>
     }
    }
   }
   slot {class=HorizontalBoxSlot, properties {horizontalAlignment=HAlign_Right, verticalAlignment=VAlign_Center}}
  }
  TextBlock_16 {
   widgetClassPath=&TextBlock
   parent=HorizontalBox_163
   widget=TextBlock_16
   properties {
    text=No Cropped LS:
    toolTipText="When checked - cropped frameranges will be directly exported to AnimSequence.\r\n\r\nWhen Unchecked - cropped frameranges will be first duplicated into cropped LeveSequences and from there each exportend to AnimSequence. This is the old way, it's much slower and prone to bugs. We keep it to edit master files that were made before support of direct linkage of multiple AnimSequences. In the future it will be removed completely. "
    colorAndOpacity {colorUseRule=UseColor_Specified, specifiedColor=<rgba 0.1545 0.1545 0.1545 1>}
    font {
     bForceMonospaced=false
     bMaterialIsStencil=false
     fontMaterial=null
     fontObject=&Roboto
     letterSpacing=0
     monospacedWidth=1
     size=9
     skewAmount=0
     typefaceFontName=Bold
     outlineSettings {
      bApplyOutlineToDropShadows=false
      bMiteredCorners=false
      bSeparateFillAlpha=false
      outlineMaterial=null
      outlineSize=0
      outlineColor=<rgba 0 0 0 1>
     }
    }
   }
   slot {class=HorizontalBoxSlot, properties {verticalAlignment=VAlign_Center, padding=<ltrb 10 2 0 2>}}
  }
  checkbox_no_ls {
   widgetClassPath=&EditorUtilityCheckBox
   bIsVariable=true
   parent=HorizontalBox_163
   widget=checkbox_no_ls
   properties {
    checkedState=Checked
    toolTipText="When checked - cropped frameranges will be directly exported to AnimSequence.\r\n\r\nWhen Unchecked - cropped frameranges will be first duplicated into cropped LeveSequences and from there each exportend to AnimSequence. This is the old way, it's much slower and prone to bugs. We keep it to edit master files that were made before support of direct linkage of multiple AnimSequences. In the future it will be removed completely. "
   }
   slot {class=HorizontalBoxSlot, properties {padding=<ltrb 10 0 10 0>}}
  }
  TextBlock_18 {
   widgetClassPath=&TextBlock
   bIsVariable=true
   parent=HorizontalBox_163
   widget=TextBlock_18
   properties {
    text=Depricated
    toolTipText="This is the old way, it's much slower and prone to bugs. \r\n\r\nWe keep it to quickly edit master files that were made before support of direct linkage of multiple AnimSequences. \r\n\r\nIn the future it will be removed completely. "
    visibility=Hidden
    colorAndOpacity {colorUseRule=UseColor_Specified, specifiedColor=<rgba 1 0 0 1>}
    font {
     bForceMonospaced=false
     bMaterialIsStencil=false
     fontMaterial=null
     fontObject=&Roboto
     letterSpacing=0
     monospacedWidth=1
     size=9
     skewAmount=0
     typefaceFontName=Regular
     outlineSettings {
      bApplyOutlineToDropShadows=false
      bMiteredCorners=false
      bSeparateFillAlpha=false
      outlineMaterial=null
      outlineSize=0
      outlineColor=<rgba 0 0 0 1>
     }
    }
   }
   slot {class=HorizontalBoxSlot, properties {horizontalAlignment=HAlign_Right, verticalAlignment=VAlign_Center}}
  }
  HorizontalBox_Mode_1 {
   widgetClassPath=&HorizontalBox
   parent=CoreLayout
   widget=HorizontalBox_Mode_1
   slot {class=VerticalBoxSlot, properties {horizontalAlignment=HAlign_Center, verticalAlignment=VAlign_Top}}
  }
  TextBlock_17 {
   widgetClassPath=&TextBlock
   parent=HorizontalBox_Mode_1
   widget=TextBlock_17
   properties {
    text=Curve Compression:
    toolTipText="Using Checkboxed select one or several criterias to scan for.\r\n"
    colorAndOpacity {colorUseRule=UseColor_Specified, specifiedColor=<rgba 0.1559 0.1559 0.1559 1>}
    font {
     bForceMonospaced=false
     bMaterialIsStencil=false
     fontMaterial=null
     fontObject=&Roboto
     letterSpacing=0
     monospacedWidth=1
     size=9
     skewAmount=0
     typefaceFontName=Bold
     outlineSettings {
      bApplyOutlineToDropShadows=false
      bMiteredCorners=false
      bSeparateFillAlpha=false
      outlineMaterial=null
      outlineSize=0
      outlineColor=<rgba 0 0 0 1>
     }
    }
   }
   slot {class=HorizontalBoxSlot, properties {verticalAlignment=VAlign_Center}}
  }
  ComboBox_CurveCompression {
   widgetClassPath=&EditorUtilityComboBoxString
   bIsVariable=true
   parent=HorizontalBox_Mode_1
   widget=ComboBox_CurveCompression
   properties {
    selectedOption=Juno
    contentPadding=<ltrb 4 2 10 2>
    defaultOptions [Juno, FN BR]
    font {
     bForceMonospaced=false
     bMaterialIsStencil=false
     fontMaterial=null
     fontObject=&Roboto
     letterSpacing=0
     monospacedWidth=1
     size=9
     skewAmount=0
     typefaceFontName=Regular
     outlineSettings {
      bApplyOutlineToDropShadows=false
      bMiteredCorners=false
      bSeparateFillAlpha=false
      outlineMaterial=null
      outlineSize=0
      outlineColor=<rgba 0 0 0 1>
     }
    }
    foregroundColor {colorUseRule=UseColor_Specified, specifiedColor=<rgba 0.5271 0.5271 0.5271 1>}
   }
   slot {class=HorizontalBoxSlot, properties {horizontalAlignment=HAlign_Right, verticalAlignment=VAlign_Center}}
  }
  Spacer_4 {
   widgetClassPath=&Spacer
   parent=CoreLayout
   widget=Spacer_4
   properties {size=<1 20>}
   slot {class=VerticalBoxSlot, properties {horizontalAlignment=HAlign_Left, verticalAlignment=VAlign_Center}}
  }
  Border_117 {
   widgetClassPath=&Border
   parent=CoreLayout
   widget=Border_117
   properties {
    background {
     bIsDynamicallyLoaded=false
     drawAs=Border
     imageType=NoImage
     mirroring=NoMirror
     resourceName=None
     resourceObject=null
     tiling=NoTile
     imageSize=<32 32>
     margin=<ltrb 0 0 0 1>
     outlineSettings {
      bUseBrushTransparency=false
      roundingType=HalfHeightRadius
      width=0
      color {colorUseRule=UseColor_Specified, specifiedColor=<rgba 0 0 0 0>}
      cornerRadii=<0 0 0 0>
     }
     tintColor {colorUseRule=UseColor_Specified, specifiedColor=<rgba 0.1302 0.1302 0.1302 1>}
     uVRegion {bIsValid=false, max=<0 0>, min=<0 0>}
    }
    padding=<ltrb 0 10 0 0>
   }
   slot {class=VerticalBoxSlot}
  }
  TextBlock_351 {
   widgetClassPath=&TextBlock
   parent=Border_117
   widget=TextBlock_351
   properties {
    text=Cropped Segments:
    toolTipText="Using Checkboxed select one or several criterias to scan for.\r\n"
    font {
     bForceMonospaced=false
     bMaterialIsStencil=false
     fontMaterial=null
     fontObject=&Roboto
     letterSpacing=0
     monospacedWidth=1
     size=9
     skewAmount=0
     typefaceFontName=Bold
     outlineSettings {
      bApplyOutlineToDropShadows=false
      bMiteredCorners=false
      bSeparateFillAlpha=false
      outlineMaterial=null
      outlineSize=0
      outlineColor=<rgba 0 0 0 1>
     }
    }
   }
   slot {class=BorderSlot, properties {padding=<ltrb 0 10 0 0>}}
  }
  Spacer {
   widgetClassPath=&Spacer
   parent=CoreLayout
   widget=Spacer
   properties {size=<1 10>}
   slot {class=VerticalBoxSlot, properties {horizontalAlignment=HAlign_Left, verticalAlignment=VAlign_Center}}
  }
  Layout2 {
   widgetClassPath=&VerticalBox
   bIsVariable=true
   parent=CoreLayout
   widget=Layout2
   slot {class=VerticalBoxSlot, properties {verticalAlignment=VAlign_Top, size {sizeRule=Fill, value=1}}}
  }
  Spacer_10 {
   widgetClassPath=&Spacer
   parent=CoreLayout
   widget=Spacer_10
   properties {size=<0 30>}
   slot {class=VerticalBoxSlot, properties {horizontalAlignment=HAlign_Center}}
  }
  HorizontalBoxBlendSpace {
   widgetClassPath=&VerticalBox
   bIsVariable=true
   parent=CoreLayout
   widget=HorizontalBoxBlendSpace
   slot {class=VerticalBoxSlot, properties {horizontalAlignment=HAlign_Center, verticalAlignment=VAlign_Center, padding=<ltrb 0 20 0 0>}}
  }
  HorizontalBox_3 {widgetClassPath=&HorizontalBox, parent=HorizontalBoxBlendSpace, widget=HorizontalBox_3, slot {class=VerticalBoxSlot}}
  SinglePropertyView_BlendSpace {
   widgetClassPath=&SinglePropertyView
   parent=HorizontalBox_3
   widget=SinglePropertyView_BlendSpace
   properties {propertyName=BlendSpace}
   slot {class=HorizontalBoxSlot}
  }
  HorizontalBox_0 {
   widgetClassPath=&HorizontalBox
   parent=HorizontalBox_3
   widget=HorizontalBox_0
   slot {class=HorizontalBoxSlot, properties {verticalAlignment=VAlign_Top, padding=<ltrb 0 2 0 0>}}
  }
  Add_Track_Text_1 {
   widgetClassPath=&TextBlock
   bIsVariable=true
   parent=HorizontalBox_0
   widget=Add_Track_Text_1
   properties {
    text=Type:
    toolTipText="Type of animation used in BlendSpace.\r\n\r\nAimOffsets are usually a MeshSpace Additives."
    colorAndOpacity {colorUseRule=UseColor_Specified, specifiedColor=<rgba 0.1545 0.1545 0.1545 1>}
    font {
     bForceMonospaced=false
     bMaterialIsStencil=false
     fontMaterial=null
     fontObject=&Roboto
     letterSpacing=0
     monospacedWidth=1
     size=9
     skewAmount=0
     typefaceFontName=Bold
     outlineSettings {
      bApplyOutlineToDropShadows=false
      bMiteredCorners=false
      bSeparateFillAlpha=false
      outlineMaterial=null
      outlineSize=0
      outlineColor=<rgba 0 0 0 1>
     }
    }
   }
   slot {class=HorizontalBoxSlot, properties {horizontalAlignment=HAlign_Center, verticalAlignment=VAlign_Center, padding=<ltrb 0 0 5 0>}}
  }
  ComboBox_AnimType {
   widgetClassPath=&EditorUtilityComboBoxString
   bIsVariable=true
   parent=HorizontalBox_0
   widget=ComboBox_AnimType
   properties {
    selectedOption=MeshSpace Additive
    toolTipText="Type of animation used in BlendSpace.\r\n\r\nAimOffsets are usually a MeshSpace Additives."
    contentPadding=<ltrb 4 2 10 2>
    defaultOptions [MeshSpace Additive, LocalSpace Additive, Not Additive]
    font {
     bForceMonospaced=false
     bMaterialIsStencil=false
     fontMaterial=null
     fontObject=&Roboto
     letterSpacing=0
     monospacedWidth=1
     size=9
     skewAmount=0
     typefaceFontName=Regular
     outlineSettings {
      bApplyOutlineToDropShadows=false
      bMiteredCorners=false
      bSeparateFillAlpha=false
      outlineMaterial=null
      outlineSize=0
      outlineColor=<rgba 0 0 0 1>
     }
    }
    foregroundColor {colorUseRule=UseColor_Specified, specifiedColor=<rgba 0.5271 0.5271 0.5271 1>}
   }
   slot {class=HorizontalBoxSlot, properties {horizontalAlignment=HAlign_Center, verticalAlignment=VAlign_Center}}
  }
  HorizontalBox_SaveExpotr {
   widgetClassPath=&HorizontalBox
   parent=CoreLayout
   widget=HorizontalBox_SaveExpotr
   slot {class=VerticalBoxSlot, properties {horizontalAlignment=HAlign_Center, verticalAlignment=VAlign_Center, padding=<ltrb 0 30 0 0>}}
  }
  Button_Save {
   widgetClassPath=&EditorUtilityButton
   parent=HorizontalBox_SaveExpotr
   widget=Button_Save
   properties {toolTipText="Saves changes for selected cropped file(s).\r\n\r\n- updates latest crop ranges in cropped files;\r\n- saves latest animation data in all cropped files."}
   slot {class=HorizontalBoxSlot, properties {horizontalAlignment=HAlign_Center, verticalAlignment=VAlign_Center, padding=<ltrb 100 0 0 0>}}
  }
  TextBlock_13 {
   widgetClassPath=&TextBlock
   parent=Button_Save
   widget=TextBlock_13
   properties {
    text=Save
    font {
     bForceMonospaced=false
     bMaterialIsStencil=false
     fontMaterial=null
     fontObject=&Roboto
     letterSpacing=0
     monospacedWidth=1
     size=10.5
     skewAmount=0
     typefaceFontName=Bold
     outlineSettings {
      bApplyOutlineToDropShadows=false
      bMiteredCorners=false
      bSeparateFillAlpha=false
      outlineMaterial=null
      outlineSize=0
      outlineColor=<rgba 0 0 0 1>
     }
    }
   }
   slot {class=ButtonSlot, properties {horizontalAlignment=HAlign_Fill, verticalAlignment=VAlign_Fill, padding=<ltrb 20 6 20 6>}}
  }
  Spacer_1 {
   widgetClassPath=&Spacer
   parent=HorizontalBox_SaveExpotr
   widget=Spacer_1
   properties {size=<200 0>}
   slot {class=HorizontalBoxSlot, properties {verticalAlignment=VAlign_Center}}
  }
  HorizontalBox_2 {
   widgetClassPath=&HorizontalBox
   parent=HorizontalBox_SaveExpotr
   widget=HorizontalBox_2
   slot {class=HorizontalBoxSlot, properties {padding=<ltrb 0 0 100 0>}}
  }
  Button_Export {
   widgetClassPath=&EditorUtilityButton
   parent=HorizontalBox_2
   widget=Button_Export
   properties {toolTipText="[SLOW] \r\nCreates missing dependant files.\r\n\r\nUse when you see \"Not Created\" in the File Status column.\r\n"}
   slot {class=HorizontalBoxSlot, properties {horizontalAlignment=HAlign_Center, verticalAlignment=VAlign_Center}}
  }
  TextBlock_1 {
   widgetClassPath=&TextBlock
   parent=Button_Export
   widget=TextBlock_1
   properties {
    text=Create Assets
    font {
     bForceMonospaced=false
     bMaterialIsStencil=false
     fontMaterial=null
     fontObject=&Roboto
     letterSpacing=0
     monospacedWidth=1
     size=10.5
     skewAmount=0
     typefaceFontName=Bold
     outlineSettings {
      bApplyOutlineToDropShadows=false
      bMiteredCorners=false
      bSeparateFillAlpha=false
      outlineMaterial=null
      outlineSize=0
      outlineColor=<rgba 0 0 0 1>
     }
    }
   }
   slot {class=ButtonSlot, properties {horizontalAlignment=HAlign_Fill, verticalAlignment=VAlign_Fill, padding=<ltrb 20 6 20 6>}}
  }
  ForceCreate {
   widgetClassPath=&EditorUtilityCheckBox
   bIsVariable=true
   parent=HorizontalBox_2
   widget=ForceCreate
   properties {toolTipText="Emergency fucntion.\r\nTheoritically you should never need to use it.\r\n\r\nForcefully deletes and then re-creates all  cropped Level and Anim Sequences. Keeps all referencers."}
   slot {class=HorizontalBoxSlot, properties {padding=<ltrb 8 0 0 0>}}
  }
  TextBlock_15 {
   widgetClassPath=&TextBlock
   parent=ForceCreate
   widget=TextBlock_15
   properties {
    text=Force
    colorAndOpacity {colorUseRule=UseColor_Specified, specifiedColor=<rgba 0.1545 0.1545 0.1545 1>}
    font {
     bForceMonospaced=false
     bMaterialIsStencil=false
     fontMaterial=null
     fontObject=&Roboto
     letterSpacing=0
     monospacedWidth=1
     size=9
     skewAmount=0
     typefaceFontName=Regular
     outlineSettings {
      bApplyOutlineToDropShadows=false
      bMiteredCorners=false
      bSeparateFillAlpha=false
      outlineMaterial=null
      outlineSize=0
      outlineColor=<rgba 0 0 0 1>
     }
    }
   }
   slot {class=PanelSlot}
  }
  Invisible {
   widgetClassPath=&EditorUtilityScrollBox
   parent=EditorUtilityScrollBox_1
   widget=Invisible
   properties {visibility=Hidden}
   slot {class=ScrollBoxSlot}
  }
  TextStyleSammple_Regular {
   widgetClassPath=&TextBlock
   bIsVariable=true
   parent=Invisible
   widget=TextStyleSammple_Regular
   properties {
    text=Sample
    toolTipText="Does animation contain float channel curves.\r\n\r\n\r\nType name in the filed on the right."
    font {
     bForceMonospaced=false
     bMaterialIsStencil=false
     fontMaterial=null
     fontObject=&Roboto
     letterSpacing=0
     monospacedWidth=1
     size=9
     skewAmount=0
     typefaceFontName=Regular
     outlineSettings {
      bApplyOutlineToDropShadows=false
      bMiteredCorners=false
      bSeparateFillAlpha=false
      outlineMaterial=null
      outlineSize=0
      outlineColor=<rgba 0 0 0 1>
     }
    }
   }
   slot {class=ScrollBoxSlot, properties {horizontalAlignment=HAlign_Left, verticalAlignment=VAlign_Center}}
  }
  TextStyleSammple_Bold_Big {
   widgetClassPath=&TextBlock
   bIsVariable=true
   parent=Invisible
   widget=TextStyleSammple_Bold_Big
   properties {
    text=Sample
    toolTipText="Does animation contain float channel curves.\r\n\r\n\r\nType name in the filed on the right."
    font {
     bForceMonospaced=false
     bMaterialIsStencil=false
     fontMaterial=null
     fontObject=&Roboto
     letterSpacing=100
     monospacedWidth=1
     size=10.5
     skewAmount=0
     typefaceFontName=Bold
     outlineSettings {
      bApplyOutlineToDropShadows=false
      bMiteredCorners=false
      bSeparateFillAlpha=false
      outlineMaterial=null
      outlineSize=0
      outlineColor=<rgba 0 0 0 1>
     }
    }
   }
   slot {class=ScrollBoxSlot, properties {horizontalAlignment=HAlign_Left, verticalAlignment=VAlign_Center}}
  }
  BorderStyle {
   widgetClassPath=&Border
   parent=Invisible
   widget=BorderStyle
   properties {
    background {
     bIsDynamicallyLoaded=false
     drawAs=RoundedBox
     imageType=NoImage
     mirroring=NoMirror
     resourceName=None
     resourceObject=null
     tiling=NoTile
     imageSize=<32 32>
     margin=<ltrb 0 0 0 0>
     outlineSettings {
      bUseBrushTransparency=false
      roundingType=FixedRadius
      width=1
      color {colorUseRule=UseColor_Specified, specifiedColor=<rgba 1 1 1 0.049>}
      cornerRadii=<10 10 10 1>
     }
     tintColor {colorUseRule=UseColor_Specified, specifiedColor=<rgba 1 1 1 1>}
     uVRegion {bIsValid=false, max=<0 0>, min=<0 0>}
    }
    brushColor=<rgba 0 0 0 0.409>
   }
   slot {class=ScrollBoxSlot, properties {horizontalAlignment=HAlign_Center, verticalAlignment=VAlign_Center, padding=<ltrb 10 10 10 10>}}
  }
  wahtever {
   widgetClassPath=&TextBlock
   bIsVariable=true
   parent=BorderStyle
   widget=wahtever
   properties {
    text="Blahblahblahbalah\r\ndassadss\r\ndasdas"
    toolTipText="Does animation contain float channel curves.\r\n\r\n\r\nType name in the filed on the right."
    font {
     bForceMonospaced=false
     bMaterialIsStencil=false
     fontMaterial=null
     fontObject=&Roboto
     letterSpacing=100
     monospacedWidth=1
     size=10.5
     skewAmount=0
     typefaceFontName=Bold
     outlineSettings {
      bApplyOutlineToDropShadows=false
      bMiteredCorners=false
      bSeparateFillAlpha=false
      outlineMaterial=null
      outlineSize=0
      outlineColor=<rgba 0 0 0 1>
     }
    }
   }
   slot {class=BorderSlot}
  }
  HorizontalBoxCurves {widgetClassPath=&HorizontalBox, parent=Invisible, widget=HorizontalBoxCurves, slot {class=ScrollBoxSlot}}
  Checkbox_Style {
   widgetClassPath=&EditorUtilityCheckBox
   bIsVariable=true
   parent=HorizontalBoxCurves
   widget=Checkbox_Style
   properties {horizontalAlignment=HAlign_Left, toolTipText="Does animation contain float channel curves.\r\n\r\n\r\nType name in the filed on the right."}
   slot {class=HorizontalBoxSlot, properties {horizontalAlignment=HAlign_Left, verticalAlignment=VAlign_Center, padding=<ltrb 30 0 10 0>}}
  }
  Spacer_124 {
   widgetClassPath=&Spacer
   parent=HorizontalBoxCurves
   widget=Spacer_124
   slot {class=HorizontalBoxSlot, properties {size {sizeRule=Fill, value=1}}}
  }
  TextBlock {
   widgetClassPath=&TextBlock
   parent=HorizontalBoxCurves
   widget=TextBlock
   properties {
    text="Has Curve(s):"
    toolTipText="Does animation contain float channel curves.\r\n\r\n\r\nType name in the filed on the right."
    font {
     bForceMonospaced=false
     bMaterialIsStencil=false
     fontMaterial=null
     fontObject=&Roboto
     letterSpacing=0
     monospacedWidth=1
     size=9
     skewAmount=0
     typefaceFontName=Regular
     outlineSettings {
      bApplyOutlineToDropShadows=false
      bMiteredCorners=false
      bSeparateFillAlpha=false
      outlineMaterial=null
      outlineSize=0
      outlineColor=<rgba 0 0 0 1>
     }
    }
   }
   slot {class=HorizontalBoxSlot, properties {horizontalAlignment=HAlign_Left, verticalAlignment=VAlign_Center}}
  }
  EditableText_Style {
   widgetClassPath=&EditorUtilityEditableText
   bIsVariable=true
   parent=HorizontalBoxCurves
   widget=EditableText_Style
   properties {
    hintText="[  CURVE NAME(S)  ]"
    justification=Right
    text=TEST
    toolTipText="Type curve(s) name in this field. \r\nMultiple curve names should be separated by commas"
    widgetStyle {
     backgroundImageComposing {
      bIsDynamicallyLoaded=false
      drawAs=Border
      imageType=FullColor
      mirroring=NoMirror
      resourceName=../../../Engine/Content/Slate/Old/HyperlinkDotted.png
      resourceObject=null
      tiling=Both
      imageSize=<0 0>
      margin=<ltrb 0 0 0 0.1875>
      outlineSettings {
       bUseBrushTransparency=false
       roundingType=HalfHeightRadius
       width=0
       color {colorUseRule=UseColor_Specified, specifiedColor=<rgba 0 0 0 0>}
       cornerRadii=<0 0 0 0>
      }
      tintColor {colorUseRule=UseColor_Specified, specifiedColor=<rgba 1 1 1 1>}
      uVRegion {bIsValid=false, max=<0 0>, min=<0 0>}
     }
     backgroundImageSelected {
      bIsDynamicallyLoaded=false
      drawAs=Image
      imageType=NoImage
      mirroring=NoMirror
      resourceName=None
      resourceObject=null
      tiling=NoTile
      imageSize=<0 0>
      margin=<ltrb 0 0 0 0>
      outlineSettings {
       bUseBrushTransparency=false
       roundingType=HalfHeightRadius
       width=0
       color {colorUseRule=UseColor_Specified, specifiedColor=<rgba 0 0 0 0>}
       cornerRadii=<0 0 0 0>
      }
      tintColor {colorUseRule=UseColor_Specified, specifiedColor=<rgba 0 0.162 0.7454 1>}
      uVRegion {bIsValid=false, max=<0 0>, min=<0 0>}
     }
     caretImage {
      bIsDynamicallyLoaded=false
      drawAs=Image
      imageType=NoImage
      mirroring=NoMirror
      resourceName=None
      resourceObject=null
      tiling=NoTile
      imageSize=<0 0>
      margin=<ltrb 0 0 0 0>
      outlineSettings {
       bUseBrushTransparency=false
       roundingType=HalfHeightRadius
       width=0
       color {colorUseRule=UseColor_Specified, specifiedColor=<rgba 0 0 0 0>}
       cornerRadii=<0 0 0 0>
      }
      tintColor {colorUseRule=UseColor_Specified, specifiedColor=<rgba 1 1 1 1>}
      uVRegion {bIsValid=false, max=<0 0>, min=<0 0>}
     }
     colorAndOpacity {colorUseRule=UseColor_Specified, specifiedColor=<rgba 1 1 1 1>}
     font {
      bForceMonospaced=false
      bMaterialIsStencil=false
      fontMaterial=null
      fontObject=&Roboto
      letterSpacing=10
      monospacedWidth=1
      size=9
      skewAmount=0.12
      typefaceFontName=Regular
      outlineSettings {
       bApplyOutlineToDropShadows=false
       bMiteredCorners=false
       bSeparateFillAlpha=false
       outlineMaterial=null
       outlineSize=0
       outlineColor=<rgba 0 0 0 1>
      }
     }
     hintColorAndOpacity {colorUseRule=UseColor_Foreground, specifiedColor=<rgba 1 0 1 1>}
     selectedBackgroundColor {colorUseRule=UseColor_Foreground, specifiedColor=<rgba 1 0 1 1>}
    }
   }
   slot {
    class=HorizontalBoxSlot
    properties {horizontalAlignment=HAlign_Left, verticalAlignment=VAlign_Center, padding=<ltrb 10 0 10 0>, size {sizeRule=Fill, value=1}}
   }
  }
 ]
}
