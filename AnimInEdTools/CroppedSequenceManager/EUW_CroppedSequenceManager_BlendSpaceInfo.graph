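# Editor utility widget row for the Cropped Sequence Manager tool that shows a BlendSpace picker and an animation type dropdown (mesh-space, local-space or non-additive).
# PreConstruct and Construct point the Single Property View at the widget's own BlendSpace variable, then reset it to empty.
# An Execute Python Script node targets the CroppedSequenceManagerUI module with the widget id, but neither it nor the property-changed event is wired.
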EditorUtilityWidgetBlueprint /AnimInEdTools/CroppedSequenceManager/EUW_CroppedSequenceManager_BlendSpaceInfo
graphs [EventGraph {nodes [
   Event PreConstruct()
   > SetObject(
    self=Get SinglePropertyView_AnimSequence() -> (SinglePropertyView_AnimSequence:Single Property View Object Reference)
    NewObject=Self-Reference()
   )
   > Set BlendSpace(BlendSpace:Blend Space Object Reference) -> (Output_Get:Blend Space Object Reference)
   Event Construct()
   > SetObject(self=Get SinglePropertyView_AnimSequence(), NewObject=Self-Reference())
   > Set BlendSpace()
   "On Property Changed (SinglePropertyView_AnimSequence)"!() -> (OutputDelegate, then, PropertyName:Name=None) {is=title K2Node_ComponentBoundEvent}
   Event Tick()
   Execute Python Script!(
    execute
    PythonScript="import sys\r\nCSM = sys.modules['CroppedSequenceManagerUI']\r\n\r\nCSM.cls.linked_anim_sequence_property_changed(anim_sequence, widget_id)\r\n"
    anim_sequence:Wildcard
    widget_id:int=Get widget id() -> (widget id:int=0)
   ) -> (then, ReturnValue) {inputs [anim_sequence, widget_id]}
  ]}]
variables [
 BlendSpace {type=BlendSpace, default=null}
 widget id {type=integer}
]
widgets {
 info {rootWidgetClass=&VerticalBox, widgetCount=6}
 widgets [
  VerticalBox_0 {widgetClassPath=&VerticalBox, widget=VerticalBox_0}
  HorizontalBox_0 {
   widgetClassPath=&HorizontalBox
   parent=VerticalBox_0
   widget=HorizontalBox_0
   slot {class=VerticalBoxSlot, properties {horizontalAlignment=HAlign_Left, verticalAlignment=VAlign_Center}}
  }
  SinglePropertyView_AnimSequence {
   widgetClassPath=&SinglePropertyView
   parent=HorizontalBox_0
   widget=SinglePropertyView_AnimSequence
   properties {nameOverride=" ", propertyName=BlendSpace}
   slot {class=HorizontalBoxSlot, properties {horizontalAlignment=HAlign_Left, verticalAlignment=VAlign_Center, padding=<ltrb -16 2 0 2>}}
  }
  HorizontalBox {
   widgetClassPath=&HorizontalBox
   parent=HorizontalBox_0
   widget=HorizontalBox
   slot {class=HorizontalBoxSlot, properties {horizontalAlignment=HAlign_Left, verticalAlignment=VAlign_Top, padding=<ltrb 0 2 0 0>}}
  }
  Add_Track_Text_1 {
   widgetClassPath=&TextBlock
   parent=HorizontalBox
   widget=Add_Track_Text_1
   properties {
    text=Type:
    toolTipText="Type of animation used in BlendSpace.\r\n\r\nAimOffsets are usually a MeshSpace Additives."
    colorAndOpacity {colorUseRule=UseColor_Specified, specifiedColor=<rgba 0.1545 0.1545 0.1545 1>}
    font {
     bForceMonospaced=false
     bMaterialIsStencil=false
     fontMaterial=null
     fontObject=&Roboto
     letterSpacing=0
     monospacedWidth=1
     size=9
     skewAmount=0
     typefaceFontName=Bold
     outlineSettings {
      bApplyOutlineToDropShadows=false
      bMiteredCorners=false
      bSeparateFillAlpha=false
      outlineMaterial=null
      outlineSize=0
      outlineColor=<rgba 0 0 0 1>
     }
    }
   }
   slot {class=HorizontalBoxSlot, properties {horizontalAlignment=HAlign_Center, verticalAlignment=VAlign_Center, padding=<ltrb 0 0 5 0>}}
  }
  ComboBox_AnimType {
   widgetClassPath=/Script/Blutility.EditorUtilityComboBoxString
   bIsVariable=true
   parent=HorizontalBox
   widget=ComboBox_AnimType
   properties {
    selectedOption=MeshSpace Additive
    toolTipText="Type of animation used in BlendSpace.\r\n\r\nAimOffsets are usually a MeshSpace Additives."
    contentPadding=<ltrb 4 2 10 2>
    defaultOptions [MeshSpace Additive, LocalSpace Additive, Not Additive]
    font {
     bForceMonospaced=false
     bMaterialIsStencil=false
     fontMaterial=null
     fontObject=&Roboto
     letterSpacing=0
     monospacedWidth=1
     size=9
     skewAmount=0
     typefaceFontName=Bold
     outlineSettings {
      bApplyOutlineToDropShadows=false
      bMiteredCorners=false
      bSeparateFillAlpha=false
      outlineMaterial=null
      outlineSize=0
      outlineColor=<rgba 0 0 0 1>
     }
    }
    foregroundColor {colorUseRule=UseColor_Specified, specifiedColor=<rgba 0.5271 0.5271 0.5271 1>}
   }
   slot {class=HorizontalBoxSlot, properties {horizontalAlignment=HAlign_Center, verticalAlignment=VAlign_Center}}
  }
 ]
}
